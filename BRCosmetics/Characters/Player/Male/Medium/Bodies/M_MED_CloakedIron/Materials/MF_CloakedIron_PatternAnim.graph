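# Material function for the CloakedIron character that layers an animated fire pattern onto incoming material attributes.
# Panned, MF_WaveUV-distorted SDF textures and an erosion mask form smoothstepped shapes and outlines driving color, emissive and roughness, blended in through a mask texture.
# An outer fire outline adjusts base color and roughness, deriving normals from its distance field.
# Static switches provide variants, alongside scaled, invertible time and polar procedural UVs for the backpack.
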MaterialFunction /BRCosmetics/Characters/Player/Male/Medium/Bodies/M_MED_CloakedIron/Materials/MF_CloakedIron_PatternAnim
graph {expressions {
  MaterialExpressionAbs_0 {class=MaterialExpressionAbs, inputs [None {from=MaterialExpressionMaterialFunctionCall_16, output=R}], properties {}}
  MaterialExpressionAdd_10 {
   class=MaterialExpressionAdd
   inputs [
    A {from=MaterialExpressionGetMaterialAttributes_2, output=Roughness}
    B {from=MaterialExpressionMultiply_26}
   ]
  }
  MaterialExpressionAdd_11 {
   class=MaterialExpressionAdd
   inputs [
    A {from=MaterialExpressionAdd_12}
    B {from=MaterialExpressionScalarParameter_30}
   ]
   properties {constB=0.01}
  }
  MaterialExpressionAdd_12 {
   class=MaterialExpressionAdd
   inputs [
    A {from=MaterialExpressionMultiply_31}
    B {from=MaterialExpressionScalarParameter_16}
   ]
  }
  MaterialExpressionAdd_13 {
   class=MaterialExpressionAdd
   inputs [
    A {from=MaterialExpressionAdd_15}
    B {from=MaterialExpressionScalarParameter_31}
   ]
   properties {constB=0.001}
  }
  MaterialExpressionAdd_14 {
   class=MaterialExpressionAdd
   inputs [
    A {from=MaterialExpressionTextureCoordinate_5}
    B {from=MaterialExpressionConvert_2}
   ]
  }
  MaterialExpressionAdd_15 {
   class=MaterialExpressionAdd
   inputs [
    A {from=MaterialExpressionAdd_12}
    B {from=MaterialExpressionScalarParameter_18}
   ]
  }
  MaterialExpressionAdd_17 {
   class=MaterialExpressionAdd
   inputs [
    A {from=MaterialExpressionOneMinus_0}
    B {from=""}
   ]
   properties {constB=0.5}
  }
  MaterialExpressionAdd_19 {
   class=MaterialExpressionAdd
   inputs [
    A {from=MaterialExpressionVertexInterpolator_1, output=PS}
    B {from=MaterialExpressionAppendVector_4}
   ]
  }
  MaterialExpressionAdd_2 {
   class=MaterialExpressionAdd
   inputs [
    A {from=MaterialExpressionMultiply_4}
    B {from=MaterialExpressionStaticSwitch_1}
   ]
  }
  MaterialExpressionAdd_20 {
   class=MaterialExpressionAdd
   inputs [
    A {from=MaterialExpressionDivide_5}
    B {from=MaterialExpressionScalarParameter_38}
   ]
  }
  MaterialExpressionAdd_21 {
   class=MaterialExpressionAdd
   inputs [
    A {from=MaterialExpressionGetMaterialAttributes_3, output=EmissiveColor}
    B {from=MaterialExpressionMultiply_41}
   ]
  }
  MaterialExpressionAdd_5 {
   class=MaterialExpressionAdd
   inputs [
    A {from=MaterialExpressionMultiply_27}
    B {from=MaterialExpressionSubtract_9}
   ]
   properties {constA=0.05}
  }
  MaterialExpressionAdd_7 {
   class=MaterialExpressionAdd
   inputs [
    A {from=MaterialExpressionMultiply_27}
    B {from=MaterialExpressionOneMinus_4}
   ]
   properties {constA=0.05}
  }
  MaterialExpressionAdd_9 {
   class=MaterialExpressionAdd
   inputs [
    A {from=MaterialExpressionMultiply_24}
    B {from=MaterialExpressionNamedRerouteUsage_30}
   ]
  }
  MaterialExpressionAppendVector_0 {
   class=MaterialExpressionAppendVector
   inputs [
    A {from=MaterialExpressionMaterialFunctionCall_12, output=G}
    B {from=MaterialExpressionMaterialFunctionCall_12, output=G}
   ]
  }
  MaterialExpressionAppendVector_4 {
   class=MaterialExpressionAppendVector
   inputs [
    A {from=MaterialExpressionConstant_18, output=""}
    B {from=MaterialExpressionScalarParameter_35, output=""}
   ]
  }
  MaterialExpressionAppendVector_5 {
   class=MaterialExpressionAppendVector
   inputs [
    A {from=MaterialExpressionMultiply_40, output=""}
    B {from=MaterialExpressionDivide_7, output=""}
   ]
  }
  MaterialExpressionAppendVector_6 {
   class=MaterialExpressionAppendVector
   inputs [
    A {from=MaterialExpressionAbs_0, output=""}
    B {from=MaterialExpressionMaterialFunctionCall_16, output=G}
   ]
  }
  MaterialExpressionArctangent2Fast_1 {
   class=MaterialExpressionArctangent2Fast
   inputs [
    Y {from=MaterialExpressionMaterialFunctionCall_14, output=G}
    X {from=MaterialExpressionMaterialFunctionCall_14, output=G}
   ]
   properties {}
  }
  MaterialExpressionBlendMaterialAttributes_0 {
   class=MaterialExpressionBlendMaterialAttributes
   inputs [
    A {from=MaterialExpressionNamedRerouteDeclaration_12, output=""}
    B {from=MaterialExpressionNamedRerouteUsage_10, output=""}
    Alpha {from=MaterialExpressionNamedRerouteDeclaration_21, output=""}
   ]
   properties {}
  }
  MaterialExpressionComponentMask_1 {class=MaterialExpressionComponentMask, inputs [None {from=MaterialExpressionLocalPosition_0, output=XYZ}], properties {r=true}}
  MaterialExpressionComponentMask_4 {class=MaterialExpressionComponentMask, inputs [None {from=MaterialExpressionLocalPosition_3, output=XYZ}], properties {b=true, r=true}}
  MaterialExpressionConstant2Vector_0 {class=MaterialExpressionConstant2Vector, properties {g=0.5, r=1}}
  MaterialExpressionConstantBiasScale_2 {class=MaterialExpressionConstantBiasScale, inputs [None {from=MaterialExpressionSine_2}]}
  MaterialExpressionConstant_0 {class=MaterialExpressionConstant, properties {r=2}}
  MaterialExpressionConstant_1 {class=MaterialExpressionConstant, properties {r=0.02}}
  MaterialExpressionConstant_10 {class=MaterialExpressionConstant, properties {r=3.5478}}
  MaterialExpressionConstant_11 {class=MaterialExpressionConstant, properties {r=0.0157}}
  MaterialExpressionConstant_12 {class=MaterialExpressionConstant, properties {r=0.359}}
  MaterialExpressionConstant_13 {class=MaterialExpressionConstant, properties {r=0.5}}
  MaterialExpressionConstant_14 {class=MaterialExpressionConstant}
  MaterialExpressionConstant_15 {class=MaterialExpressionConstant, properties {r=1}}
  MaterialExpressionConstant_18 {class=MaterialExpressionConstant}
  MaterialExpressionConstant_19 {class=MaterialExpressionConstant, properties {r=1}}
  MaterialExpressionConstant_3 {class=MaterialExpressionConstant, properties {r=1}}
  MaterialExpressionConstant_4 {class=MaterialExpressionConstant}
  MaterialExpressionConstant_5 {class=MaterialExpressionConstant, properties {r=1}}
  MaterialExpressionConstant_6 {class=MaterialExpressionConstant, properties {r=0.3486}}
  MaterialExpressionConvert_1 {
   class=MaterialExpressionConvert
   inputs [
    None {from=""}
    None {from=MaterialExpressionScalarParameter_24}
   ]
   properties {
    nodeName=Make Float2
    convertInputs [
     {type=Scalar, expressionInput {expression=null, inputName=None, mask=0, maskA=0, maskB=0, maskG=0, maskR=0, outputIndex=0}}
     {
      type=Scalar
      expressionInput {
       expression=&self:MaterialExpressionScalarParameter_24
       inputName=None
       mask=0
       maskA=0
       maskB=0
       maskG=0
       maskR=0
       outputIndex=0
      }
     }
    ]
    convertOutputs [{type=Vector2}]
   }
  }
  MaterialExpressionConvert_2 {
   class=MaterialExpressionConvert
   inputs [
    None {from=MaterialExpressionScalarParameter_27}
    None {from=""}
   ]
   properties {
    nodeName=Make Float2
    convertInputs [
     {
      type=Scalar
      expressionInput {
       expression=&self:MaterialExpressionScalarParameter_27
       inputName=None
       mask=0
       maskA=0
       maskB=0
       maskG=0
       maskR=0
       outputIndex=0
      }
     }
     {type=Scalar, expressionInput {expression=null, inputName=None, mask=0, maskA=0, maskB=0, maskG=0, maskR=0, outputIndex=0}}
    ]
    convertOutputs [{type=Vector2}]
   }
  }
  MaterialExpressionConvert_3 {
   class=MaterialExpressionConvert
   inputs [
    None {from=""}
    None {from=MaterialExpressionScalarParameter_28}
   ]
   properties {
    nodeName=Make Float2
    convertInputs [
     {type=Scalar, expressionInput {expression=null, inputName=None, mask=0, maskA=0, maskB=0, maskG=0, maskR=0, outputIndex=0}}
     {
      type=Scalar
      expressionInput {
       expression=&self:MaterialExpressionScalarParameter_28
       inputName=None
       mask=0
       maskA=0
       maskB=0
       maskG=0
       maskR=0
       outputIndex=0
      }
     }
    ]
    convertOutputs [{type=Vector2}]
   }
  }
  MaterialExpressionDivide_1 {
   class=MaterialExpressionDivide
   inputs [
    A {from=MaterialExpressionSubtract_13}
    B {from=MaterialExpressionScalarParameter_14}
   ]
  }
  MaterialExpressionDivide_5 {
   class=MaterialExpressionDivide
   inputs [
    A {from=MaterialExpressionArctangent2Fast_1}
    B {from=MaterialExpressionDivide_6}
   ]
  }
  MaterialExpressionDivide_6 {
   class=MaterialExpressionDivide
   inputs [
    A {from=""}
    B {from=""}
   ]
   properties {constA=710, constB=113}
  }
  MaterialExpressionDivide_7 {
   class=MaterialExpressionDivide
   inputs [
    A {from=MaterialExpressionLength_1}
    B {from=MaterialExpressionScalarParameter_36}
   ]
  }
  MaterialExpressionFrac_2 {class=MaterialExpressionFrac, inputs [None {from=MaterialExpressionAdd_20, output=""}], properties {}}
  MaterialExpressionFunctionInput_0 {class=MaterialExpressionFunctionInput, inputs [Preview {from="", output=""}], properties {inputType=FunctionInput_MaterialAttributes}}
  MaterialExpressionFunctionOutput_0 {class=MaterialExpressionFunctionOutput, inputs [None {from=MaterialExpressionSetMaterialAttributes_4, output=""}]}
  MaterialExpressionGetMaterialAttributes_0 {
   class=MaterialExpressionGetMaterialAttributes
   inputs [None {from=MaterialExpressionNamedRerouteUsage_3, output=""}]
   properties {attributeGetTypes [#1]}
  }
  MaterialExpressionGetMaterialAttributes_1 {class=MaterialExpressionGetMaterialAttributes, inputs [None {from=MaterialExpressionFunctionInput_0, output=""}]}
  MaterialExpressionGetMaterialAttributes_2 {
   class=MaterialExpressionGetMaterialAttributes
   inputs [None {from=MaterialExpressionBlendMaterialAttributes_0, output=""}]
   properties {attributeGetTypes [#2, #1, #3]}
  }
  MaterialExpressionGetMaterialAttributes_3 {
   class=MaterialExpressionGetMaterialAttributes
   inputs [None {from=MaterialExpressionStaticSwitch_0, output=""}]
   properties {attributeGetTypes [#4, #5]}
  }
  MaterialExpressionLength_1 {class=MaterialExpressionLength, inputs [None {from=MaterialExpressionAdd_19, output=""}], properties {}}
  MaterialExpressionLinearInterpolate_0 {
   class=MaterialExpressionLinearInterpolate
   inputs [
    A {from=MaterialExpressionVectorParameter_0, output=RGB}
    B {from=MaterialExpressionVectorParameter_1, output=RGB}
    Alpha {from=MaterialExpressionNamedRerouteUsage_26}
   ]
  }
  MaterialExpressionLinearInterpolate_1 {
   class=MaterialExpressionLinearInterpolate
   inputs [
    A {from=MaterialExpressionLinearInterpolate_0}
    B {from=MaterialExpressionVectorParameter_2, output=RGB}
    Alpha {from=MaterialExpressionNamedRerouteUsage_27}
   ]
  }
  MaterialExpressionLinearInterpolate_10 {
   class=MaterialExpressionLinearInterpolate
   inputs [
    A {from=MaterialExpressionGetMaterialAttributes_2, output=BaseColor}
    B {from=""}
    Alpha {from=MaterialExpressionNamedRerouteUsage_34}
   ]
   properties {constB=0.015}
  }
  MaterialExpressionLinearInterpolate_6 {
   class=MaterialExpressionLinearInterpolate
   inputs [
    A {from=MaterialExpressionGetMaterialAttributes_0, output=Roughness}
    B {from=MaterialExpressionSubtract_8}
    Alpha {from=MaterialExpressionNamedRerouteUsage_4}
   ]
   properties {constB=0.6}
  }
  MaterialExpressionLinearInterpolate_7 {
   class=MaterialExpressionLinearInterpolate
   inputs [
    A {from=MaterialExpressionLinearInterpolate_6}
    B {from=""}
    Alpha {from=MaterialExpressionNamedRerouteUsage_5}
   ]
   properties {constB=0.3}
  }
  MaterialExpressionLinearInterpolate_8 {
   class=MaterialExpressionLinearInterpolate
   inputs [
    A {from=""}
    B {from=""}
    Alpha {from=MaterialExpressionConstantBiasScale_2}
   ]
   properties {constA=0.8}
  }
  MaterialExpressionLocalPosition_0 {class=MaterialExpressionLocalPosition, inputs [], properties {}}
  MaterialExpressionLocalPosition_3 {class=MaterialExpressionLocalPosition, inputs [], properties {localOrigin=InstancePreSkinning}}
  MaterialExpressionMaterialFunctionCall_0 {
   class=MaterialExpressionMaterialFunctionCall
   inputs [
    UV {from=MaterialExpressionPanner_0}
    U Frequency {from=MaterialExpressionConstant_0}
    V Frequency {from=MaterialExpressionConstant_0}
    U Amplitude {from=MaterialExpressionConstant_1}
    V Amplitude {from=MaterialExpressionConstant_1}
    U Speed {from=MaterialExpressionConstant_3}
    V Speed {from=MaterialExpressionConstant_3}
    U Time {from=MaterialExpressionNamedRerouteUsage_56}
    V Time {from=MaterialExpressionNamedRerouteUsage_56}
   ]
   properties {materialFunction=/Game/Effects/Fort_Effects/MaterialFunctions/MF_WaveUV}
  }
  MaterialExpressionMaterialFunctionCall_10 {
   class=MaterialExpressionMaterialFunctionCall
   inputs [
    BaseNormal {from=MaterialExpressionGetMaterialAttributes_2, output=Normal}
    AdditionalNormal {from=MaterialExpressionNamedRerouteUsage_33}
   ]
   properties {materialFunction=&BlendAngleCorrectedNormals}
  }
  MaterialExpressionMaterialFunctionCall_11 {
   class=MaterialExpressionMaterialFunctionCall
   inputs [Float4 {from=MaterialExpressionTextureSampleParameter2D_0, output=RGBA}]
   properties {materialFunction=&BreakOutFloat4Components}
  }
  MaterialExpressionMaterialFunctionCall_12 {
   class=MaterialExpressionMaterialFunctionCall
   inputs [Float2 {from=MaterialExpressionTextureCoordinate_0}]
   properties {materialFunction=&BreakOutFloat2Components}
  }
  MaterialExpressionMaterialFunctionCall_14 {
   class=MaterialExpressionMaterialFunctionCall
   inputs [Float2 {from=MaterialExpressionAdd_19}]
   properties {materialFunction=&BreakOutFloat2Components}
  }
  MaterialExpressionMaterialFunctionCall_16 {
   class=MaterialExpressionMaterialFunctionCall
   inputs [Float2 {from=MaterialExpressionComponentMask_4}]
   properties {materialFunction=&BreakOutFloat2Components}
  }
  MaterialExpressionMaterialFunctionCall_2 {
   class=MaterialExpressionMaterialFunctionCall
   inputs [
    UV {from=MaterialExpressionPanner_1}
    U Frequency {from=MaterialExpressionScalarParameter_25}
    V Frequency {from=MaterialExpressionScalarParameter_25}
    U Amplitude {from=MaterialExpressionMultiply_1}
    V Amplitude {from=MaterialExpressionMultiply_1}
    U Speed {from=""}
    V Speed {from=MaterialExpressionConstant_6}
    U Time {from=MaterialExpressionNamedRerouteUsage_54}
    V Time {from=MaterialExpressionNamedRerouteUsage_54}
   ]
   properties {materialFunction=&MF_WaveUV}
  }
  MaterialExpressionMaterialFunctionCall_5 {
   class=MaterialExpressionMaterialFunctionCall
   inputs [
    UV {from=MaterialExpressionPanner_3}
    U Frequency {from=MaterialExpressionConstant_10}
    V Frequency {from=MaterialExpressionConstant_10}
    U Amplitude {from=MaterialExpressionMultiply_21}
    V Amplitude {from=MaterialExpressionMultiply_21}
    U Speed {from=MaterialExpressionConstant_12}
    V Speed {from=MaterialExpressionConstant_12}
    U Time {from=MaterialExpressionNamedRerouteUsage_57}
    V Time {from=MaterialExpressionNamedRerouteUsage_57}
   ]
   properties {materialFunction=&MF_WaveUV}
  }
  MaterialExpressionMaterialFunctionCall_6 {
   class=MaterialExpressionMaterialFunctionCall
   inputs [
    Normal {from=MaterialExpressionMaterialFunctionCall_8, output=Normal}
    Flatness {from=MaterialExpressionOneMinus_5}
   ]
   properties {materialFunction=&FlattenNormal}
  }
  MaterialExpressionMaterialFunctionCall_7 {
   class=MaterialExpressionMaterialFunctionCall
   inputs [Float3 {from=MaterialExpressionNamedRerouteDeclaration_24}]
   properties {materialFunction=&BreakOutFloat3Components}
  }
  MaterialExpressionMaterialFunctionCall_8 {
   class=MaterialExpressionMaterialFunctionCall
   inputs [
    Right Shifted Distance Field {from=MaterialExpressionMaterialFunctionCall_7, output=B}
    Downward Shifted Distance Field {from=MaterialExpressionMaterialFunctionCall_7, output=B}
    Unshifted distance field {from=MaterialExpressionMaterialFunctionCall_7, output=B}
    Dissolve Alpha {from=MaterialExpressionConstant_13}
    Normal Lip Tightness {from=MaterialExpressionConstant_14}
    Normal Lip Curvature {from=MaterialExpressionConstant_15}
   ]
   properties {materialFunction=/Engine/Functions/Engine_MaterialFunctions02/Normals/DynamicNormalFromDistanceField}
  }
  MaterialExpressionMultiply_1 {
   class=MaterialExpressionMultiply
   inputs [
    A {from=MaterialExpressionScalarParameter_20}
    B {from=MaterialExpressionNamedRerouteUsage_15}
   ]
  }
  MaterialExpressionMultiply_10 {
   class=MaterialExpressionMultiply
   inputs [
    A {from=MaterialExpressionMultiply_8}
    B {from=MaterialExpressionSaturate_7}
   ]
  }
  MaterialExpressionMultiply_11 {
   class=MaterialExpressionMultiply
   inputs [
    A {from=MaterialExpressionNamedRerouteUsage_40}
    B {from=MaterialExpressionScalarParameter_2}
   ]
  }
  MaterialExpressionMultiply_2 {
   class=MaterialExpressionMultiply
   inputs [
    A {from=MaterialExpressionVectorParameter_2, output=RGB}
    B {from=MaterialExpressionNamedRerouteUsage_28}
   ]
  }
  MaterialExpressionMultiply_21 {
   class=MaterialExpressionMultiply
   inputs [
    A {from=MaterialExpressionConstant_11}
    B {from=MaterialExpressionTextureSample_4, output=R}
   ]
  }
  MaterialExpressionMultiply_22 {
   class=MaterialExpressionMultiply
   inputs [
    A {from=MaterialExpressionScalarParameter_9}
    B {from=MaterialExpressionNamedRerouteUsage_38}
   ]
  }
  MaterialExpressionMultiply_23 {
   class=MaterialExpressionMultiply
   inputs [
    A {from=MaterialExpressionScalarParameter_11}
    B {from=MaterialExpressionMultiply_25}
   ]
  }
  MaterialExpressionMultiply_24 {
   class=MaterialExpressionMultiply
   inputs [
    A {from=MaterialExpressionTime_2}
    B {from=""}
   ]
   properties {constB=0.25}
  }
  MaterialExpressionMultiply_25 {
   class=MaterialExpressionMultiply
   inputs [
    A {from=MaterialExpressionLinearInterpolate_8}
    B {from=MaterialExpressionSaturate_8}
   ]
  }
  MaterialExpressionMultiply_26 {
   class=MaterialExpressionMultiply
   inputs [
    A {from=MaterialExpressionNamedRerouteUsage_34}
    B {from=""}
   ]
   properties {constB=0.2}
  }
  MaterialExpressionMultiply_27 {
   class=MaterialExpressionMultiply
   inputs [
    A {from=MaterialExpressionScalarParameter_10}
    B {from=MaterialExpressionNamedRerouteUsage_38}
   ]
  }
  MaterialExpressionMultiply_28 {
   class=MaterialExpressionMultiply
   inputs [
    A {from=MaterialExpressionScalarParameter_13}
    B {from=MaterialExpressionNamedRerouteUsage_6}
   ]
  }
  MaterialExpressionMultiply_29 {
   class=MaterialExpressionMultiply
   inputs [
    A {from=MaterialExpressionPower_3}
    B {from=""}
   ]
   properties {constB=2}
  }
  MaterialExpressionMultiply_3 {
   class=MaterialExpressionMultiply
   inputs [
    A {from=MaterialExpressionNamedRerouteUsage_44}
    B {from=MaterialExpressionMultiply_9}
   ]
  }
  MaterialExpressionMultiply_31 {
   class=MaterialExpressionMultiply
   inputs [
    A {from=MaterialExpressionNamedRerouteUsage_37}
    B {from=MaterialExpressionScalarParameter_17}
   ]
   properties {constB=0.1}
  }
  MaterialExpressionMultiply_34 {
   class=MaterialExpressionMultiply
   inputs [
    A {from=MaterialExpressionSign_0}
    B {from=""}
   ]
   properties {constB=-1}
  }
  MaterialExpressionMultiply_35 {
   class=MaterialExpressionMultiply
   inputs [
    A {from=MaterialExpressionNamedRerouteUsage_59}
    B {from=MaterialExpressionTime_3}
   ]
  }
  MaterialExpressionMultiply_36 {
   class=MaterialExpressionMultiply
   inputs [
    A {from=MaterialExpressionMultiply_35}
    B {from=MaterialExpressionScalarParameter_29}
   ]
  }
  MaterialExpressionMultiply_37 {
   class=MaterialExpressionMultiply
   inputs [
    A {from=MaterialExpressionPower_2}
    B {from=MaterialExpressionNamedRerouteUsage_32}
   ]
  }
  MaterialExpressionMultiply_38 {
   class=MaterialExpressionMultiply
   inputs [
    A {from=MaterialExpressionAppendVector_5}
    B {from=MaterialExpressionScalarParameter_33}
   ]
  }
  MaterialExpressionMultiply_4 {
   class=MaterialExpressionMultiply
   inputs [
    A {from=MaterialExpressionMultiply_2}
    B {from=MaterialExpressionScalarParameter_1}
   ]
  }
  MaterialExpressionMultiply_40 {
   class=MaterialExpressionMultiply
   inputs [
    A {from=MaterialExpressionFrac_2}
    B {from=MaterialExpressionScalarParameter_39}
   ]
  }
  MaterialExpressionMultiply_41 {
   class=MaterialExpressionMultiply
   inputs [
    A {from=MaterialExpressionGetMaterialAttributes_3, output=SubsurfaceColor}
    B {from=MaterialExpressionScalarParameter_41}
   ]
  }
  MaterialExpressionMultiply_5 {
   class=MaterialExpressionMultiply
   inputs [
    A {from=MaterialExpressionTextureCoordinate_3}
    B {from=MaterialExpressionScalarParameter_8}
   ]
  }
  MaterialExpressionMultiply_8 {
   class=MaterialExpressionMultiply
   inputs [
    A {from=MaterialExpressionNamedRerouteUsage_7}
    B {from=MaterialExpressionVectorParameter_3, output=RGB}
   ]
  }
  MaterialExpressionMultiply_9 {
   class=MaterialExpressionMultiply
   inputs [
    A {from=MaterialExpressionScalarParameter_5}
    B {from=MaterialExpressionConstant2Vector_0}
   ]
  }
  MaterialExpressionNamedRerouteDeclaration_0 {
   class=MaterialExpressionNamedRerouteDeclaration
   inputs [None {from=MaterialExpressionTextureSample_2, output=R}]
   properties {name=SDF 1, nodeColor=<rgba 0 1 0.3203 1>}
  }
  MaterialExpressionNamedRerouteDeclaration_10 {
   class=MaterialExpressionNamedRerouteDeclaration
   inputs [None {from=MaterialExpressionSetMaterialAttributes_0}]
   properties {name=FirePatternMat, nodeColor=<rgba 1 0.2344 0 1>}
  }
  MaterialExpressionNamedRerouteDeclaration_12 {
   class=MaterialExpressionNamedRerouteDeclaration
   inputs [None {from=MaterialExpressionStaticSwitch_3}]
   properties {name=InMat, nodeColor=<rgba 0 1 0.2031 1>}
  }
  MaterialExpressionNamedRerouteDeclaration_19 {
   class=MaterialExpressionNamedRerouteDeclaration
   inputs [None {from=MaterialExpressionNamedRerouteUsage_42}]
   properties {name=Blended Shape, nodeColor=<rgba 0.5234 0 1 1>}
  }
  MaterialExpressionNamedRerouteDeclaration_20 {
   class=MaterialExpressionNamedRerouteDeclaration
   inputs [None {from=MaterialExpressionNamedRerouteUsage_41}]
   properties {name=Blended Outline, nodeColor=<rgba 0 1 0.2031 1>}
  }
  MaterialExpressionNamedRerouteDeclaration_21 {
   class=MaterialExpressionNamedRerouteDeclaration
   inputs [None {from=MaterialExpressionMaterialFunctionCall_11, output=R}]
   properties {name=Mask, nodeColor=<rgba 0 0.4375 1 1>}
  }
  MaterialExpressionNamedRerouteDeclaration_22 {
   class=MaterialExpressionNamedRerouteDeclaration
   inputs [None {from=MaterialExpressionTextureSample_3, output=RGB}]
   properties {name=Normal_SDF, nodeColor=<rgba 0 1 0.9766 1>}
  }
  MaterialExpressionNamedRerouteDeclaration_23 {
   class=MaterialExpressionNamedRerouteDeclaration
   inputs [None {from=MaterialExpressionTextureSample_3, output=A}]
   properties {name=Normal_Random, nodeColor=<rgba 1 0.6562 0 1>}
  }
  MaterialExpressionNamedRerouteDeclaration_24 {
   class=MaterialExpressionNamedRerouteDeclaration
   inputs [None {from=MaterialExpressionSaturate_6}]
   properties {name=Normal_Outline, nodeColor=<rgba 0.6406 1 0 1>}
  }
  MaterialExpressionNamedRerouteDeclaration_26 {
   class=MaterialExpressionNamedRerouteDeclaration
   inputs [None {from=MaterialExpressionMaterialFunctionCall_11, output=B}]
   properties {name=Normal Gradient, nodeColor=<rgba 0.9922 0 1 1>}
  }
  MaterialExpressionNamedRerouteDeclaration_27 {
   class=MaterialExpressionNamedRerouteDeclaration
   inputs [None {from=MaterialExpressionMaterialFunctionCall_6, output=Result}]
   properties {name=Normal_Normal, nodeColor=<rgba 0 1 0.6719 1>}
  }
  MaterialExpressionNamedRerouteDeclaration_28 {
   class=MaterialExpressionNamedRerouteDeclaration
   inputs [None {from=MaterialExpressionTextureSample_1, output=R}]
   properties {name=ShapeThicknessMask, nodeColor=<rgba 1 0.7734 0 1>}
  }
  MaterialExpressionNamedRerouteDeclaration_29 {
   class=MaterialExpressionNamedRerouteDeclaration
   inputs [None {from=MaterialExpressionMaterialFunctionCall_11, output=A}]
   properties {name=Normal Mask, nodeColor=<rgba 1 0 0.3281 1>}
  }
  MaterialExpressionNamedRerouteDeclaration_30 {
   class=MaterialExpressionNamedRerouteDeclaration
   inputs [None {from=MaterialExpressionStaticBoolParameter_0}]
   properties {name=Simple FX, nodeColor=<rgba 1 0.1641 0 1>}
  }
  MaterialExpressionNamedRerouteDeclaration_31 {
   class=MaterialExpressionNamedRerouteDeclaration
   inputs [None {from=MaterialExpressionStaticSwitch_4}]
   properties {name=Inner Pattern UV, nodeColor=<rgba 0 0.0156 1 1>}
  }
  MaterialExpressionNamedRerouteDeclaration_32 {
   class=MaterialExpressionNamedRerouteDeclaration
   inputs [None {from=MaterialExpressionMultiply_36}]
   properties {name=Custom Time, nodeColor=<rgba 1 0.9141 0 1>}
  }
  MaterialExpressionNamedRerouteDeclaration_33 {
   class=MaterialExpressionNamedRerouteDeclaration
   inputs [None {from=MaterialExpressionStaticSwitch_5}]
   properties {name=Time Inversion Sign, nodeColor=<rgba 0.875 1 0 1>}
  }
  MaterialExpressionNamedRerouteDeclaration_34 {
   class=MaterialExpressionNamedRerouteDeclaration
   inputs [None {from=MaterialExpressionMultiply_38}]
   properties {name=Backpack Procedural UVs, nodeColor=<rgba 0.0547 1 0 1>}
  }
  MaterialExpressionNamedRerouteDeclaration_35 {
   class=MaterialExpressionNamedRerouteDeclaration
   inputs [None {from=MaterialExpressionTextureObject_0}]
   properties {name=Fire Pattern SDF Texture Object, nodeColor=<rgba 1 0.2109 0 1>}
  }
  MaterialExpressionNamedRerouteDeclaration_4 {
   class=MaterialExpressionNamedRerouteDeclaration
   inputs [None {from=MaterialExpressionLinearInterpolate_1}]
   properties {name=Base Color, nodeColor=<rgba 1 0 0.9609 1>}
  }
  MaterialExpressionNamedRerouteDeclaration_5 {
   class=MaterialExpressionNamedRerouteDeclaration
   inputs [None {from=MaterialExpressionPower_1}]
   properties {name=Erosion Mask, nodeColor=<rgba 0 1 0.7188 1>}
  }
  MaterialExpressionNamedRerouteDeclaration_6 {
   class=MaterialExpressionNamedRerouteDeclaration
   inputs [None {from=MaterialExpressionPower_9}]
   properties {name=MyOutline, nodeColor=<rgba 1 0.9844 0 1>}
  }
  MaterialExpressionNamedRerouteDeclaration_7 {
   class=MaterialExpressionNamedRerouteDeclaration
   inputs [None {from=MaterialExpressionOneMinus_8}]
   properties {name=MyBlendedShape, nodeColor=<rgba 0.6641 0 1 1>}
  }
  MaterialExpressionNamedRerouteDeclaration_8 {
   class=MaterialExpressionNamedRerouteDeclaration
   inputs [None {from=MaterialExpressionMaterialFunctionCall_11, output=G}]
   properties {name=Gradient, nodeColor=<rgba 0.2422 1 0 1>}
  }
  MaterialExpressionNamedRerouteDeclaration_9 {
   class=MaterialExpressionNamedRerouteDeclaration
   inputs [None {from=MaterialExpressionAdd_2}]
   properties {name=Emissive, nodeColor=<rgba 0 1 0.5547 1>}
  }
  MaterialExpressionNamedRerouteUsage_10 {class=MaterialExpressionNamedRerouteUsage}
  MaterialExpressionNamedRerouteUsage_13 {class=MaterialExpressionNamedRerouteUsage}
  MaterialExpressionNamedRerouteUsage_15 {class=MaterialExpressionNamedRerouteUsage}
  MaterialExpressionNamedRerouteUsage_26 {class=MaterialExpressionNamedRerouteUsage}
  MaterialExpressionNamedRerouteUsage_27 {class=MaterialExpressionNamedRerouteUsage}
  MaterialExpressionNamedRerouteUsage_28 {class=MaterialExpressionNamedRerouteUsage}
  MaterialExpressionNamedRerouteUsage_29 {class=MaterialExpressionNamedRerouteUsage}
  MaterialExpressionNamedRerouteUsage_3 {class=MaterialExpressionNamedRerouteUsage}
  MaterialExpressionNamedRerouteUsage_30 {class=MaterialExpressionNamedRerouteUsage}
  MaterialExpressionNamedRerouteUsage_31 {class=MaterialExpressionNamedRerouteUsage}
  MaterialExpressionNamedRerouteUsage_32 {class=MaterialExpressionNamedRerouteUsage}
  MaterialExpressionNamedRerouteUsage_33 {class=MaterialExpressionNamedRerouteUsage}
  MaterialExpressionNamedRerouteUsage_34 {class=MaterialExpressionNamedRerouteUsage}
  MaterialExpressionNamedRerouteUsage_35 {class=MaterialExpressionNamedRerouteUsage}
  MaterialExpressionNamedRerouteUsage_37 {class=MaterialExpressionNamedRerouteUsage}
  MaterialExpressionNamedRerouteUsage_38 {class=MaterialExpressionNamedRerouteUsage}
  MaterialExpressionNamedRerouteUsage_39 {class=MaterialExpressionNamedRerouteUsage}
  MaterialExpressionNamedRerouteUsage_4 {class=MaterialExpressionNamedRerouteUsage}
  MaterialExpressionNamedRerouteUsage_40 {class=MaterialExpressionNamedRerouteUsage}
  MaterialExpressionNamedRerouteUsage_41 {class=MaterialExpressionNamedRerouteUsage}
  MaterialExpressionNamedRerouteUsage_42 {class=MaterialExpressionNamedRerouteUsage}
  MaterialExpressionNamedRerouteUsage_44 {class=MaterialExpressionNamedRerouteUsage}
  MaterialExpressionNamedRerouteUsage_45 {class=MaterialExpressionNamedRerouteUsage}
  MaterialExpressionNamedRerouteUsage_47 {class=MaterialExpressionNamedRerouteUsage}
  MaterialExpressionNamedRerouteUsage_5 {class=MaterialExpressionNamedRerouteUsage}
  MaterialExpressionNamedRerouteUsage_50 {class=MaterialExpressionNamedRerouteUsage}
  MaterialExpressionNamedRerouteUsage_51 {class=MaterialExpressionNamedRerouteUsage}
  MaterialExpressionNamedRerouteUsage_53 {class=MaterialExpressionNamedRerouteUsage}
  MaterialExpressionNamedRerouteUsage_54 {class=MaterialExpressionNamedRerouteUsage}
  MaterialExpressionNamedRerouteUsage_56 {class=MaterialExpressionNamedRerouteUsage}
  MaterialExpressionNamedRerouteUsage_57 {class=MaterialExpressionNamedRerouteUsage}
  MaterialExpressionNamedRerouteUsage_59 {class=MaterialExpressionNamedRerouteUsage}
  MaterialExpressionNamedRerouteUsage_6 {class=MaterialExpressionNamedRerouteUsage}
  MaterialExpressionNamedRerouteUsage_7 {class=MaterialExpressionNamedRerouteUsage}
  MaterialExpressionNamedRerouteUsage_8 {class=MaterialExpressionNamedRerouteUsage}
  MaterialExpressionNamedRerouteUsage_9 {class=MaterialExpressionNamedRerouteUsage}
  MaterialExpressionOneMinus_0 {class=MaterialExpressionOneMinus, inputs [None {from=MaterialExpressionNamedRerouteUsage_29}]}
  MaterialExpressionOneMinus_11 {class=MaterialExpressionOneMinus, inputs [None {from=MaterialExpressionNamedRerouteUsage_37}]}
  MaterialExpressionOneMinus_4 {class=MaterialExpressionOneMinus, inputs [None {from=MaterialExpressionMultiply_23}]}
  MaterialExpressionOneMinus_5 {class=MaterialExpressionOneMinus, inputs [None {from=MaterialExpressionMultiply_28}]}
  MaterialExpressionOneMinus_6 {class=MaterialExpressionOneMinus, inputs [None {from=MaterialExpressionMultiply_37}]}
  MaterialExpressionOneMinus_8 {class=MaterialExpressionOneMinus, inputs [None {from=MaterialExpressionSmoothStep_6}]}
  MaterialExpressionOneMinus_9 {class=MaterialExpressionOneMinus, inputs [None {from=MaterialExpressionSmoothStep_8}]}
  MaterialExpressionPanner_0 {
   class=MaterialExpressionPanner
   inputs [
    Coordinate {from=MaterialExpressionMultiply_3}
    Time {from=MaterialExpressionNamedRerouteUsage_45}
    Speed {from=MaterialExpressionConvert_3}
   ]
   properties {bFractionalPart=true, speedY=0.187}
  }
  MaterialExpressionPanner_1 {
   class=MaterialExpressionPanner
   inputs [
    Coordinate {from=MaterialExpressionMultiply_11}
    Time {from=MaterialExpressionNamedRerouteUsage_47}
    Speed {from=MaterialExpressionConvert_1}
   ]
   properties {bFractionalPart=true, speedY=0.2}
  }
  MaterialExpressionPanner_3 {
   class=MaterialExpressionPanner
   inputs [
    Coordinate {from=MaterialExpressionMultiply_5}
    Time {from=""}
    Speed {from=""}
   ]
   properties {speedY=0.1843}
  }
  MaterialExpressionPanner_4 {
   class=MaterialExpressionPanner
   inputs [
    Coordinate {from=MaterialExpressionAdd_14}
    Time {from=""}
    Speed {from=""}
   ]
   properties {speedY=0.278}
  }
  MaterialExpressionPower_1 {
   class=MaterialExpressionPower
   inputs [
    Base {from=MaterialExpressionTextureSample_1, output=R}
    Exp {from=""}
   ]
   properties {constExponent=0.8}
  }
  MaterialExpressionPower_2 {
   class=MaterialExpressionPower
   inputs [
    Base {from=MaterialExpressionOneMinus_11}
    Exp {from=MaterialExpressionScalarParameter_19}
   ]
  }
  MaterialExpressionPower_3 {
   class=MaterialExpressionPower
   inputs [
    Base {from=MaterialExpressionNamedRerouteUsage_31}
    Exp {from=MaterialExpressionScalarParameter_12}
   ]
  }
  MaterialExpressionPower_9 {
   class=MaterialExpressionPower
   inputs [
    Base {from=MaterialExpressionSubtract_15}
    Exp {from=MaterialExpressionScalarParameter_32}
   ]
  }
  MaterialExpressionReroute_0 {class=MaterialExpressionReroute, inputs [None {from=MaterialExpressionOneMinus_6}]}
  MaterialExpressionReroute_1 {class=MaterialExpressionReroute, inputs [None {from=MaterialExpressionReroute_0}]}
  MaterialExpressionReroute_2 {class=MaterialExpressionReroute, inputs [None {from=MaterialExpressionBlendMaterialAttributes_0}]}
  MaterialExpressionSaturate_4 {class=MaterialExpressionSaturate, inputs [None {from=MaterialExpressionSmoothStep_4}]}
  MaterialExpressionSaturate_5 {class=MaterialExpressionSaturate, inputs [None {from=MaterialExpressionSmoothStep_5}]}
  MaterialExpressionSaturate_6 {class=MaterialExpressionSaturate, inputs [None {from=MaterialExpressionSubtract_11}]}
  MaterialExpressionSaturate_7 {class=MaterialExpressionSaturate, inputs [None {from=MaterialExpressionAdd_17}]}
  MaterialExpressionSaturate_8 {class=MaterialExpressionSaturate, inputs [None {from=MaterialExpressionDivide_1}]}
  MaterialExpressionScalarParameter_1 {
   class=MaterialExpressionScalarParameter
   properties {defaultValue=0.2, group=FX - Inner Fire Pattern, parameterName=Fire Pattern - Outline Emissive Intensity}
  }
  MaterialExpressionScalarParameter_10 {class=MaterialExpressionScalarParameter, properties {defaultValue=0.05, group=FX - Outer Fire Outline, parameterName=Fire Outline - Feather}}
  MaterialExpressionScalarParameter_11 {
   class=MaterialExpressionScalarParameter
   properties {defaultValue=0.6, group=FX - Outer Fire Outline, parameterName=Fire Outline - Shape Size}
  }
  MaterialExpressionScalarParameter_12 {class=MaterialExpressionScalarParameter, properties {defaultValue=0.75, group=FX - Outer Fire Outline}}
  MaterialExpressionScalarParameter_13 {
   class=MaterialExpressionScalarParameter
   properties {defaultValue=0.5, group=FX - Outer Fire Outline, parameterName=Fire Outline - Normal Intensity}
  }
  MaterialExpressionScalarParameter_14 {class=MaterialExpressionScalarParameter, properties {defaultValue=1.5, group=FX - Outer Fire Outline, parameterName=Param_1}}
  MaterialExpressionScalarParameter_16 {class=MaterialExpressionScalarParameter, properties {defaultValue=0.55, group=FX - Mystery, parameterName=ShapeThickness}}
  MaterialExpressionScalarParameter_17 {class=MaterialExpressionScalarParameter, properties {defaultValue=0.01, group=FX - Mystery, parameterName=ShapeThicknessModulation}}
  MaterialExpressionScalarParameter_18 {class=MaterialExpressionScalarParameter, properties {defaultValue=0.04, group=FX - Mystery, parameterName=OutlineThickness}}
  MaterialExpressionScalarParameter_19 {class=MaterialExpressionScalarParameter, properties {defaultValue=0.3, group=FX - Mystery, parameterName=Exp}}
  MaterialExpressionScalarParameter_2 {class=MaterialExpressionScalarParameter, properties {defaultValue=7, group=FX - Inner Fire Pattern, parameterName=Fire Pattern - Tile 1}}
  MaterialExpressionScalarParameter_20 {class=MaterialExpressionScalarParameter, properties {defaultValue=0.02, group=FX - Inner Fire Pattern, parameterName=Param_5}}
  MaterialExpressionScalarParameter_24 {
   class=MaterialExpressionScalarParameter
   properties {defaultValue=0.2657, group=FX - Inner Fire Pattern, parameterName=Fire Pattern - Speed1}
  }
  MaterialExpressionScalarParameter_25 {class=MaterialExpressionScalarParameter, properties {defaultValue=3, group=FX - Inner Fire Pattern, parameterName=Param_3}}
  MaterialExpressionScalarParameter_27 {class=MaterialExpressionScalarParameter, properties {defaultValue=0.384, group=FX - Outer Fire Outline, parameterName=Param_4}}
  MaterialExpressionScalarParameter_28 {
   class=MaterialExpressionScalarParameter
   properties {defaultValue=0.187, group=FX - Inner Fire Pattern, parameterName=Fire Pattern - Erosion Speed}
  }
  MaterialExpressionScalarParameter_29 {class=MaterialExpressionScalarParameter, properties {defaultValue=0.5, group=FX - Global, parameterName=Global Time Scale}}
  MaterialExpressionScalarParameter_30 {class=MaterialExpressionScalarParameter, properties {defaultValue=0.05, parameterName=Shape Thickness SmoothStep Range}}
  MaterialExpressionScalarParameter_31 {class=MaterialExpressionScalarParameter, properties {defaultValue=0.05, parameterName=Outline Thickness SmoothStep Range}}
  MaterialExpressionScalarParameter_32 {class=MaterialExpressionScalarParameter, properties {defaultValue=2, parameterName=Outline Thickness Exponent}}
  MaterialExpressionScalarParameter_33 {class=MaterialExpressionScalarParameter, properties {defaultValue=7, group=FX - Backpack Specific, parameterName=Procedural UVs Scale}}
  MaterialExpressionScalarParameter_35 {class=MaterialExpressionScalarParameter, properties {defaultValue=-44, group=FX - Backpack Specific, parameterName=UV Z Offset}}
  MaterialExpressionScalarParameter_36 {class=MaterialExpressionScalarParameter, properties {defaultValue=444, group=FX - Backpack Specific, parameterName=Radius}}
  MaterialExpressionScalarParameter_38 {class=MaterialExpressionScalarParameter, properties {defaultValue=0.25, group=FX - Backpack Specific, parameterName=Polar Offset}}
  MaterialExpressionScalarParameter_39 {class=MaterialExpressionScalarParameter, properties {defaultValue=1.111, group=FX - Backpack Specific, parameterName=Polar Tiling}}
  MaterialExpressionScalarParameter_41 {
   class=MaterialExpressionScalarParameter
   properties {group=FX - Backpack Specific, parameterName=Subsurface Additional Contribution to Emissive}
  }
  MaterialExpressionScalarParameter_5 {
   class=MaterialExpressionScalarParameter
   properties {defaultValue=6, group=FX - Inner Fire Pattern, parameterName=Fire Pattern - Erosion Tile}
  }
  MaterialExpressionScalarParameter_8 {class=MaterialExpressionScalarParameter, properties {defaultValue=5, group=FX - Outer Fire Outline, parameterName=Fire Outline - Tile}}
  MaterialExpressionScalarParameter_9 {
   class=MaterialExpressionScalarParameter
   properties {defaultValue=0.08, group=FX - Outer Fire Outline, parameterName=Fire Outline - Thickness}
  }
  MaterialExpressionSetMaterialAttributes_0 {
   class=MaterialExpressionSetMaterialAttributes
   inputs [
    MaterialAttributes {from=MaterialExpressionNamedRerouteUsage_13, output=""}
    BaseColor {from=MaterialExpressionNamedRerouteDeclaration_4, output=""}
    EmissiveColor {from=MaterialExpressionNamedRerouteDeclaration_9, output=""}
    Roughness {from=MaterialExpressionLinearInterpolate_7, output=""}
   ]
   properties {attributeSetTypes [#2, #5, #1]}
  }
  MaterialExpressionSetMaterialAttributes_1 {
   class=MaterialExpressionSetMaterialAttributes
   inputs [
    MaterialAttributes {from=MaterialExpressionGetMaterialAttributes_2, output=MaterialAttributes}
    BaseColor {from=MaterialExpressionLinearInterpolate_10, output=""}
    Roughness {from=MaterialExpressionAdd_10, output=""}
    Normal {from=MaterialExpressionMaterialFunctionCall_10, output=Result}
   ]
   properties {attributeSetTypes [#2, #1, #3]}
  }
  MaterialExpressionSetMaterialAttributes_2 {
   class=MaterialExpressionSetMaterialAttributes
   inputs [
    MaterialAttributes {from=MaterialExpressionGetMaterialAttributes_1, output=MaterialAttributes}
    Metallic {from=MaterialExpressionVectorParameter_4, output=R}
    Specular {from=MaterialExpressionVectorParameter_4, output=R}
    Roughness {from=MaterialExpressionVectorParameter_4, output=R}
   ]
   properties {attributeSetTypes [#6, #7, #1]}
  }
  MaterialExpressionSetMaterialAttributes_4 {
   class=MaterialExpressionSetMaterialAttributes
   inputs [
    MaterialAttributes {from=MaterialExpressionGetMaterialAttributes_3, output=MaterialAttributes}
    EmissiveColor {from=MaterialExpressionAdd_21, output=""}
   ]
   properties {attributeSetTypes [#5]}
  }
  MaterialExpressionSign_0 {class=MaterialExpressionSign, inputs [None {from=MaterialExpressionComponentMask_1, output=""}], properties {}}
  MaterialExpressionSine_2 {class=MaterialExpressionSine, inputs [None {from=MaterialExpressionAdd_9}]}
  MaterialExpressionSmoothStep_4 {
   class=MaterialExpressionSmoothStep
   inputs [
    Min {from=MaterialExpressionSubtract_10, output=""}
    Max {from=MaterialExpressionAdd_5, output=""}
    Value {from=MaterialExpressionNamedRerouteUsage_8, output=""}
   ]
   properties {constMax=0.55, constMin=0.45}
  }
  MaterialExpressionSmoothStep_5 {
   class=MaterialExpressionSmoothStep
   inputs [
    Min {from=MaterialExpressionSubtract_12, output=""}
    Max {from=MaterialExpressionAdd_7, output=""}
    Value {from=MaterialExpressionNamedRerouteUsage_9, output=""}
   ]
   properties {constMax=0.55, constMin=0.45}
  }
  MaterialExpressionSmoothStep_6 {
   class=MaterialExpressionSmoothStep
   inputs [
    Min {from=MaterialExpressionAdd_12, output=""}
    Max {from=MaterialExpressionAdd_11, output=""}
    Value {from=MaterialExpressionOneMinus_6, output=""}
   ]
   properties {}
  }
  MaterialExpressionSmoothStep_8 {
   class=MaterialExpressionSmoothStep
   inputs [
    Min {from=MaterialExpressionAdd_15, output=""}
    Max {from=MaterialExpressionAdd_13, output=""}
    Value {from=MaterialExpressionReroute_1, output=""}
   ]
   properties {}
  }
  MaterialExpressionStaticBoolParameter_0 {
   class=MaterialExpressionStaticBoolParameter
   inputs []
   properties {
    desc=Will skip the more complex elements and just apply the pattern on the purple parts
    group=FX - Global
    parameterName=Use Simple FX Version
   }
  }
  MaterialExpressionStaticBoolParameter_1 {
   class=MaterialExpressionStaticBoolParameter
   inputs []
   properties {desc=Needed for backpack, group=FX - Global, parameterName=Swizzle UV Components}
  }
  MaterialExpressionStaticBoolParameter_2 {
   class=MaterialExpressionStaticBoolParameter
   inputs []
   properties {
    desc="Will replace texture lookup with flat values, backpack has flame pattern baked into spec masks"
    group=FX - Global
    parameterName=Override Specular Masks
   }
  }
  MaterialExpressionStaticBoolParameter_3 {
   class=MaterialExpressionStaticBoolParameter
   inputs []
   properties {desc=Needed for backpack, group=FX - Backpack Specific, parameterName=Use Procedural Backpack UV}
  }
  MaterialExpressionStaticSwitchParameter_3 {
   class=MaterialExpressionStaticSwitchParameter
   inputs [
    True {from=MaterialExpressionMultiply_34, output=""}
    False {from=MaterialExpressionConstant_5, output=""}
   ]
   properties {group=FX - Backpack Specific, parameterName=Invert Left Side for Backpack}
  }
  MaterialExpressionStaticSwitch_0 {
   class=MaterialExpressionStaticSwitch
   inputs [
    True {from=MaterialExpressionReroute_2, output=""}
    False {from=MaterialExpressionSetMaterialAttributes_1, output=""}
    Value {from=MaterialExpressionNamedRerouteUsage_35, output=""}
   ]
   properties {}
  }
  MaterialExpressionStaticSwitch_1 {
   class=MaterialExpressionStaticSwitch
   inputs [
    True {from=MaterialExpressionConstant_4, output=""}
    False {from=MaterialExpressionMultiply_10, output=""}
    Value {from=MaterialExpressionNamedRerouteUsage_39, output=""}
   ]
   properties {}
  }
  MaterialExpressionStaticSwitch_2 {
   class=MaterialExpressionStaticSwitch
   inputs [
    True {from=MaterialExpressionAppendVector_0, output=""}
    False {from=MaterialExpressionTextureCoordinate_0, output=""}
    Value {from=MaterialExpressionStaticBoolParameter_1, output=""}
   ]
   properties {}
  }
  MaterialExpressionStaticSwitch_3 {
   class=MaterialExpressionStaticSwitch
   inputs [
    True {from=MaterialExpressionSetMaterialAttributes_2, output=""}
    False {from=MaterialExpressionGetMaterialAttributes_1, output=MaterialAttributes}
    Value {from=MaterialExpressionStaticBoolParameter_2, output=""}
   ]
   properties {}
  }
  MaterialExpressionStaticSwitch_4 {
   class=MaterialExpressionStaticSwitch
   inputs [
    True {from=MaterialExpressionNamedRerouteUsage_50, output=""}
    False {from=MaterialExpressionStaticSwitch_2, output=""}
    Value {from=MaterialExpressionStaticBoolParameter_3, output=""}
   ]
   properties {}
  }
  MaterialExpressionStaticSwitch_5 {
   class=MaterialExpressionStaticSwitch
   inputs [
    True {from=MaterialExpressionConstant_19, output=""}
    False {from=MaterialExpressionStaticSwitchParameter_3, output=""}
    Value {from=MaterialExpressionStaticBoolParameter_3, output=""}
   ]
   properties {}
  }
  MaterialExpressionSubtract_10 {
   class=MaterialExpressionSubtract
   inputs [
    A {from=MaterialExpressionSubtract_9}
    B {from=MaterialExpressionMultiply_27}
   ]
   properties {constB=0.05}
  }
  MaterialExpressionSubtract_11 {
   class=MaterialExpressionSubtract
   inputs [
    A {from=MaterialExpressionSaturate_4}
    B {from=MaterialExpressionSaturate_5}
   ]
  }
  MaterialExpressionSubtract_12 {
   class=MaterialExpressionSubtract
   inputs [
    A {from=MaterialExpressionOneMinus_4}
    B {from=MaterialExpressionMultiply_27}
   ]
   properties {constB=0.05}
  }
  MaterialExpressionSubtract_13 {
   class=MaterialExpressionSubtract
   inputs [
    A {from=MaterialExpressionMultiply_29}
    B {from=MaterialExpressionTextureSample_4, output=R}
   ]
  }
  MaterialExpressionSubtract_15 {
   class=MaterialExpressionSubtract
   inputs [
    A {from=MaterialExpressionOneMinus_9}
    B {from=MaterialExpressionOneMinus_8}
   ]
  }
  MaterialExpressionSubtract_8 {
   class=MaterialExpressionSubtract
   inputs [
    A {from=MaterialExpressionGetMaterialAttributes_0, output=Roughness}
    B {from=""}
   ]
   properties {constB=0.05}
  }
  MaterialExpressionSubtract_9 {
   class=MaterialExpressionSubtract
   inputs [
    A {from=MaterialExpressionOneMinus_4}
    B {from=MaterialExpressionMultiply_22}
   ]
  }
  MaterialExpressionTextureCoordinate_0 {class=MaterialExpressionTextureCoordinate}
  MaterialExpressionTextureCoordinate_3 {class=MaterialExpressionTextureCoordinate}
  MaterialExpressionTextureCoordinate_5 {class=MaterialExpressionTextureCoordinate, properties {uTiling=4, vTiling=2}}
  MaterialExpressionTextureObject_0 {
   class=MaterialExpressionTextureObject
   inputs []
   properties {
    samplerType=SAMPLERTYPE_LinearColor
    texture=/BRCosmetics/Characters/Player/Male/Medium/Bodies/M_MED_CloakedIron/Textures/T_CloakedIron_FirePattern_SDF_Normal
   }
  }
  MaterialExpressionTextureObject_1 {
   class=MaterialExpressionTextureObject
   inputs []
   properties {samplerType=SAMPLERTYPE_LinearGrayscale, texture=&T_SimplePerlin_Greyscale}
  }
  MaterialExpressionTextureObject_3 {
   class=MaterialExpressionTextureObject
   inputs []
   properties {samplerType=SAMPLERTYPE_LinearColor, texture=/Game/Effects/SnowSandInteraction/Textures/SSI_Perlin}
  }
  MaterialExpressionTextureSampleParameter2D_0 {
   class=MaterialExpressionTextureSampleParameter2D
   inputs [
    UVs {from="", output=""}
    Apply View MipBias {from="", output=""}
   ]
   properties {group=FX - Global, parameterName=Masks, samplerType=SAMPLERTYPE_Masks, texture=&T_WhiteMask}
  }
  MaterialExpressionTextureSample_1 {
   class=MaterialExpressionTextureSample
   inputs [
    UVs {from=MaterialExpressionMaterialFunctionCall_0, output=Wave UVs}
    Tex {from=MaterialExpressionTextureObject_3}
    Apply View MipBias {from=""}
   ]
  }
  MaterialExpressionTextureSample_2 {
   class=MaterialExpressionTextureSample
   inputs [
    UVs {from=MaterialExpressionMaterialFunctionCall_2, output=Wave UVs}
    Tex {from=MaterialExpressionNamedRerouteUsage_51}
    Apply View MipBias {from=""}
   ]
  }
  MaterialExpressionTextureSample_3 {
   class=MaterialExpressionTextureSample
   inputs [
    UVs {from=MaterialExpressionMaterialFunctionCall_5, output=Wave UVs}
    Tex {from=MaterialExpressionNamedRerouteUsage_53}
    Apply View MipBias {from=""}
   ]
  }
  MaterialExpressionTextureSample_4 {
   class=MaterialExpressionTextureSample
   inputs [
    UVs {from=MaterialExpressionPanner_4}
    Tex {from=""}
    Apply View MipBias {from=""}
   ]
   properties {samplerType=SAMPLERTYPE_LinearColor, texture=&SSI_Perlin}
  }
  MaterialExpressionTime_2 {class=MaterialExpressionTime}
  MaterialExpressionTime_3 {class=MaterialExpressionTime}
  MaterialExpressionVectorParameter_0 {
   class=MaterialExpressionVectorParameter
   properties {group=FX - Inner Fire Pattern, parameterName=Fira Pattern - Background Color, defaultValue=<rgba 0.117 0.0232 0.1981 1>}
  }
  MaterialExpressionVectorParameter_1 {
   class=MaterialExpressionVectorParameter
   properties {group=FX - Inner Fire Pattern, parameterName=Fire Pattern - Shape Color, defaultValue=<rgba 0.0908 0.0203 0.147 1>}
  }
  MaterialExpressionVectorParameter_2 {
   class=MaterialExpressionVectorParameter
   properties {group=FX - Inner Fire Pattern, parameterName=Fire Pattern - Outline Color, defaultValue=<rgba 0.2167 0.0954 0.333 1>}
  }
  MaterialExpressionVectorParameter_3 {
   class=MaterialExpressionVectorParameter
   properties {group=FX - Inner Fire Pattern, parameterName=Fire Pattern - Gradient Emissive Color, defaultValue=<rgba 0.05 0 0.0199 1>}
  }
  MaterialExpressionVectorParameter_4 {class=MaterialExpressionVectorParameter, properties {group=FX - Global, parameterName="Metallic (R) - Specular (G) - Roughness (B)"}}
  MaterialExpressionVertexInterpolator_1 {class=MaterialExpressionVertexInterpolator, inputs [VS {from=MaterialExpressionAppendVector_6, output=""}], properties {}}
 }}
