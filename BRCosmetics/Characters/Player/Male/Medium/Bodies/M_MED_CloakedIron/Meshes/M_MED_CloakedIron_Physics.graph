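# Physics asset for the male medium CloakedIron character body, providing capsule bodies for a dynamic jacket and the lower body.
# Seven four-segment jacket chains (front, side, rear, center) hang from pelvis-linked master bones, colliding shapes covering pelvis, lower spine, legs and feet.
# Joints within chains are locked, while front, side and rear chain roots have narrow swing limits.
# A huge root capsule sits below the character like a floor.
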PhysicsAsset /BRCosmetics/Characters/Player/Male/Medium/Bodies/M_MED_CloakedIron/Meshes/M_MED_CloakedIron_Physics
bodies {
 calf_l [{shapeType=Capsule, length=40, radius=7.5, shapeName=calf_r_capsule, center=<-18 0 0>, rotation=<rot 90 0 0>}]
 calf_r [{shapeType=Capsule, length=40, radius=7.5, shapeName=calf_l_capsule, center=<18 0 0>, rotation=<rot 90 180 0>}]
 dyn_jacket_center_01 [{shapeType=Capsule, length=15, radius=7, shapeName=dyn_jacket_center_01_capsule, center=<6 0 5>, rotation=<rot 0 0 -90>}]
 dyn_jacket_center_02 [{shapeType=Capsule, length=15, radius=8, shapeName=dyn_jacket_center_02_capsule, center=<6 0 6>, rotation=<rot 0 0 90>}]
 dyn_jacket_center_03 [{shapeType=Capsule, length=15, radius=10, shapeName=dyn_jacket_center_03_capsule, center=<8 0 7>, rotation=<rot 0 0 -90>}]
 dyn_jacket_center_04 [{shapeType=Capsule, length=15, radius=8, shapeName=dyn_jacket_center_04_capsule, center=<8 0 9>, rotation=<rot 0 0 -90>}]
 dyn_jacket_front_01_l [{shapeType=Capsule, length=15, radius=8, shapeName=dyn_jacket_front_01_l_capsule, center=<6 0 8>, rotation=<rot 0 0 -90>}]
 dyn_jacket_front_01_r [{shapeType=Capsule, length=15, radius=8, shapeName=dyn_jacket_front_01_r_capsule, center=<8 0 8>, rotation=<rot 0.0001 -0.0001 -81.4456>}]
 dyn_jacket_front_02_l [{shapeType=Capsule, length=15, radius=10, shapeName=dyn_jacket_front_02_l_capsule, center=<4 0 10>, rotation=<rot 0 0 -90>}]
 dyn_jacket_front_02_r [{shapeType=Capsule, length=15, radius=9, shapeName=dyn_jacket_front_02_r_capsule, center=<7 0 9>, rotation=<rot 0.0001 -0.0001 -81.4456>}]
 dyn_jacket_front_03_l [{shapeType=Capsule, length=15, radius=9.5, shapeName=dyn_jacket_front_03_l_capsule, center=<8 0 7>, rotation=<rot 0 0 -90>}]
 dyn_jacket_front_03_r [{shapeType=Capsule, length=15, radius=10, shapeName=dyn_jacket_front_03_r_capsule, center=<8 0 9>, rotation=<rot 0.0001 -0.0001 -81.4456>}]
 dyn_jacket_front_04_l [{shapeType=Capsule, length=15, radius=8, shapeName=dyn_jacket_front_04_l_capsule, center=<8 0 8>, rotation=<rot 0 0 -90>}]
 dyn_jacket_front_04_r [{shapeType=Capsule, length=15, radius=8, shapeName=dyn_jacket_front_04_r_capsule, center=<8 0 6>, rotation=<rot 0 0 -90>}]
 dyn_jacket_rear_01_l [{shapeType=Capsule, length=15, radius=7, shapeName=dyn_jacket_rear_01_l_capsule, center=<6 0 5>, rotation=<rot 0 0 -90>}]
 dyn_jacket_rear_01_r [{shapeType=Capsule, length=15, radius=7, shapeName=dyn_jacket_rear_01_r_capsule, center=<6 0 5>, rotation=<rot 0 0 -90>}]
 dyn_jacket_rear_02_l [{shapeType=Capsule, length=15, radius=8, shapeName=dyn_jacket_rear_02_l_capsule, center=<6 0 6>, rotation=<rot 0 0 90>}]
 dyn_jacket_rear_02_r [{shapeType=Capsule, length=15, radius=8, shapeName=dyn_jacket_rear_02_r_capsule, center=<6 0 6>, rotation=<rot 0 0 90>}]
 dyn_jacket_rear_03_l [{shapeType=Capsule, length=15, radius=10, shapeName=dyn_jacket_rear_03_l_capsule, center=<8 0 7>, rotation=<rot 0 0 -90>}]
 dyn_jacket_rear_03_r [{shapeType=Capsule, length=15, radius=10, shapeName=dyn_jacket_rear_03_r_capsule, center=<8 0 7>, rotation=<rot 0 0 -90>}]
 dyn_jacket_rear_04_l [{shapeType=Capsule, length=20, radius=8, shapeName=dyn_jacket_rear_04_l_capsule, center=<9 5 8>, rotation=<rot 0 0 -90>}]
 dyn_jacket_rear_04_r [{shapeType=Capsule, length=15, radius=8, shapeName=dyn_jacket_rear_04_r_capsule, center=<8 0 9>, rotation=<rot 0 0 -90>}]
 dyn_jacket_side_01_l [{shapeType=Capsule, length=15, radius=7, shapeName=dyn_jacket_side_01_l_capsule, center=<6 0 7.5>, rotation=<rot 0 0 -90>}]
 dyn_jacket_side_01_r [{shapeType=Capsule, length=15, radius=7, shapeName=dyn_jacket_side_01_r_capsule, center=<7 0 7>, rotation=<rot 0 0 -90>}]
 dyn_jacket_side_02_l [{shapeType=Capsule, length=15, radius=10, shapeName=dyn_jacket_side_02_l_capsule, center=<5 0 11>, rotation=<rot 0 0 90>}]
 dyn_jacket_side_02_r [{shapeType=Capsule, length=15, radius=10, shapeName=dyn_jacket_side_02_r_capsule, center=<6 0 11>, rotation=<rot 0 0 90>}]
 dyn_jacket_side_03_l [{shapeType=Capsule, length=15, radius=10, shapeName=dyn_jacket_side_03_l_capsule, center=<8 0 9>, rotation=<rot 0 0 -90>}]
 dyn_jacket_side_03_r [{shapeType=Capsule, length=15, radius=10, shapeName=dyn_jacket_side_03_r_capsule, center=<8 0 9>, rotation=<rot 0 -0.0001 -99.2455>}]
 dyn_jacket_side_04_l [{shapeType=Capsule, length=15, radius=8, shapeName=dyn_jacket_side_04_l_capsule, center=<8 0 8>, rotation=<rot 0 0 -90>}]
 dyn_jacket_side_04_r [{shapeType=Capsule, length=15, radius=8, shapeName=dyn_jacket_side_04_r_capsule, center=<8 0 8>, rotation=<rot 0 0 -90>}]
 foot_l [{shapeType=Capsule, length=18.4826, radius=8.0647, shapeName=foot_r_capsule, center=<-4 7.0799 0>, rotation=<rot 0 0 -90>}]
 foot_r [{shapeType=Capsule, length=18.4826, radius=8.0647, shapeName=foot_l_capsule, center=<4 -7.0799 0>, rotation=<rot 0 0 90>}]
 mstr_Jacket [{shapeType=Capsule, length=0.505, radius=0.505, shapeName=mstr_Jacket_capsule, rotation=<rot 0 -90 0>}]
 mstr_Jacket_Left [{shapeType=Capsule, length=4, radius=1, shapeName=mstr_Jacket_Left_capsule, center=<-0.4114 0 0>}]
 mstr_Jacket_Right [{shapeType=Capsule, length=4, radius=1, shapeName=mstr_Jacket_Right_capsule, center=<-0.4114 0 0>, rotation=<rot 0 0 90>}]
 mstr_jacket_front_01_l [{shapeType=Capsule, length=4, radius=1, shapeName=mstr_jacket_front_01_l_capsule, center=<-0.4114 0 0>, rotation=<rot 0 0 90>}]
 mstr_jacket_front_01_r [{shapeType=Capsule, length=4, radius=1, shapeName=mstr_jacket_front_01_r_capsule, center=<-0.4114 0 0>, rotation=<rot 0 0 90>}]
 mstr_jacket_rear_01_l [{shapeType=Capsule, length=0.505, radius=0.505, shapeName=mstr_jacket_rear_01_l_capsule, rotation=<rot 0 -90 0>}]
 mstr_jacket_rear_01_r [{shapeType=Capsule, length=0.505, radius=0.505, shapeName=mstr_jacket_rear_01_r_capsule, rotation=<rot 0 -90 0>}]
 mstr_jacket_side_01_l [{shapeType=Capsule, length=1, radius=1, shapeName=mstr_jacket_side_01_l_capsule, rotation=<rot 0 0 -90>}]
 mstr_jacket_side_01_r [{shapeType=Capsule, length=4, radius=1, shapeName=mstr_jacket_side_01_r_capsule, center=<-0.4114 0 0>, rotation=<rot 0 0 90>}]
 pelvis [{shapeType=Capsule, length=7, radius=10.4, shapeName=spine_01_capsule, center=<-0.4269 -0.4086 -0.1905>}]
 root [{shapeType=Capsule, length=200, radius=500, shapeName=root_capsule, center=<0 0 -500>, rotation=<rot 0 0 90>}]
 spine_01 [{shapeType=Capsule, length=4.1, radius=10.1157, shapeName=spine_01_capsule, center=<-0.4269 0.4787 0>}]
 thigh_l [{shapeType=Capsule, length=28.9433, radius=7.5, shapeName=thigh_l_capsule, center=<-15 0 0>, rotation=<rot 90 0 0>}]
 thigh_r [{shapeType=Capsule, length=37.9992, radius=9, shapeName=thigh_r_capsule, center=<17.1813 0 0>, rotation=<rot 90 180 0>}]
}
constraints [
 {
  bone1Name=thigh_l
  bone2Name=pelvis
  swing1LimitDegrees=45
  swing1Motion=Free
  swing2LimitDegrees=45
  swing2Motion=Free
  twistLimitDegrees=45
  twistMotion=Free
 }
 {
  bone1Name=calf_l
  bone2Name=thigh_l
  swing1LimitDegrees=45
  swing1Motion=Free
  swing2LimitDegrees=45
  swing2Motion=Free
  twistLimitDegrees=45
  twistMotion=Free
 }
 {
  bone1Name=thigh_r
  bone2Name=pelvis
  swing1LimitDegrees=45
  swing1Motion=Free
  swing2LimitDegrees=45
  swing2Motion=Free
  twistLimitDegrees=45
  twistMotion=Free
 }
 {
  bone1Name=calf_r
  bone2Name=thigh_r
  swing1LimitDegrees=45
  swing1Motion=Free
  swing2LimitDegrees=45
  swing2Motion=Free
  twistLimitDegrees=45
  twistMotion=Free
 }
 {
  bone1Name=foot_r
  bone2Name=calf_r
  swing1LimitDegrees=45
  swing1Motion=Free
  swing2LimitDegrees=45
  swing2Motion=Free
  twistLimitDegrees=45
  twistMotion=Free
 }
 {
  bone1Name=foot_l
  bone2Name=calf_l
  swing1LimitDegrees=45
  swing1Motion=Free
  swing2LimitDegrees=45
  swing2Motion=Free
  twistLimitDegrees=45
  twistMotion=Free
 }
 {
  bone1Name=mstr_Jacket_Right
  bone2Name=pelvis
  swing1LimitDegrees=45
  swing1Motion=Locked
  swing2LimitDegrees=45
  swing2Motion=Locked
  twistLimitDegrees=45
  twistMotion=Locked
 }
 {
  bone1Name=mstr_jacket_front_01_r
  bone2Name=mstr_Jacket_Right
  swing1LimitDegrees=45
  swing1Motion=Locked
  swing2LimitDegrees=45
  swing2Motion=Locked
  twistLimitDegrees=45
  twistMotion=Locked
 }
 {
  bone1Name=mstr_jacket_rear_01_r
  bone2Name=mstr_Jacket_Right
  swing1LimitDegrees=10
  swing1Motion=Locked
  swing2LimitDegrees=2
  swing2Motion=Locked
  twistLimitDegrees=2
  twistMotion=Locked
 }
 {
  bone1Name=mstr_jacket_side_01_r
  bone2Name=mstr_Jacket_Right
  swing1LimitDegrees=45
  swing1Motion=Locked
  swing2LimitDegrees=45
  swing2Motion=Locked
  twistLimitDegrees=45
  twistMotion=Locked
 }
 {bone1Name=dyn_jacket_front_01_r, bone2Name=mstr_jacket_front_01_r, swing1LimitDegrees=2, swing2LimitDegrees=10, twistLimitDegrees=0}
 {
  bone1Name=dyn_jacket_front_02_r
  bone2Name=dyn_jacket_front_01_r
  swing1LimitDegrees=45
  swing1Motion=Locked
  swing2LimitDegrees=45
  swing2Motion=Locked
  twistLimitDegrees=0
  twistMotion=Locked
 }
 {
  bone1Name=dyn_jacket_front_03_r
  bone2Name=dyn_jacket_front_02_r
  swing1LimitDegrees=45
  swing1Motion=Locked
  swing2LimitDegrees=45
  swing2Motion=Locked
  twistLimitDegrees=0
  twistMotion=Locked
 }
 {bone1Name=dyn_jacket_rear_01_r, bone2Name=mstr_jacket_rear_01_r, swing1LimitDegrees=2, swing2LimitDegrees=10, twistLimitDegrees=0}
 {
  bone1Name=dyn_jacket_rear_02_r
  bone2Name=dyn_jacket_rear_01_r
  swing1LimitDegrees=45
  swing1Motion=Locked
  swing2LimitDegrees=45
  swing2Motion=Locked
  twistLimitDegrees=0
  twistMotion=Locked
 }
 {
  bone1Name=dyn_jacket_rear_03_r
  bone2Name=dyn_jacket_rear_02_r
  swing1LimitDegrees=45
  swing1Motion=Locked
  swing2LimitDegrees=45
  swing2Motion=Locked
  twistLimitDegrees=0
  twistMotion=Locked
 }
 {bone1Name=dyn_jacket_side_01_r, bone2Name=mstr_jacket_side_01_r, swing1LimitDegrees=2, swing2LimitDegrees=10, twistLimitDegrees=0}
 {
  bone1Name=dyn_jacket_side_02_r
  bone2Name=dyn_jacket_side_01_r
  swing1LimitDegrees=45
  swing1Motion=Locked
  swing2LimitDegrees=45
  swing2Motion=Locked
  twistLimitDegrees=0
  twistMotion=Locked
 }
 {
  bone1Name=dyn_jacket_side_03_r
  bone2Name=dyn_jacket_side_02_r
  swing1LimitDegrees=45
  swing1Motion=Locked
  swing2LimitDegrees=45
  swing2Motion=Locked
  twistLimitDegrees=0
  twistMotion=Locked
 }
 {
  bone1Name=dyn_jacket_center_01
  bone2Name=pelvis
  swing1LimitDegrees=0
  swing1Motion=Locked
  swing2LimitDegrees=10
  swing2Motion=Locked
  twistLimitDegrees=45
  twistMotion=Locked
 }
 {
  bone1Name=dyn_jacket_center_02
  bone2Name=dyn_jacket_center_01
  swing1LimitDegrees=45
  swing1Motion=Locked
  swing2LimitDegrees=45
  swing2Motion=Locked
  twistLimitDegrees=45
  twistMotion=Locked
 }
 {
  bone1Name=dyn_jacket_center_03
  bone2Name=dyn_jacket_center_02
  swing1LimitDegrees=45
  swing1Motion=Locked
  swing2LimitDegrees=45
  swing2Motion=Locked
  twistLimitDegrees=45
  twistMotion=Locked
 }
 {
  bone1Name=dyn_jacket_front_04_r
  bone2Name=dyn_jacket_front_03_r
  swing1LimitDegrees=45
  swing1Motion=Locked
  swing2LimitDegrees=45
  swing2Motion=Locked
  twistLimitDegrees=45
  twistMotion=Locked
 }
 {
  bone1Name=dyn_jacket_side_04_r
  bone2Name=dyn_jacket_side_03_r
  swing1LimitDegrees=45
  swing1Motion=Locked
  swing2LimitDegrees=45
  swing2Motion=Locked
  twistLimitDegrees=45
  twistMotion=Locked
 }
 {
  bone1Name=dyn_jacket_center_04
  bone2Name=dyn_jacket_center_03
  swing1LimitDegrees=45
  swing1Motion=Locked
  swing2LimitDegrees=45
  swing2Motion=Locked
  twistLimitDegrees=45
  twistMotion=Locked
 }
 {
  bone1Name=pelvis
  bone2Name=root
  swing1LimitDegrees=45
  swing1Motion=Free
  swing2LimitDegrees=45
  swing2Motion=Free
  twistLimitDegrees=45
  twistMotion=Free
 }
 {
  bone1Name=mstr_Jacket_Left
  bone2Name=pelvis
  swing1LimitDegrees=45
  swing1Motion=Free
  swing2LimitDegrees=45
  swing2Motion=Free
  twistLimitDegrees=45
  twistMotion=Free
 }
 {
  bone1Name=mstr_jacket_front_01_l
  bone2Name=mstr_Jacket_Left
  swing1LimitDegrees=45
  swing1Motion=Locked
  swing2LimitDegrees=45
  swing2Motion=Locked
  twistLimitDegrees=45
  twistMotion=Locked
 }
 {bone1Name=dyn_jacket_front_01_l, bone2Name=mstr_jacket_front_01_l, swing1LimitDegrees=2, swing2LimitDegrees=10, twistLimitDegrees=0}
 {
  bone1Name=dyn_jacket_front_03_l
  bone2Name=dyn_jacket_front_02_l
  swing1LimitDegrees=45
  swing1Motion=Locked
  swing2LimitDegrees=45
  swing2Motion=Locked
  twistLimitDegrees=0
  twistMotion=Locked
 }
 {
  bone1Name=dyn_jacket_front_04_l
  bone2Name=dyn_jacket_front_03_l
  swing1LimitDegrees=45
  swing1Motion=Locked
  swing2LimitDegrees=45
  swing2Motion=Locked
  twistLimitDegrees=45
  twistMotion=Locked
 }
 {
  bone1Name=mstr_jacket_rear_01_l
  bone2Name=mstr_Jacket_Left
  swing1LimitDegrees=45
  swing1Motion=Locked
  swing2LimitDegrees=45
  swing2Motion=Locked
  twistLimitDegrees=45
  twistMotion=Locked
 }
 {bone1Name=dyn_jacket_rear_01_l, bone2Name=mstr_jacket_rear_01_l, swing1LimitDegrees=2, swing2LimitDegrees=10, twistLimitDegrees=0}
 {
  bone1Name=dyn_jacket_rear_02_l
  bone2Name=dyn_jacket_rear_01_l
  swing1LimitDegrees=45
  swing1Motion=Locked
  swing2LimitDegrees=45
  swing2Motion=Locked
  twistLimitDegrees=0
  twistMotion=Locked
 }
 {
  bone1Name=dyn_jacket_rear_03_l
  bone2Name=dyn_jacket_rear_02_l
  swing1LimitDegrees=45
  swing1Motion=Locked
  swing2LimitDegrees=45
  swing2Motion=Locked
  twistLimitDegrees=0
  twistMotion=Locked
 }
 {
  bone1Name=mstr_jacket_side_01_l
  bone2Name=mstr_Jacket_Left
  swing1LimitDegrees=45
  swing1Motion=Locked
  swing2LimitDegrees=45
  swing2Motion=Locked
  twistLimitDegrees=45
  twistMotion=Locked
 }
 {bone1Name=dyn_jacket_side_01_l, bone2Name=mstr_jacket_side_01_l, swing1LimitDegrees=2, swing2LimitDegrees=10, twistLimitDegrees=0}
 {
  bone1Name=dyn_jacket_side_02_l
  bone2Name=dyn_jacket_side_01_l
  swing1LimitDegrees=45
  swing1Motion=Locked
  swing2LimitDegrees=45
  swing2Motion=Locked
  twistLimitDegrees=0
  twistMotion=Locked
 }
 {
  bone1Name=dyn_jacket_side_03_l
  bone2Name=dyn_jacket_side_02_l
  swing1LimitDegrees=45
  swing1Motion=Locked
  swing2LimitDegrees=45
  swing2Motion=Locked
  twistLimitDegrees=0
  twistMotion=Locked
 }
 {
  bone1Name=mstr_Jacket
  bone2Name=pelvis
  swing1LimitDegrees=45
  swing1Motion=Free
  swing2LimitDegrees=45
  swing2Motion=Free
  twistLimitDegrees=45
  twistMotion=Free
 }
 {
  bone1Name=mstr_Jacket_Left
  bone2Name=mstr_Jacket
  swing1LimitDegrees=45
  swing1Motion=Free
  swing2LimitDegrees=45
  swing2Motion=Free
  twistLimitDegrees=45
  twistMotion=Free
 }
 {
  bone1Name=mstr_Jacket_Right
  bone2Name=mstr_Jacket
  swing1LimitDegrees=45
  swing1Motion=Locked
  swing2LimitDegrees=45
  swing2Motion=Locked
  twistLimitDegrees=45
  twistMotion=Locked
 }
 {
  bone1Name=dyn_jacket_center_01
  bone2Name=mstr_Jacket
  swing1LimitDegrees=0
  swing1Motion=Locked
  swing2LimitDegrees=10
  swing2Motion=Locked
  twistLimitDegrees=45
  twistMotion=Locked
 }
 {
  bone1Name=dyn_jacket_front_02_l
  bone2Name=dyn_jacket_front_01_l
  swing1LimitDegrees=45
  swing1Motion=Locked
  swing2LimitDegrees=45
  swing2Motion=Locked
  twistLimitDegrees=0
  twistMotion=Locked
 }
 {
  bone1Name=dyn_jacket_rear_04_r
  bone2Name=dyn_jacket_rear_03_r
  swing1LimitDegrees=45
  swing1Motion=Locked
  swing2LimitDegrees=45
  swing2Motion=Locked
  twistLimitDegrees=45
  twistMotion=Locked
 }
 {bone1Name=thigh_l, bone2Name=dyn_jacket_front_01_l, swing1LimitDegrees=5, swing2LimitDegrees=20, twistLimitDegrees=45, twistMotion=Locked}
 {
  bone1Name=dyn_jacket_rear_04_l
  bone2Name=dyn_jacket_rear_03_l
  swing1LimitDegrees=45
  swing1Motion=Locked
  swing2LimitDegrees=45
  swing2Motion=Locked
  twistLimitDegrees=45
  twistMotion=Locked
 }
 {
  bone1Name=dyn_jacket_side_04_l
  bone2Name=dyn_jacket_side_03_l
  swing1LimitDegrees=45
  swing1Motion=Locked
  swing2LimitDegrees=45
  swing2Motion=Locked
  twistLimitDegrees=45
  twistMotion=Locked
 }
 {
  bone1Name=dyn_jacket_rear_04_l
  bone2Name=dyn_jacket_side_04_l
  swing1LimitDegrees=45
  swing1Motion=Locked
  swing2LimitDegrees=45
  swing2Motion=Locked
  twistLimitDegrees=45
  twistMotion=Locked
 }
]
