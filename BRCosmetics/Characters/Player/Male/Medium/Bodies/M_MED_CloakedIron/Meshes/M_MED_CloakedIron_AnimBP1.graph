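# Animation Blueprint for the CloakedIron player body that copies the parent mesh pose and adds procedural jacket motion.
# Two pelvis-space RigidBody simulations blend on crouching, then feed a Control Rig.
# Additive bone offsets follow, weighted by thigh lift, crouch, movement direction and twisting in place.
# The output chain blends cached dynamic and in-air poses, applies further jacket offsets and constrains the tail between pelvis and rear jacket bones.
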AnimBlueprint /BRCosmetics/Characters/Player/Male/Medium/Bodies/M_MED_CloakedIron/Meshes/M_MED_CloakedIron_AnimBP1
target_skeleton=&/M_MED_CloakedIron_Skeleton
graphs [
 AnimGraph {nodes [
   Local To Component(LocalPose=Use cached pose 'InputPose'())
   Local To Component(LocalPose=Use cached pose 'InputPose'())
   n1 = Get bIsSkyDiving()
   n2 = "Transform (Modify) Bone - Bone: dyn_jacket_rear_01_l"!(ComponentPose, bAlphaBoolEnabled:bool=@n1, Translation="0,0,0", Rotation="0,0,0", Scale="1,1,1") -> (Pose) {
    alphaInputType=Bool
    rotationMode=BMM_Additive
    rotationSpace=BCS_BoneSpace
    translationMode=BMM_Additive
    translationSpace=BCS_BoneSpace
    alphaBoolBlend {blendOption=Linear}
    alphaScaleBiasClamp {bInterpResult=true}
    boneToModify {boneName=dyn_jacket_rear_01_l}
   }
   n3 = "Transform (Modify) Bone - Bone: dyn_jacket_side_01_l"!(ComponentPose=@n2, bAlphaBoolEnabled:bool=@n1, Translation="0,0,0", Rotation="10,0,0", Scale="1,1,1") -> (Pose) {
    alphaInputType=Bool
    rotationMode=BMM_Additive
    rotationSpace=BCS_BoneSpace
    translationMode=BMM_Additive
    translationSpace=BCS_BoneSpace
    alphaBoolBlend {blendOption=Linear}
    alphaScaleBiasClamp {bInterpResult=true}
    boneToModify {boneName=dyn_jacket_side_01_l}
    rotation {pitch=10}
   }
   n4 = "Transform (Modify) Bone - Bone: dyn_jacket_rear_01_r"!(ComponentPose=@n3, bAlphaBoolEnabled:bool=@n1, Translation="0,0,0", Rotation="10,0,0", Scale="1,1,1") -> (Pose) {
    alphaInputType=Bool
    rotationMode=BMM_Additive
    rotationSpace=BCS_BoneSpace
    translationMode=BMM_Additive
    translationSpace=BCS_BoneSpace
    alphaBoolBlend {blendOption=Linear}
    alphaScaleBiasClamp {bInterpResult=true}
    boneToModify {boneName=dyn_jacket_rear_01_r}
    rotation {pitch=10}
   }
   "Transform (Modify) Bone - Bone: dyn_jacket_side_01_r"!(ComponentPose=@n4, bAlphaBoolEnabled:bool=@n1, Translation="0,0,0", Rotation="5,0,0", Scale="1,1,1") -> (Pose) {
    alphaInputType=Bool
    rotationMode=BMM_Additive
    rotationSpace=BCS_BoneSpace
    translationMode=BMM_Additive
    translationSpace=BCS_BoneSpace
    alphaBoolBlend {blendOption=Linear}
    alphaScaleBiasClamp {bInterpResult=true}
    boneToModify {boneName=dyn_jacket_side_01_r}
    rotation {pitch=5}
   }
   n5 = "Transform (Modify) Bone - Bone: dyn_jacket_rear_01_l"(bAlphaBoolEnabled=True, Translation="0,0,0", Rotation="0,0,0", Scale="1,1,1") {
    alphaInputType=Bool
    rotationMode=BMM_Additive
    rotationSpace=BCS_BoneSpace
    translationMode=BMM_Additive
    translationSpace=BCS_BoneSpace
    alphaBoolBlend {blendOption=Linear}
    alphaScaleBiasClamp {bInterpResult=true}
    boneToModify {boneName=dyn_jacket_rear_01_l}
   }
   n6 = "Transform (Modify) Bone - Bone: dyn_jacket_rear_02_l"!(ComponentPose=@n5, bAlphaBoolEnabled:bool=True, Translation="0,0,0", Rotation="10,0,0", Scale="1,1,1") -> (Pose) {
    alphaInputType=Bool
    rotationMode=BMM_Additive
    rotationSpace=BCS_BoneSpace
    translationMode=BMM_Additive
    translationSpace=BCS_BoneSpace
    alphaBoolBlend {blendOption=Linear}
    alphaScaleBiasClamp {bInterpResult=true}
    boneToModify {boneName=dyn_jacket_rear_02_l}
    rotation {pitch=10}
   }
   n7 = "Transform (Modify) Bone - Bone: dyn_jacket_rear_03_l"!(ComponentPose=@n6, bAlphaBoolEnabled:bool=True, Translation="0,0,0", Rotation="-10,0,0", Scale="1,1,1") -> (Pose) {
    alphaInputType=Bool
    rotationMode=BMM_Additive
    rotationSpace=BCS_BoneSpace
    translationMode=BMM_Additive
    translationSpace=BCS_BoneSpace
    alphaBoolBlend {blendOption=Linear}
    alphaScaleBiasClamp {bInterpResult=true}
    boneToModify {boneName=dyn_jacket_rear_03_l}
    rotation {pitch=-10}
   }
   n8 = "Transform (Modify) Bone - Bone: dyn_jacket_rear_04_l"!(ComponentPose=@n7, bAlphaBoolEnabled:bool=True, Translation="0,0,0", Rotation="0,0,0", Scale="1,1,1") -> (Pose) {
    alphaInputType=Bool
    rotationMode=BMM_Additive
    rotationSpace=BCS_BoneSpace
    translationMode=BMM_Additive
    translationSpace=BCS_BoneSpace
    alphaBoolBlend {blendOption=Linear}
    alphaScaleBiasClamp {bInterpResult=true}
    boneToModify {boneName=dyn_jacket_rear_04_l}
   }
   n9 = "Transform (Modify) Bone - Bone: dyn_jacket_side_01_l"(ComponentPose=@n8, bAlphaBoolEnabled=True, Translation="0,0,0", Rotation="1,0,0", Scale="1,1,1") {
    alphaInputType=Bool
    rotationMode=BMM_Additive
    rotationSpace=BCS_BoneSpace
    translationMode=BMM_Additive
    translationSpace=BCS_BoneSpace
    alphaBoolBlend {blendOption=Linear}
    alphaScaleBiasClamp {bInterpResult=true}
    boneToModify {boneName=dyn_jacket_side_01_l}
    rotation {pitch=1}
   }
   n10 = "Transform (Modify) Bone - Bone: dyn_jacket_side_02_l"!(ComponentPose=@n9, bAlphaBoolEnabled:bool=True, Translation="0,0,0", Rotation="0,0,0", Scale="1,1,1") -> (Pose) {
    alphaInputType=Bool
    rotationMode=BMM_Additive
    rotationSpace=BCS_BoneSpace
    translationMode=BMM_Additive
    translationSpace=BCS_BoneSpace
    alphaBoolBlend {blendOption=Linear}
    alphaScaleBiasClamp {bInterpResult=true}
    boneToModify {boneName=dyn_jacket_side_02_l}
   }
   n11 = "Transform (Modify) Bone - Bone: dyn_jacket_side_03_l"!(ComponentPose=@n10, bAlphaBoolEnabled:bool=True, Translation="0,0,0", Rotation="-3,0,0", Scale="1,1,1") -> (Pose) {
    alphaInputType=Bool
    rotationMode=BMM_Additive
    rotationSpace=BCS_BoneSpace
    translationMode=BMM_Additive
    translationSpace=BCS_BoneSpace
    alphaBoolBlend {blendOption=Linear}
    alphaScaleBiasClamp {bInterpResult=true}
    boneToModify {boneName=dyn_jacket_side_03_l}
    rotation {pitch=-3}
   }
   n12 = "Transform (Modify) Bone - Bone: dyn_jacket_side_04_l"!(ComponentPose=@n11, bAlphaBoolEnabled:bool=True, Translation="0,0,0", Rotation="-12,0,0", Scale="1,1,1") -> (Pose) {
    alphaInputType=Bool
    rotationMode=BMM_Additive
    rotationSpace=BCS_BoneSpace
    translationMode=BMM_Additive
    translationSpace=BCS_BoneSpace
    alphaBoolBlend {blendOption=Linear}
    alphaScaleBiasClamp {bInterpResult=true}
    boneToModify {boneName=dyn_jacket_side_04_l}
    rotation {pitch=-12}
   }
   n13 = "Transform (Modify) Bone - Bone: dyn_jacket_rear_01_r"(ComponentPose=@n12, bAlphaBoolEnabled=True, Translation="0,0,0", Rotation="0,0,0", Scale="1,1,1") {
    alphaInputType=Bool
    rotationMode=BMM_Additive
    rotationSpace=BCS_BoneSpace
    translationMode=BMM_Additive
    translationSpace=BCS_BoneSpace
    alphaBoolBlend {blendOption=Linear}
    alphaScaleBiasClamp {bInterpResult=true}
    boneToModify {boneName=dyn_jacket_rear_01_r}
   }
   n14 = "Transform (Modify) Bone - Bone: dyn_jacket_rear_02_r"!(ComponentPose=@n13, bAlphaBoolEnabled:bool=True, Translation="0,0,0", Rotation="-2,0,0", Scale="1,1,1") -> (Pose) {
    alphaInputType=Bool
    rotationMode=BMM_Additive
    rotationSpace=BCS_BoneSpace
    translationMode=BMM_Additive
    translationSpace=BCS_BoneSpace
    alphaBoolBlend {blendOption=Linear}
    alphaScaleBiasClamp {bInterpResult=true}
    boneToModify {boneName=dyn_jacket_rear_02_r}
    rotation {pitch=-2}
   }
   n15 = "Transform (Modify) Bone - Bone: dyn_jacket_rear_03_r"!(ComponentPose=@n14, bAlphaBoolEnabled:bool=True, Translation="0,0,2", Rotation="-10,0,0", Scale="1,1,1") -> (Pose) {
    alphaInputType=Bool
    rotationMode=BMM_Additive
    rotationSpace=BCS_BoneSpace
    translationMode=BMM_Additive
    translationSpace=BCS_BoneSpace
    alphaBoolBlend {blendOption=Linear}
    alphaScaleBiasClamp {bInterpResult=true}
    boneToModify {boneName=dyn_jacket_rear_03_r}
    rotation {pitch=-10}
    translation {z=2}
   }
   n16 = "Transform (Modify) Bone - Bone: dyn_jacket_side_01_r"(ComponentPose=@n15, bAlphaBoolEnabled=True, Translation="0,0,0", Rotation="-8,0,0", Scale="1,1,1") {
    alphaInputType=Bool
    rotationMode=BMM_Additive
    rotationSpace=BCS_BoneSpace
    translationMode=BMM_Additive
    translationSpace=BCS_BoneSpace
    alphaBoolBlend {blendOption=Linear}
    alphaScaleBiasClamp {bInterpResult=true}
    boneToModify {boneName=dyn_jacket_side_01_r}
    rotation {pitch=-8}
   }
   n17 = "Transform (Modify) Bone - Bone: dyn_jacket_side_02_r"!(ComponentPose=@n16, bAlphaBoolEnabled:bool=True, Translation="0,0,0", Rotation="0,0,0", Scale="1,1,1") -> (Pose) {
    alphaInputType=Bool
    rotationMode=BMM_Additive
    rotationSpace=BCS_BoneSpace
    translationMode=BMM_Additive
    translationSpace=BCS_BoneSpace
    alphaBoolBlend {blendOption=Linear}
    alphaScaleBiasClamp {bInterpResult=true}
    boneToModify {boneName=dyn_jacket_side_02_r}
   }
   n18 = "Transform (Modify) Bone - Bone: dyn_jacket_side_03_r"!(ComponentPose=@n17, bAlphaBoolEnabled:bool=True, Translation="0,0,0", Rotation="0,0,0", Scale="1,1,1") -> (Pose) {
    alphaInputType=Bool
    rotationMode=BMM_Additive
    rotationSpace=BCS_BoneSpace
    translationMode=BMM_Additive
    translationSpace=BCS_BoneSpace
    alphaBoolBlend {blendOption=Linear}
    alphaScaleBiasClamp {bInterpResult=true}
    boneToModify {boneName=dyn_jacket_side_03_r}
   }
   "Transform (Modify) Bone - Bone: mstr_jacket_front_01_l"!(ComponentPose=@n18, bAlphaBoolEnabled:bool=Get bIsPlayingEmote(), Translation="0,0,0", Rotation="0,0,0", Scale="1,1,1") -> (Pose) {
    alphaInputType=Bool
    rotationMode=BMM_Additive
    rotationSpace=BCS_BoneSpace
    translationMode=BMM_Additive
    translationSpace=BCS_BoneSpace
    alphaBoolBlend {blendOption=Linear}
    alphaScaleBiasClamp {bInterpResult=true}
    boneToModify {boneName=mstr_jacket_front_01_l}
   }
   n19 = Use cached pose 'DynPose'()
   n20 = Use cached pose 'InAirPose'()
   n21 = Get is In Air() -> (is In Air:bool)
   n22 = "Transform (Modify) Bone - Bone: mstr_Jacket_Right"(
    ComponentPose=Local To Component(LocalPose=Blend Poses by bool(BlendPose_0=@n20, BlendPose_1=@n19, BlendTime_0=0.5, BlendTime_1=0.5, bActiveValue=@n21) {blendType=Linear})
    Alpha=1
    Translation="0,0,0"
    Rotation="-3.791,48.184,33.569"
    Scale="1,1,1"
   ) {
    rotationMode=BMM_Additive
    translationMode=BMM_Additive
    alphaBoolBlend {blendOption=Linear}
    boneToModify {boneName=mstr_Jacket_Right}
    rotation=<rot -3.791 48.184 33.569>
   }
   n23 = "Transform (Modify) Bone - Bone: dyn_jacket_side_01_r"!(ComponentPose=@n22, Alpha:float=1, Translation="0,0,0", Rotation="0,0,0", Scale="1,1,1") -> (Pose) {rotationMode=BMM_Additive, translationMode=BMM_Additive, alphaBoolBlend {blendOption=Linear}, boneToModify {boneName=dyn_jacket_side_01_r}}
   n24 = "Transform (Modify) Bone - Bone: dyn_jacket_rear_01_r"!(ComponentPose=@n23, Alpha:float=1, Translation="0,0,0", Rotation="0,0,0", Scale="1,1,1") -> (Pose) {rotationMode=BMM_Additive, translationMode=BMM_Additive, alphaBoolBlend {blendOption=Linear}, boneToModify {boneName=dyn_jacket_rear_01_r}}
   n25 = "Transform (Modify) Bone - Bone: mstr_Jacket_Left"(ComponentPose=@n24, Alpha=1, Translation="0,0,0", Rotation="43.204,17.206,-11.274", Scale="1,1,1") {
    rotationMode=BMM_Additive
    translationMode=BMM_Additive
    alphaBoolBlend {blendOption=Linear}
    boneToModify {boneName=mstr_Jacket_Left}
    rotation=<rot 43.204 17.206 -11.274>
   }
   n26 = "Transform (Modify) Bone - Bone: dyn_jacket_side_01_l"!(ComponentPose=@n25, Alpha:float=1, Translation="0,0,0", Rotation="-16.186,-6.435,-0.054", Scale="1,1,1") -> (Pose) {
    rotationMode=BMM_Additive
    translationMode=BMM_Additive
    alphaBoolBlend {blendOption=Linear}
    boneToModify {boneName=dyn_jacket_side_01_l}
    rotation=<rot -16.186 -6.435 -0.054>
   }
   n27 = "Transform (Modify) Bone - Bone: dyn_jacket_front_01_l"(ComponentPose=@n26, Alpha=1, Translation="0,0,0", Rotation="-2.762,-23.114,-17.392", Scale="1,1,1") {
    rotationMode=BMM_Additive
    translationMode=BMM_Additive
    alphaBoolBlend {blendOption=Linear}
    boneToModify {boneName=dyn_jacket_front_01_l}
    rotation=<rot -2.762 -23.114 -17.392>
   }
   n28 = "Transform (Modify) Bone - Bone: mstr_jacket_rear_01_l"(ComponentPose=@n27, Alpha=1, Translation="0,0,0", Rotation="34.167,0.000,0.000", Scale="1,1,1") {
    rotationMode=BMM_Additive
    translationMode=BMM_Additive
    alphaBoolBlend {blendOption=Linear}
    alphaScaleBiasClamp {bInterpResult=true, interpSpeedIncreasing=4}
    boneToModify {boneName=mstr_jacket_rear_01_l}
    rotation {pitch=34.167}
   }
   n29 = "Transform (Modify) Bone - Bone: dyn_jacket_side_01_l"(ComponentPose=@n28, Alpha=1, Translation="0,0,0", Rotation="20,0,0", Scale="1,1,1") {
    rotationMode=BMM_Additive
    translationMode=BMM_Additive
    alphaBoolBlend {blendOption=Linear}
    alphaScaleBiasClamp {bInterpResult=true, interpSpeedIncreasing=4}
    boneToModify {boneName=dyn_jacket_side_01_l}
    rotation {pitch=20}
   }
   n30 = "Transform (Modify) Bone - Bone: mstr_jacket_front_01_l"!(ComponentPose=@n29, Alpha:float=1, Translation="0,0,0", Rotation="-10.430,-19.454,-7.048", Scale="1,1,1") -> (Pose) {
    rotationMode=BMM_Additive
    translationMode=BMM_Additive
    alphaBoolBlend {blendOption=Linear}
    boneToModify {boneName=mstr_jacket_front_01_l}
    rotation=<rot -10.43 -19.454 -7.048>
   }
   n31 = "Transform (Modify) Bone - Bone: mstr_jacket_rear_01_r"(ComponentPose=@n30, Alpha=1, Translation="0,0,0", Rotation="-7.965,-5.939,-12.283", Scale="1,1,1") {
    rotationMode=BMM_Additive
    translationMode=BMM_Additive
    alphaBoolBlend {blendOption=Linear}
    alphaScaleBiasClamp {bInterpResult=true, interpSpeedIncreasing=4}
    boneToModify {boneName=mstr_jacket_rear_01_r}
    rotation=<rot -7.965 -5.939 -12.283>
   }
   n32 = "Transform (Modify) Bone - Bone: dyn_jacket_side_01_r"(ComponentPose=@n31, Alpha=1, Translation="0,0,0", Rotation="5.904,12.772,-12.335", Scale="1,1,1") {
    rotationMode=BMM_Additive
    translationMode=BMM_Additive
    alphaBoolBlend {blendOption=Linear}
    alphaScaleBiasClamp {bInterpResult=true, interpSpeedIncreasing=4}
    boneToModify {boneName=dyn_jacket_side_01_r}
    rotation=<rot 5.904 12.772 -12.335>
   }
   n33 = "Transform (Modify) Bone - Bone: dyn_jacket_front_01_r"(ComponentPose=@n32, Alpha=1, Translation="0,0,0", Rotation="-31.115,-12.960,6.603", Scale="1,1,1") {
    rotationMode=BMM_Additive
    translationMode=BMM_Additive
    alphaBoolBlend {blendOption=Linear}
    alphaScaleBiasClamp {bInterpResult=true, interpSpeedIncreasing=4}
    boneToModify {boneName=dyn_jacket_front_01_r}
    rotation=<rot -31.115 -12.96 6.603>
   }
   n34 = "Transform (Modify) Bone - Bone: mstr_jacket_rear_01_l"(ComponentPose=@n33, Alpha=1, Translation="0,0,0", Rotation="37.031,-58.991,0.000", Scale="1,1,1") {
    rotationMode=BMM_Additive
    translationMode=BMM_Additive
    alphaBoolBlend {blendOption=Linear}
    alphaScaleBiasClamp {bInterpResult=true, interpSpeedIncreasing=4}
    boneToModify {boneName=mstr_jacket_rear_01_l}
    rotation {pitch=37.031, yaw=-58.991}
   }
   n35 = "Transform (Modify) Bone - Bone: dyn_jacket_side_01_l"(ComponentPose=@n34, Alpha=1, Translation="0,0,0", Rotation="35.324,-15.496,-4.445", Scale="1,1,1") {
    rotationMode=BMM_Additive
    translationMode=BMM_Additive
    alphaBoolBlend {blendOption=Linear}
    alphaScaleBiasClamp {bInterpResult=true, interpSpeedIncreasing=4}
    boneToModify {boneName=dyn_jacket_side_01_l}
    rotation=<rot 35.324 -15.496 -4.445>
   }
   n36 = "Transform (Modify) Bone - Bone: mstr_jacket_front_01_l"(ComponentPose=@n35, Alpha=1, Translation="0,0,0", Rotation="46.317,-8.777,-33.133", Scale="1,1,1") {
    rotationMode=BMM_Additive
    translationMode=BMM_Additive
    alphaBoolBlend {blendOption=Linear}
    boneToModify {boneName=mstr_jacket_front_01_l}
    rotation=<rot 46.317 -8.777 -33.133>
   }
   n37 = "Transform (Modify) Bone - Bone: mstr_jacket_rear_01_r"(ComponentPose=@n36, Alpha=1, Translation="0,0,5", Rotation="55.499,34.813,130.678", Scale="1,1,1") {
    rotationMode=BMM_Additive
    translationMode=BMM_Additive
    alphaBoolBlend {blendOption=Linear}
    alphaScaleBiasClamp {bInterpResult=true, interpSpeedIncreasing=4}
    boneToModify {boneName=mstr_jacket_rear_01_r}
    rotation=<rot 55.499 34.813 130.678>
    translation {z=5}
   }
   n38 = "Transform (Modify) Bone - Bone: dyn_jacket_side_01_r"(ComponentPose=@n37, Alpha=1, Translation="0,0,0", Rotation="-30.915,-3.313,113.327", Scale="1,1,1") {
    rotationMode=BMM_Additive
    translationMode=BMM_Additive
    alphaBoolBlend {blendOption=Linear}
    alphaScaleBiasClamp {bInterpResult=true, interpSpeedIncreasing=4}
    boneToModify {boneName=dyn_jacket_side_01_r}
    rotation=<rot -30.915 -3.313 113.327>
   }
   n39 = "Transform (Modify) Bone - Bone: dyn_jacket_front_01_r"(ComponentPose=@n38, Alpha=1, Translation="0,0,5", Rotation="-36.613,81.331,-2.289", Scale="1,1,1") {
    rotationMode=BMM_Additive
    translationMode=BMM_Additive
    alphaBoolBlend {blendOption=Linear}
    alphaScaleBiasClamp {bInterpResult=true, interpSpeedIncreasing=4}
    boneToModify {boneName=dyn_jacket_front_01_r}
    rotation=<rot -36.613 81.331 -2.289>
    translation {z=5}
   }
   n40 = "Transform (Modify) Bone - Bone: mstr_jacket_rear_01_l"(ComponentPose=@n39, Alpha=1, Translation="0,0,0", Rotation="-11.442,3.911,64.431", Scale="1,1,1") {
    rotationMode=BMM_Additive
    translationMode=BMM_Additive
    alphaBoolBlend {blendOption=Linear}
    alphaScaleBiasClamp {bInterpResult=true, interpSpeedIncreasing=4}
    boneToModify {boneName=mstr_jacket_rear_01_l}
    rotation=<rot -11.442 3.911 64.431>
   }
   n41 = "Transform (Modify) Bone - Bone: dyn_jacket_side_01_l"(ComponentPose=@n40, Alpha=1, Translation="0,0,0", Rotation="36.337,1.880,41.419", Scale="1,1,1") {
    rotationMode=BMM_Additive
    translationMode=BMM_Additive
    alphaBoolBlend {blendOption=Linear}
    alphaScaleBiasClamp {bInterpResult=true, interpSpeedIncreasing=4}
    boneToModify {boneName=dyn_jacket_side_01_l}
    rotation=<rot 36.337 1.88 41.419>
   }
   n42 = "Transform (Modify) Bone - Bone: mstr_jacket_front_01_l"(ComponentPose=@n41, Alpha=1, Translation="0,0,0", Rotation="16.293,-25.932,-8.503", Scale="1,1,1") {
    rotationMode=BMM_Additive
    translationMode=BMM_Additive
    alphaBoolBlend {blendOption=Linear}
    boneToModify {boneName=mstr_jacket_front_01_l}
    rotation=<rot 16.293 -25.932 -8.503>
   }
   n43 = "Transform (Modify) Bone - Bone: mstr_jacket_rear_01_r"(ComponentPose=@n42, Alpha=1, Translation="0,0,0", Rotation="-56.461,22.723,-24.989", Scale="1,1,1") {
    rotationMode=BMM_Additive
    translationMode=BMM_Additive
    alphaBoolBlend {blendOption=Linear}
    alphaScaleBiasClamp {bInterpResult=true, interpSpeedIncreasing=4}
    boneToModify {boneName=mstr_jacket_rear_01_r}
    rotation=<rot -56.461 22.723 -24.989>
   }
   n44 = "Transform (Modify) Bone - Bone: dyn_jacket_side_01_r"(ComponentPose=@n43, Alpha=1, Translation="0,0,0", Rotation="-23.009,6.004,-17.761", Scale="1,1,1") {
    rotationMode=BMM_Additive
    translationMode=BMM_Additive
    alphaBoolBlend {blendOption=Linear}
    alphaScaleBiasClamp {bInterpResult=true, interpSpeedIncreasing=4}
    boneToModify {boneName=dyn_jacket_side_01_r}
    rotation=<rot -23.009 6.004 -17.761>
   }
   n45 = "Transform (Modify) Bone - Bone: dyn_jacket_front_01_r"(ComponentPose=@n44, Alpha=1, Translation="5,0,0", Rotation="-1.496,-5.207,28.415", Scale="1,1,1") {
    rotationMode=BMM_Additive
    translationMode=BMM_Additive
    alphaBoolBlend {blendOption=Linear}
    alphaScaleBiasClamp {bInterpResult=true, interpSpeedIncreasing=4}
    boneToModify {boneName=dyn_jacket_front_01_r}
    rotation=<rot -1.496 -5.207 28.415>
    translation {x=5}
   }
   n46 = "Transform (Modify) Bone - Bone: mstr_jacket_rear_01_l"(ComponentPose=@n45, Alpha=1, Translation="1.484,4.171,-0.417", Rotation="-11.651,28.384,22.964", Scale="1,1,1") {
    rotationMode=BMM_Additive
    translationMode=BMM_Additive
    alphaBoolBlend {blendOption=Linear}
    alphaScaleBiasClamp {bInterpResult=true, interpSpeedDecreasing=50}
    boneToModify {boneName=mstr_jacket_rear_01_l}
    rotation=<rot -11.651 28.384 22.964>
    translation=<1.484 4.171 -0.417>
   }
   n47 = "Transform (Modify) Bone - Bone: mstr_jacket_side_01_l"(ComponentPose=@n46, Alpha=1, Translation="0.784,0.000,0.000", Rotation="-20.356,44.673,-5.504", Scale="1,1,1") {
    rotationMode=BMM_Additive
    translationMode=BMM_Additive
    alphaBoolBlend {blendOption=Linear}
    alphaScaleBiasClamp {bInterpResult=true, interpSpeedDecreasing=50}
    boneToModify {boneName=mstr_jacket_side_01_l}
    rotation=<rot -20.356 44.673 -5.504>
    translation {x=0.784}
   }
   n48 = "Transform (Modify) Bone - Bone: mstr_jacket_rear_01_r"(ComponentPose=@n47, Alpha=1, Translation="0.000,3.933,0.000", Rotation="0.990,-10.370,17.836", Scale="1,1,1") {
    rotationMode=BMM_Additive
    translationMode=BMM_Additive
    alphaBoolBlend {blendOption=Linear}
    alphaScaleBiasClamp {bInterpResult=true, interpSpeedDecreasing=50}
    boneToModify {boneName=mstr_jacket_rear_01_r}
    rotation=<rot 0.99 -10.37 17.836>
    translation {y=3.933}
   }
   n49 = "Transform (Modify) Bone - Bone: dyn_jacket_side_01_r"(ComponentPose=@n48, Alpha=1, Translation="0,0,0", Rotation="0.866,-14.236,9.571", Scale="1,1,1") {
    rotationMode=BMM_Additive
    translationMode=BMM_Additive
    alphaBoolBlend {blendOption=Linear}
    alphaScaleBiasClamp {bInterpResult=true, interpSpeedDecreasing=50}
    boneToModify {boneName=dyn_jacket_side_01_r}
    rotation=<rot 0.866 -14.236 9.571>
   }
   n50 = "Transform (Modify) Bone - Bone: dyn_jacket_front_01_l"(ComponentPose=@n49, Alpha=1, Translation="0,0,0", Rotation="-22.016,23.086,4.859", Scale="1,1,1") {
    rotationMode=BMM_Additive
    translationMode=BMM_Additive
    alphaBoolBlend {blendOption=Linear}
    alphaScaleBiasClamp {bInterpResult=true, interpSpeedDecreasing=50}
    boneToModify {boneName=dyn_jacket_front_01_l}
    rotation=<rot -22.016 23.086 4.859>
   }
   n51 = "Constraint - mstr_jacket_tail (PPP)"!(ComponentPose=@n50, Alpha=1, ConstraintWeights_0=0.348, ConstraintWeights_1:float=0.343, ConstraintWeights_2:float=0.309) -> (Pose) {
    alphaBoolBlend {blendOption=Linear}
    boneToModify {boneName=mstr_jacket_tail}
    constraintSetup [{targetBone {boneName=pelvis}}, {targetBone {boneName=dyn_jacket_rear_01_r}}, {targetBone {boneName=dyn_jacket_rear_01_l}}]
    constraintWeights [1,1,1]
   }
   Output Pose(Result=Component To Local(ComponentPose=@n51)) {layerGroup=None}
   Save cached pose 'InputPose'(Pose=Copy Pose From Mesh() {bUseAttachedParent=true}) {cacheName=InputPose}
   Save cached pose 'InputPose'(Pose=Copy Pose From Mesh() {bUseAttachedParent=true}) {cacheName=InputPose}
   n52 = Get Left() -> (Left:double)
   n53 = Get Right() -> (Right:double)
   n54 = RigidBody(ComponentPose=Local To Component(LocalPose=Use cached pose 'InputPose'())) {
    bOverrideWorldGravity=true
    overridePhysicsAsset=&/M_MED_CloakedIron_Physics
    simulationSpace=BaseBoneSpace
    alphaBoolBlend {blendOption=Linear}
    baseBoneRef {boneName=pelvis}
    componentAppliedLinearAccClamp=<2000 2000 2000>
    simSpaceSettings {maxAngularAcceleration=200, maxAngularVelocity=1, maxLinearAcceleration=100, maxLinearVelocity=50, velocityScaleZ=0, worldAlpha=1}
   }
   n55 = RigidBody(ComponentPose=Local To Component(LocalPose=Use cached pose 'InputPose'())) {
    bClampLinearTranslationLimitToRefPose=true
    bOverrideWorldGravity=true
    overridePhysicsAsset=&/M_MED_CloakedIron_Physics1
    simulationSpace=BaseBoneSpace
    alphaBoolBlend {blendOption=Linear}
    alphaScaleBiasClamp {interpSpeedDecreasing=15, interpSpeedIncreasing=20}
    baseBoneRef {boneName=pelvis}
    simSpaceSettings {maxAngularAcceleration=500, maxAngularVelocity=4, maxLinearAcceleration=700, velocityScaleZ=0, worldAlpha=1}
   }
   n56 = Control Rig!(
    Source=Blend Poses by bool(BlendPose_0=Component To Local(ComponentPose=@n55), BlendPose_1=Component To Local(ComponentPose=@n54), bActiveValue=Get bIsCrouching()) {blendType=Linear}
    Alpha
    Crouch:bool=Get bIsCrouching()
    playing emote:bool=Get bIsPlayingEmote()
    Thigh Rot R:double=Get thigh_rot_up_r() -> (thigh_rot_up_r:double)
    Thigh Rot L:double=Get thigh_rot_up_l() -> (thigh_rot_up_l:double)
    crouch walking:bool=false
    crouching:bool=false
   ) -> (Pose) {
    alphaBoolBlend {blendOption=Linear}
    controlRigAssetReference {blueprintRigClass=&/M_MED_CloakedIron_CtrlRig1.M_MED_CloakedIron_CtrlRig1_C}
   }
   n57 = Get thigh_rot_up_l()
   n58 = "Transform (Modify) Bone - Bone: mstr_jacket_side_01_l"(
    ComponentPose=Local To Component(LocalPose=@n56)
    Alpha=@n57
    Translation="2.770,0.198,-1.256"
    Rotation="-2.268,-5.582,2.807"
    Scale="1,1,1"
   ) {
    rotationMode=BMM_Additive
    rotationSpace=BCS_BoneSpace
    translationMode=BMM_Additive
    translationSpace=BCS_BoneSpace
    alphaBoolBlend {blendOption=Linear}
    boneToModify {boneName=mstr_jacket_side_01_l}
    rotation=<rot -2.268 -5.582 2.807>
    translation=<2.77 0.198 -1.256>
   }
   n59 = "Transform (Modify) Bone - Bone: mstr_Jacket_Left"(ComponentPose=@n58, Alpha=@n57, Translation="-0.393,-0.921,-5", Rotation="0,0,31.355", Scale="1,1,1") {
    rotationMode=BMM_Additive
    rotationSpace=BCS_BoneSpace
    translationMode=BMM_Additive
    translationSpace=BCS_BoneSpace
    alphaBoolBlend {blendOption=Linear}
    boneToModify {boneName=mstr_Jacket_Left}
    rotation {roll=31.355}
    translation=<-0.393 -0.921 -5>
   }
   n60 = Get thigh_rot_up_r()
   n61 = "Transform (Modify) Bone - Bone: mstr_jacket_side_01_r"(ComponentPose=@n59, Alpha=@n60, Translation="0,0,0", Rotation="1.002,5.991,6.827", Scale="1,1,1") {
    rotationMode=BMM_Additive
    rotationSpace=BCS_BoneSpace
    translationMode=BMM_Additive
    translationSpace=BCS_BoneSpace
    alphaBoolBlend {blendOption=Linear}
    boneToModify {boneName=mstr_jacket_side_01_r}
    rotation=<rot 1.002 5.991 6.827>
   }
   n62 = "Transform (Modify) Bone - Bone: mstr_Jacket_Right"(ComponentPose=@n61, Alpha=@n60, Translation="0,0,6", Rotation="-9.125,1.622,-4", Scale="1,1,1") {
    rotationMode=BMM_Additive
    rotationSpace=BCS_BoneSpace
    translationMode=BMM_Additive
    translationSpace=BCS_BoneSpace
    alphaBoolBlend {blendOption=Linear}
    boneToModify {boneName=mstr_Jacket_Right}
    rotation=<rot -9.125 1.622 -4>
    translation {z=6}
   }
   n63 = Get IsCrouchingAlpha() -> (IsCrouchingAlpha:double)
   n64 = "Transform (Modify) Bone - Bone: mstr_jacket_front_01_l"(ComponentPose=@n62, Alpha=@n63, Translation="-2,-2,0", Rotation="0,10,0", Scale="1,1,1") {
    rotationMode=BMM_Additive
    rotationSpace=BCS_BoneSpace
    translationMode=BMM_Additive
    translationSpace=BCS_BoneSpace
    alphaBoolBlend {blendOption=Linear}
    alphaScaleBiasClamp {bInterpResult=true}
    boneToModify {boneName=mstr_jacket_front_01_l}
    rotation {yaw=10}
    translation=<-2 -2>
   }
   n65 = "Transform (Modify) Bone - Bone: mstr_Jacket"(ComponentPose=@n64, Alpha=@n63, Translation="0,0,0", Rotation="0,10,0", Scale="1,1,1") {
    rotationMode=BMM_Additive
    rotationSpace=BCS_BoneSpace
    translationMode=BMM_Additive
    translationSpace=BCS_BoneSpace
    alphaBoolBlend {blendOption=Linear}
    alphaScaleBiasClamp {bInterpResult=true}
    boneToModify {boneName=mstr_Jacket}
    rotation {yaw=10}
   }
   n66 = "Transform (Modify) Bone - Bone: mstr_jacket_rear_01_r"(ComponentPose=@n65, Alpha=@n63, Translation="0,0,0", Rotation="-5,0,0", Scale="1,1,1") {
    rotationMode=BMM_Additive
    rotationSpace=BCS_BoneSpace
    translationMode=BMM_Additive
    translationSpace=BCS_BoneSpace
    alphaBoolBlend {blendOption=Linear}
    alphaScaleBiasClamp {bInterpResult=true}
    boneToModify {boneName=mstr_jacket_rear_01_r}
    rotation {pitch=-5}
   }
   n67 = "Transform (Modify) Bone - Bone: mstr_Jacket"(ComponentPose=@n66, Alpha=@n63, Translation="0,1,0", Rotation="0,0,0", Scale="1,1,1") {
    rotationMode=BMM_Additive
    rotationSpace=BCS_BoneSpace
    translationMode=BMM_Additive
    translationSpace=BCS_BoneSpace
    alphaBoolBlend {blendOption=Linear}
    alphaScaleBiasClamp {bInterpResult=true}
    boneToModify {boneName=mstr_Jacket}
    translation {y=1}
   }
   n68 = Get Forward() -> (Forward:double)
   n69 = "Transform (Modify) Bone - Bone: mstr_jacket_rear_01_l"(ComponentPose=@n67, Alpha=@n68, Translation="0,0,0", Rotation="25,0,10", Scale="1,1,1") {
    rotationMode=BMM_Additive
    rotationSpace=BCS_BoneSpace
    translationMode=BMM_Additive
    translationSpace=BCS_BoneSpace
    alphaBoolBlend {blendOption=Linear}
    alphaScaleBiasClamp {bInterpResult=true, interpSpeedIncreasing=2}
    boneToModify {boneName=mstr_jacket_rear_01_l}
    rotation {pitch=25, roll=10}
   }
   n70 = "Transform (Modify) Bone - Bone: mstr_jacket_rear_01_r"(ComponentPose=@n69, Alpha=@n68, Translation="0,0,0", Rotation="25,0,0", Scale="1,1,1") {
    rotationMode=BMM_Additive
    rotationSpace=BCS_BoneSpace
    translationMode=BMM_Additive
    translationSpace=BCS_BoneSpace
    alphaBoolBlend {blendOption=Linear}
    alphaScaleBiasClamp {bInterpResult=true, interpSpeedIncreasing=2}
    boneToModify {boneName=mstr_jacket_rear_01_r}
    rotation {pitch=25}
   }
   n71 = "Transform (Modify) Bone - Bone: dyn_jacket_side_01_l"(ComponentPose=@n70, Alpha=@n68, Translation="0,0,0", Rotation="20,-16.056,-36.203", Scale="1,1,1") {
    rotationMode=BMM_Additive
    rotationSpace=BCS_BoneSpace
    translationMode=BMM_Additive
    translationSpace=BCS_BoneSpace
    alphaBoolBlend {blendOption=Linear}
    alphaScaleBiasClamp {bInterpResult=true, interpSpeedIncreasing=2}
    boneToModify {boneName=dyn_jacket_side_01_l}
    rotation=<rot 20 -16.056 -36.203>
   }
   n72 = "Transform (Modify) Bone - Bone: dyn_jacket_side_01_r"(ComponentPose=@n71, Alpha=@n68, Translation="0,0,0", Rotation="5,0,0", Scale="1,1,1") {
    rotationMode=BMM_Additive
    rotationSpace=BCS_BoneSpace
    translationMode=BMM_Additive
    translationSpace=BCS_BoneSpace
    alphaBoolBlend {blendOption=Linear}
    alphaScaleBiasClamp {bInterpResult=true, interpSpeedIncreasing=2}
    boneToModify {boneName=dyn_jacket_side_01_r}
    rotation {pitch=5}
   }
   n73 = "Transform (Modify) Bone - Bone: dyn_jacket_front_01_l"(ComponentPose=@n72, Alpha=@n68, Translation="0,0,0", Rotation="5,0,0", Scale="1,1,1") {
    rotationMode=BMM_Additive
    rotationSpace=BCS_BoneSpace
    translationMode=BMM_Additive
    translationSpace=BCS_BoneSpace
    alphaBoolBlend {blendOption=Linear}
    alphaScaleBiasClamp {bInterpResult=true, interpSpeedIncreasing=2}
    boneToModify {boneName=dyn_jacket_front_01_l}
    rotation {pitch=5}
   }
   n74 = "Transform (Modify) Bone - Bone: dyn_jacket_side_02_l"!(ComponentPose=@n73, Alpha:float=@n68, Translation="0,0,0", Rotation="0,0,0", Scale="1,1,1") -> (Pose) {
    rotationMode=BMM_Additive
    rotationSpace=BCS_BoneSpace
    translationMode=BMM_Additive
    translationSpace=BCS_BoneSpace
    alphaBoolBlend {blendOption=Linear}
    alphaScaleBiasClamp {bInterpResult=true, interpSpeedIncreasing=2}
    boneToModify {boneName=dyn_jacket_side_02_l}
   }
   n75 = Get Backward() -> (Backward:double)
   n76 = "Transform (Modify) Bone - Bone: dyn_jacket_rear_01_l"!(ComponentPose=@n74, Alpha:float=@n75, Translation="0,0,0", Rotation="10,0,0", Scale="1,1,1") -> (Pose) {
    rotationMode=BMM_Additive
    rotationSpace=BCS_BoneSpace
    translationMode=BMM_Additive
    translationSpace=BCS_BoneSpace
    alphaBoolBlend {blendOption=Linear}
    alphaScaleBiasClamp {bInterpResult=true, interpSpeedIncreasing=2}
    boneToModify {boneName=dyn_jacket_rear_01_l}
    rotation {pitch=10}
   }
   n77 = "Transform (Modify) Bone - Bone: dyn_jacket_rear_01_r"(ComponentPose=@n76, Alpha=@n75, Translation="0,0,0", Rotation="10,0,0", Scale="1,1,1") {
    rotationMode=BMM_Additive
    rotationSpace=BCS_BoneSpace
    translationMode=BMM_Additive
    translationSpace=BCS_BoneSpace
    alphaBoolBlend {blendOption=Linear}
    alphaScaleBiasClamp {bInterpResult=true, interpSpeedIncreasing=2}
    boneToModify {boneName=dyn_jacket_rear_01_r}
    rotation {pitch=10}
   }
   n78 = "Transform (Modify) Bone - Bone: dyn_jacket_side_01_l"(ComponentPose=@n77, Alpha=@n75, Translation="0,0,0", Rotation="5,0,0", Scale="1,1,1") {
    rotationMode=BMM_Additive
    rotationSpace=BCS_BoneSpace
    translationMode=BMM_Additive
    translationSpace=BCS_BoneSpace
    alphaBoolBlend {blendOption=Linear}
    alphaScaleBiasClamp {bInterpResult=true, interpSpeedIncreasing=2}
    boneToModify {boneName=dyn_jacket_side_01_l}
    rotation {pitch=5}
   }
   n79 = "Transform (Modify) Bone - Bone: dyn_jacket_side_01_r"(ComponentPose=@n78, Alpha=@n75, Translation="0,0,0", Rotation="0,0,0", Scale="1,1,1") {
    rotationMode=BMM_Additive
    rotationSpace=BCS_BoneSpace
    translationMode=BMM_Additive
    translationSpace=BCS_BoneSpace
    alphaBoolBlend {blendOption=Linear}
    alphaScaleBiasClamp {bInterpResult=true, interpSpeedIncreasing=2}
    boneToModify {boneName=dyn_jacket_side_01_r}
   }
   n80 = "Transform (Modify) Bone - Bone: mstr_jacket_front_01_l"(ComponentPose=@n79, Alpha=@n75, Translation="0,0,0", Rotation="5,0,0", Scale="1,1,1") {
    rotationMode=BMM_Additive
    rotationSpace=BCS_BoneSpace
    translationMode=BMM_Additive
    translationSpace=BCS_BoneSpace
    alphaBoolBlend {blendOption=Linear}
    alphaScaleBiasClamp {bInterpResult=true, interpSpeedIncreasing=2}
    boneToModify {boneName=mstr_jacket_front_01_l}
    rotation {pitch=5}
   }
   n81 = "Transform (Modify) Bone - Bone: dyn_jacket_front_01_r"(ComponentPose=@n80, Alpha=@n75, Translation="0,0,0", Rotation="5,0,0", Scale="1,1,1") {
    rotationMode=BMM_Additive
    rotationSpace=BCS_BoneSpace
    translationMode=BMM_Additive
    translationSpace=BCS_BoneSpace
    alphaBoolBlend {blendOption=Linear}
    alphaScaleBiasClamp {bInterpResult=true, interpSpeedIncreasing=2}
    boneToModify {boneName=dyn_jacket_front_01_r}
    rotation {pitch=5}
   }
   n82 = "Transform (Modify) Bone - Bone: dyn_jacket_front_01_l"(ComponentPose=@n81, Alpha=@n53, Translation="0,0,2", Rotation="10,5,0", Scale="1,1,1") {
    rotationMode=BMM_Additive
    rotationSpace=BCS_BoneSpace
    translationMode=BMM_Additive
    translationSpace=BCS_BoneSpace
    alphaBoolBlend {blendOption=Linear}
    alphaScaleBiasClamp {bInterpResult=true, interpSpeedIncreasing=2}
    boneToModify {boneName=dyn_jacket_front_01_l}
    rotation {pitch=10, yaw=5}
    translation {z=2}
   }
   n83 = "Transform (Modify) Bone - Bone: dyn_jacket_side_01_l"(ComponentPose=@n82, Alpha=@n53, Translation="0,0,0", Rotation="10,0,0", Scale="1,1,1") {
    rotationMode=BMM_Additive
    rotationSpace=BCS_BoneSpace
    translationMode=BMM_Additive
    translationSpace=BCS_BoneSpace
    alphaBoolBlend {blendOption=Linear}
    alphaScaleBiasClamp {bInterpResult=true, interpSpeedIncreasing=2}
    boneToModify {boneName=dyn_jacket_side_01_l}
    rotation {pitch=10}
   }
   n84 = "Transform (Modify) Bone - Bone: dyn_jacket_front_01_r"(ComponentPose=@n83, Alpha=@n53, Translation="0,0,0", Rotation="10,0,0", Scale="1,1,1") {
    rotationMode=BMM_Additive
    rotationSpace=BCS_BoneSpace
    translationMode=BMM_Additive
    translationSpace=BCS_BoneSpace
    alphaBoolBlend {blendOption=Linear}
    alphaScaleBiasClamp {bInterpResult=true, interpSpeedIncreasing=2}
    boneToModify {boneName=dyn_jacket_front_01_r}
    rotation {pitch=10}
   }
   n85 = "Transform (Modify) Bone - Bone: dyn_jacket_side_01_r"(ComponentPose=@n84, Alpha=@n53, Translation="0,0,0", Rotation="10,0,0", Scale="1,1,1") {
    rotationMode=BMM_Additive
    rotationSpace=BCS_BoneSpace
    translationMode=BMM_Additive
    translationSpace=BCS_BoneSpace
    alphaBoolBlend {blendOption=Linear}
    alphaScaleBiasClamp {bInterpResult=true, interpSpeedIncreasing=2}
    boneToModify {boneName=dyn_jacket_side_01_r}
    rotation {pitch=10}
   }
   n86 = "Transform (Modify) Bone - Bone: dyn_jacket_rear_01_l"(ComponentPose=@n85, Alpha=@n53, Translation="0,0,0", Rotation="10,0,0", Scale="1,1,1") {
    rotationMode=BMM_Additive
    rotationSpace=BCS_BoneSpace
    translationMode=BMM_Additive
    translationSpace=BCS_BoneSpace
    alphaBoolBlend {blendOption=Linear}
    alphaScaleBiasClamp {bInterpResult=true, interpSpeedIncreasing=2}
    boneToModify {boneName=dyn_jacket_rear_01_l}
    rotation {pitch=10}
   }
   n87 = "Transform (Modify) Bone - Bone: dyn_jacket_front_01_r"(ComponentPose=@n86, Alpha=@n52, Translation="0,0,0", Rotation="5,0,0", Scale="1,1,1") {
    rotationMode=BMM_Additive
    rotationSpace=BCS_BoneSpace
    translationMode=BMM_Additive
    translationSpace=BCS_BoneSpace
    alphaBoolBlend {blendOption=Linear}
    alphaScaleBiasClamp {bInterpResult=true, interpSpeedIncreasing=2}
    boneToModify {boneName=dyn_jacket_front_01_r}
    rotation {pitch=5}
   }
   n88 = "Transform (Modify) Bone - Bone: dyn_jacket_side_01_r"(ComponentPose=@n87, Alpha=@n52, Translation="0,0,0", Rotation="5,0,0", Scale="1,1,1") {
    rotationMode=BMM_Additive
    rotationSpace=BCS_BoneSpace
    translationMode=BMM_Additive
    translationSpace=BCS_BoneSpace
    alphaBoolBlend {blendOption=Linear}
    alphaScaleBiasClamp {bInterpResult=true, interpSpeedIncreasing=2}
    boneToModify {boneName=dyn_jacket_side_01_r}
    rotation {pitch=5}
   }
   n89 = "Transform (Modify) Bone - Bone: dyn_jacket_front_01_l"(ComponentPose=@n88, Alpha=@n52, Translation="0,0,0", Rotation="10,0,0", Scale="1,1,1") {
    rotationMode=BMM_Additive
    rotationSpace=BCS_BoneSpace
    translationMode=BMM_Additive
    translationSpace=BCS_BoneSpace
    alphaBoolBlend {blendOption=Linear}
    alphaScaleBiasClamp {bInterpResult=true, interpSpeedIncreasing=2}
    boneToModify {boneName=dyn_jacket_front_01_l}
    rotation {pitch=10}
   }
   n90 = "Transform (Modify) Bone - Bone: dyn_jacket_side_01_l"(ComponentPose=@n89, Alpha=@n52, Translation="0,0,0", Rotation="10,0,0", Scale="1,1,1") {
    rotationMode=BMM_Additive
    rotationSpace=BCS_BoneSpace
    translationMode=BMM_Additive
    translationSpace=BCS_BoneSpace
    alphaBoolBlend {blendOption=Linear}
    alphaScaleBiasClamp {bInterpResult=true, interpSpeedIncreasing=2}
    boneToModify {boneName=dyn_jacket_side_01_l}
    rotation {pitch=10}
   }
   n91 = "Transform (Modify) Bone - Bone: mstr_jacket_rear_01_l"(ComponentPose=@n90, Alpha=@n52, Translation="0,0,0", Rotation="15,0,0", Scale="1,1,1") {
    rotationMode=BMM_Additive
    rotationSpace=BCS_BoneSpace
    translationMode=BMM_Additive
    translationSpace=BCS_BoneSpace
    alphaBoolBlend {blendOption=Linear}
    alphaScaleBiasClamp {bInterpResult=true, interpSpeedIncreasing=2}
    boneToModify {boneName=mstr_jacket_rear_01_l}
    rotation {pitch=15}
   }
   n92 = Get Is Twisting in Place() -> (Is Twisting in Place:double)
   n93 = "Transform (Modify) Bone - Bone: dyn_jacket_front_01_r"(ComponentPose=@n91, Alpha=@n92, Translation="0,0,0", Rotation="5,-10,0", Scale="1,1,1") {
    rotationMode=BMM_Additive
    rotationSpace=BCS_BoneSpace
    alphaBoolBlend {blendOption=Linear}
    alphaScaleBiasClamp {bInterpResult=true, interpSpeedIncreasing=5}
    boneToModify {boneName=dyn_jacket_front_01_r}
    rotation {pitch=5, yaw=-10}
   }
   n94 = "Transform (Modify) Bone - Bone: dyn_jacket_side_01_r"(ComponentPose=@n93, Alpha=@n92, Translation="0,0,0", Rotation="5,5,0", Scale="1,1,1") {
    rotationMode=BMM_Additive
    rotationSpace=BCS_BoneSpace
    alphaBoolBlend {blendOption=Linear}
    alphaScaleBiasClamp {bInterpResult=true, interpSpeedIncreasing=5}
    boneToModify {boneName=dyn_jacket_side_01_r}
    rotation {pitch=5, yaw=5}
   }
   n95 = "Transform (Modify) Bone - Bone: dyn_jacket_front_01_l"(ComponentPose=@n94, Alpha=@n92, Translation="0,1000,0", Rotation="0,0,0", Scale="1,1,1") {
    rotationMode=BMM_Additive
    rotationSpace=BCS_BoneSpace
    alphaBoolBlend {blendOption=Linear}
    alphaScaleBiasClamp {bInterpResult=true, interpSpeedIncreasing=5}
    boneToModify {boneName=dyn_jacket_front_01_l}
    translation {y=1000}
   }
   n96 = Get is Crouch Walking() -> (is Crouch Walking:double)
   n97 = "Transform (Modify) Bone - Bone: dyn_jacket_front_01_l"(ComponentPose=@n95, Alpha=@n96, Translation="0,0,0", Rotation="-5,0,0", Scale="1,1,1") {
    rotationMode=BMM_Additive
    rotationSpace=BCS_BoneSpace
    translationMode=BMM_Additive
    translationSpace=BCS_BoneSpace
    alphaBoolBlend {blendOption=Linear}
    alphaScaleBiasClamp {bInterpResult=true, interpSpeedDecreasing=6}
    boneToModify {boneName=dyn_jacket_front_01_l}
    rotation {pitch=-5}
   }
   n98 = "Transform (Modify) Bone - Bone: dyn_jacket_side_01_l"(ComponentPose=@n97, Alpha=@n96, Translation="0,0,0", Rotation="-5,0,0", Scale="1,1,1") {
    rotationMode=BMM_Additive
    rotationSpace=BCS_BoneSpace
    translationMode=BMM_Additive
    translationSpace=BCS_BoneSpace
    alphaBoolBlend {blendOption=Linear}
    alphaScaleBiasClamp {bInterpResult=true, interpSpeedDecreasing=6}
    boneToModify {boneName=dyn_jacket_side_01_l}
    rotation {pitch=-5}
   }
   n99 = "Transform (Modify) Bone - Bone: dyn_jacket_front_01_r"(ComponentPose=@n98, Alpha=@n96, Translation="0,0,0", Rotation="0,0,0", Scale="1,1,1") {
    rotationMode=BMM_Additive
    rotationSpace=BCS_BoneSpace
    translationMode=BMM_Additive
    translationSpace=BCS_BoneSpace
    alphaBoolBlend {blendOption=Linear}
    alphaScaleBiasClamp {bInterpResult=true, interpSpeedIncreasing=5}
    boneToModify {boneName=dyn_jacket_front_01_r}
   }
   n100 = "Transform (Modify) Bone - Bone: dyn_jacket_side_01_r"(ComponentPose=@n99, Alpha=@n96, Translation="0,0,0", Rotation="0,0,0", Scale="1,1,1") {
    rotationMode=BMM_Additive
    rotationSpace=BCS_BoneSpace
    translationMode=BMM_Additive
    translationSpace=BCS_BoneSpace
    alphaBoolBlend {blendOption=Linear}
    alphaScaleBiasClamp {bInterpResult=true, interpSpeedIncreasing=5}
    boneToModify {boneName=dyn_jacket_side_01_r}
   }
   n101 = RigidBody!(ComponentPose=@n100, bAlphaBoolEnabled:bool, ExternalForce) -> (Pose) {
    alphaInputType=Bool
    bOverrideWorldGravity=true
    overridePhysicsAsset=&/M_MED_CloakedIron_Emotes_PhysicsAsset
    simulationSpace=BaseBoneSpace
    alphaBoolBlend {blendOption=Linear}
    baseBoneRef {boneName=pelvis}
    overrideWorldGravity {z=-980}
    simSpaceSettings {maxAngularAcceleration=20, maxAngularVelocity=5, maxLinearAcceleration=3000, maxLinearVelocity=600, velocityScaleZ=0, worldAlpha=1}
   }
   n102 = RigidBody(ComponentPose=@n101) {
    alphaInputType=Bool
    bOverrideWorldGravity=true
    overridePhysicsAsset=&M_MED_CloakedIron_Emotes_PhysicsAsset
    simulationSpace=BaseBoneSpace
    alphaBoolBlend {blendOption=Linear}
    baseBoneRef {boneName=pelvis}
    overrideWorldGravity {z=-980}
    simSpaceSettings {maxAngularAcceleration=10, maxAngularVelocity=2, maxLinearAcceleration=3000, maxLinearVelocity=600, velocityScaleZ=0, worldAlpha=1}
   }
   n103 = "Transform (Modify) Bone - Bone: mstr_Jacket_Right"!(ComponentPose=@n102, bAlphaBoolEnabled:bool=True, Translation="0,0,0", Rotation="0,0,0", Scale="1,1,1") -> (Pose) {
    alphaInputType=Bool
    rotationMode=BMM_Additive
    translationMode=BMM_Additive
    alphaBoolBlend {blendOption=Linear}
    boneToModify {boneName=mstr_Jacket_Right}
   }
   n104 = "Transform (Modify) Bone - Bone: mstr_Jacket_Left"!(ComponentPose=@n103, bAlphaBoolEnabled:bool=True, Translation="0,0,0", Rotation="0,0,0", Scale="1,1,1") -> (Pose) {
    alphaInputType=Bool
    rotationMode=BMM_Additive
    translationMode=BMM_Additive
    alphaBoolBlend {blendOption=Linear}
    boneToModify {boneName=mstr_Jacket_Left}
   }
   n105 = Anim Dynamics - Bone: dyn_jacket_front_01_l!(SimulationSpace:AnimPhysSimSpaceType Enum, ComponentPose=@n104, Alpha=0.8, GravityOverride:Vector="0,0,-980", ExternalForce="0,0,0") -> (Pose) {
    angularDampingOverride=0.9999
    angularSpringConstant=6000
    bAngularSpring=true
    bOverrideAngularDamping=true
    bUseGravityOverride=true
    gravityScale=0.35
    simulationSpace=RootRelative
    alphaBoolBlend {blendOption=Linear}
    boundBone {boneName=dyn_jacket_front_01_l}
    componentAppliedLinearAccClamp=<3000 3000 3000>
    componentLinearAccScale=<0.1 0.1 0.1>
    componentLinearVelScale=<0.1 0.1 0.1>
    gravityOverride {z=-980}
    physicsBodyDefinitions [{
      boundBone {boneName=dyn_jacket_front_01_l}
      boxExtents=<30 30 20>
      constraintSetup {angularLimitsMax=<0 0 0>, angularLimitsMin=<0 -30 -10>, angularTarget=<1 0 0>}
      localJointOffset=<3 0 0>
     }]
   }
   n106 = Anim Dynamics - Bone: dyn_jacket_front_02_l!(SimulationSpace:AnimPhysSimSpaceType Enum, ComponentPose=@n105, Alpha=0.8, GravityOverride:Vector="0,0,-1000", ExternalForce="0,0,0") -> (Pose) {
    angularDampingOverride=0.9999
    angularSpringConstant=6000
    bAngularSpring=true
    bOverrideAngularDamping=true
    bUseGravityOverride=true
    gravityScale=0.35
    simulationSpace=RootRelative
    alphaBoolBlend {blendOption=Linear}
    boundBone {boneName=dyn_jacket_front_02_l}
    componentAppliedLinearAccClamp=<3000 3000 3000>
    componentLinearAccScale=<0.1 0.1 0.1>
    componentLinearVelScale=<0.1 0.1 0.1>
    gravityOverride {z=-1000}
    physicsBodyDefinitions [{
      boundBone {boneName=dyn_jacket_front_02_l}
      boxExtents=<30 30 20>
      constraintSetup {angularLimitsMax=<0 0 0>, angularLimitsMin=<0 -25 -10>, angularTarget=<1 0 0>}
      localJointOffset=<3 0 0>
     }]
   }
   n107 = Anim Dynamics - Bone: dyn_jacket_front_03_l!(SimulationSpace:AnimPhysSimSpaceType Enum, ComponentPose=@n106, Alpha=0.8, GravityOverride:Vector="0,0,-1000", ExternalForce="0,0,0") -> (Pose) {
    angularDampingOverride=0.9999
    angularSpringConstant=6000
    bAngularSpring=true
    bOverrideAngularDamping=true
    bUseGravityOverride=true
    gravityScale=0.35
    simulationSpace=RootRelative
    alphaBoolBlend {blendOption=Linear}
    boundBone {boneName=dyn_jacket_front_03_l}
    componentAppliedLinearAccClamp=<3000 3000 3000>
    componentLinearAccScale=<0.1 0.1 0.1>
    componentLinearVelScale=<0.1 0.1 0.1>
    gravityOverride {z=-1000}
    physicsBodyDefinitions [{
      boundBone {boneName=dyn_jacket_front_03_l}
      boxExtents=<30 30 20>
      constraintSetup {angularLimitsMax=<0 0 0>, angularLimitsMin=<0 -20 -10>, angularTarget=<1 0 0>}
      localJointOffset=<3 0 0>
     }]
   }
   n108 = Anim Dynamics - Bone: dyn_jacket_front_04_l!(SimulationSpace:AnimPhysSimSpaceType Enum, ComponentPose=@n107, Alpha=0.5, GravityOverride:Vector="0,0,-1000", ExternalForce="0,0,0") -> (Pose) {
    angularDampingOverride=0.9999
    angularSpringConstant=6000
    bAngularSpring=true
    bOverrideAngularDamping=true
    bUseGravityOverride=true
    gravityScale=0.35
    simulationSpace=RootRelative
    alphaBoolBlend {blendOption=Linear}
    boundBone {boneName=dyn_jacket_front_04_l}
    componentAppliedLinearAccClamp=<3000 3000 3000>
    componentLinearAccScale=<0.1 0.1 0.1>
    componentLinearVelScale=<0.1 0.1 0.1>
    gravityOverride {z=-1000}
    physicsBodyDefinitions [{
      boundBone {boneName=dyn_jacket_front_04_l}
      boxExtents=<30 30 20>
      constraintSetup {angularLimitsMax=<0 0 0>, angularLimitsMin=<0 -15 -10>, angularTarget=<1 0 0>}
      localJointOffset=<3 0 0>
     }]
   }
   n109 = Anim Dynamics - Bone: mstr_jacket_side_01_l!(SimulationSpace:AnimPhysSimSpaceType Enum, ComponentPose=@n108, Alpha=0.8, GravityOverride:Vector="0,0,-980", ExternalForce="0,0,0") -> (Pose) {
    angularDampingOverride=0.9999
    angularSpringConstant=6000
    bAngularSpring=true
    bUseGravityOverride=true
    gravityScale=0.35
    simulationSpace=RootRelative
    alphaBoolBlend {blendOption=Linear}
    boundBone {boneName=mstr_jacket_side_01_l}
    componentAppliedLinearAccClamp=<3000 3000 3000>
    componentLinearAccScale=<0.1 0.1 0.1>
    componentLinearVelScale=<0.1 0.1 0.1>
    gravityOverride {z=-980}
    physicsBodyDefinitions [{
      boundBone {boneName=mstr_jacket_side_01_l}
      boxExtents=<30 30 20>
      constraintSetup {angularLimitsMax=<0 0 0>, angularLimitsMin=<0 -30 -10>, angularTarget=<1 0 0>}
      localJointOffset=<3 0 0>
     }]
   }
   n110 = Anim Dynamics - Bone: dyn_jacket_side_02_l!(SimulationSpace:AnimPhysSimSpaceType Enum, ComponentPose=@n109, Alpha=0.8, GravityOverride:Vector="0,0,-1000", ExternalForce="0,0,0") -> (Pose) {
    angularDampingOverride=0.9999
    angularSpringConstant=6000
    bAngularSpring=true
    bOverrideAngularDamping=true
    bUseGravityOverride=true
    gravityScale=0.35
    simulationSpace=RootRelative
    alphaBoolBlend {blendOption=Linear}
    boundBone {boneName=dyn_jacket_side_02_l}
    componentAppliedLinearAccClamp=<3000 3000 3000>
    componentLinearAccScale=<0.1 0.1 0.1>
    componentLinearVelScale=<0.1 0.1 0.1>
    gravityOverride {z=-1000}
    physicsBodyDefinitions [{
      boundBone {boneName=dyn_jacket_side_02_l}
      boxExtents=<30 30 20>
      constraintSetup {angularLimitsMax=<0 0 0>, angularLimitsMin=<0 -25 -10>, angularTarget=<1 0 0>}
      localJointOffset=<3 0 0>
     }]
   }
   n111 = Anim Dynamics - Bone: dyn_jacket_side_03_l!(SimulationSpace:AnimPhysSimSpaceType Enum, ComponentPose=@n110, Alpha=0.8, GravityOverride:Vector="0,0,-1000", ExternalForce="0,0,0") -> (Pose) {
    angularDampingOverride=0.9999
    angularSpringConstant=6000
    bAngularSpring=true
    bOverrideAngularDamping=true
    bUseGravityOverride=true
    gravityScale=0.35
    simulationSpace=RootRelative
    alphaBoolBlend {blendOption=Linear}
    boundBone {boneName=dyn_jacket_side_03_l}
    componentAppliedLinearAccClamp=<3000 3000 3000>
    componentLinearAccScale=<0.1 0.1 0.1>
    componentLinearVelScale=<0.1 0.1 0.1>
    gravityOverride {z=-1000}
    physicsBodyDefinitions [{
      boundBone {boneName=dyn_jacket_side_03_l}
      boxExtents=<30 30 20>
      constraintSetup {angularLimitsMax=<0 0 0>, angularLimitsMin=<0 -20 -10>, angularTarget=<1 0 0>}
      localJointOffset=<3 0 0>
     }]
   }
   n112 = Anim Dynamics - Bone: dyn_jacket_side_04_l!(SimulationSpace:AnimPhysSimSpaceType Enum, ComponentPose=@n111, Alpha=0.5, GravityOverride:Vector="0,0,-1000", ExternalForce="0,0,0") -> (Pose) {
    angularDampingOverride=0.9999
    angularSpringConstant=6000
    bAngularSpring=true
    bOverrideAngularDamping=true
    bUseGravityOverride=true
    gravityScale=0.35
    simulationSpace=RootRelative
    alphaBoolBlend {blendOption=Linear}
    boundBone {boneName=dyn_jacket_side_04_l}
    componentAppliedLinearAccClamp=<3000 3000 3000>
    componentLinearAccScale=<0.1 0.1 0.1>
    componentLinearVelScale=<0.1 0.1 0.1>
    gravityOverride {z=-1000}
    physicsBodyDefinitions [{
      boundBone {boneName=dyn_jacket_side_04_l}
      boxExtents=<30 30 20>
      constraintSetup {angularLimitsMax=<0 0 0>, angularLimitsMin=<0 -15 -10>, angularTarget=<1 0 0>}
      localJointOffset=<3 0 0>
     }]
   }
   n113 = Anim Dynamics - Bone: dyn_jacket_rear_01_l!(SimulationSpace:AnimPhysSimSpaceType Enum, ComponentPose=@n112, Alpha=0.8, GravityOverride:Vector="0,0,-980", ExternalForce="0,0,0") -> (Pose) {
    angularDampingOverride=0.9999
    angularSpringConstant=6000
    bAngularSpring=true
    bUseGravityOverride=true
    gravityScale=0.35
    simulationSpace=RootRelative
    alphaBoolBlend {blendOption=Linear}
    boundBone {boneName=dyn_jacket_rear_01_l}
    componentAppliedLinearAccClamp=<3000 3000 3000>
    componentLinearAccScale=<0.1 0.1 0.1>
    componentLinearVelScale=<0.1 0.1 0.1>
    gravityOverride {z=-980}
    physicsBodyDefinitions [{
      boundBone {boneName=dyn_jacket_rear_01_l}
      boxExtents=<30 30 20>
      constraintSetup {angularLimitsMax=<0 0 0>, angularLimitsMin=<0 -30 -10>, angularTarget=<1 0 0>}
      localJointOffset=<3 0 0>
     }]
   }
   n114 = Anim Dynamics - Bone: dyn_jacket_rear_02_l!(SimulationSpace:AnimPhysSimSpaceType Enum, ComponentPose=@n113, Alpha=0.8, GravityOverride:Vector="0,0,-1000", ExternalForce="0,0,0") -> (Pose) {
    angularDampingOverride=0.9999
    angularSpringConstant=6000
    bAngularSpring=true
    bOverrideAngularDamping=true
    bUseGravityOverride=true
    gravityScale=0.35
    simulationSpace=RootRelative
    alphaBoolBlend {blendOption=Linear}
    boundBone {boneName=dyn_jacket_rear_02_l}
    componentAppliedLinearAccClamp=<3000 3000 3000>
    componentLinearAccScale=<0.1 0.1 0.1>
    componentLinearVelScale=<0.1 0.1 0.1>
    gravityOverride {z=-1000}
    physicsBodyDefinitions [{
      boundBone {boneName=dyn_jacket_rear_02_l}
      boxExtents=<30 30 20>
      constraintSetup {angularLimitsMax=<0 0 0>, angularLimitsMin=<0 -25 -10>, angularTarget=<1 0 0>}
      localJointOffset=<3 0 0>
     }]
   }
   n115 = Anim Dynamics - Bone: dyn_jacket_rear_03_l!(SimulationSpace:AnimPhysSimSpaceType Enum, ComponentPose=@n114, Alpha=0.8, GravityOverride:Vector="0,0,-1000", ExternalForce="0,0,0") -> (Pose) {
    angularDampingOverride=0.9999
    angularSpringConstant=6000
    bAngularSpring=true
    bOverrideAngularDamping=true
    bUseGravityOverride=true
    gravityScale=0.35
    simulationSpace=RootRelative
    alphaBoolBlend {blendOption=Linear}
    boundBone {boneName=dyn_jacket_rear_03_l}
    componentAppliedLinearAccClamp=<3000 3000 3000>
    componentLinearAccScale=<0.1 0.1 0.1>
    componentLinearVelScale=<0.1 0.1 0.1>
    gravityOverride {z=-1000}
    physicsBodyDefinitions [{
      boundBone {boneName=dyn_jacket_rear_03_l}
      boxExtents=<30 30 20>
      constraintSetup {angularLimitsMax=<0 0 0>, angularLimitsMin=<0 -20 -10>, angularTarget=<1 0 0>}
      localJointOffset=<3 0 0>
     }]
   }
   n116 = Anim Dynamics - Bone: dyn_jacket_rear_04_l!(SimulationSpace:AnimPhysSimSpaceType Enum, ComponentPose=@n115, Alpha=0.5, GravityOverride:Vector="0,0,-1000", ExternalForce="0,0,0") -> (Pose) {
    angularDampingOverride=0.9999
    angularSpringConstant=6000
    bAngularSpring=true
    bOverrideAngularDamping=true
    bUseGravityOverride=true
    gravityScale=0.35
    simulationSpace=RootRelative
    alphaBoolBlend {blendOption=Linear}
    boundBone {boneName=dyn_jacket_rear_04_l}
    componentAppliedLinearAccClamp=<3000 3000 3000>
    componentLinearAccScale=<0.1 0.1 0.1>
    componentLinearVelScale=<0.1 0.1 0.1>
    gravityOverride {z=-1000}
    physicsBodyDefinitions [{
      boundBone {boneName=dyn_jacket_rear_04_l}
      boxExtents=<30 30 20>
      constraintSetup {angularLimitsMax=<0 0 0>, angularLimitsMin=<0 -15 -10>, angularTarget=<1 0 0>}
      localJointOffset=<3 0 0>
     }]
   }
   n117 = Anim Dynamics - Bone: dyn_jacket_front_01_r!(SimulationSpace:AnimPhysSimSpaceType Enum, ComponentPose=@n116, Alpha=0.8, GravityOverride:Vector="0,0,-980", ExternalForce="0,0,0") -> (Pose) {
    angularDampingOverride=0.9999
    angularSpringConstant=6000
    bAngularSpring=true
    bOverrideAngularDamping=true
    bUseGravityOverride=true
    gravityScale=0.35
    simulationSpace=RootRelative
    alphaBoolBlend {blendOption=Linear}
    boundBone {boneName=dyn_jacket_front_01_r}
    componentAppliedLinearAccClamp=<3000 3000 3000>
    componentLinearAccScale=<0.1 0.1 0.1>
    componentLinearVelScale=<0.1 0.1 0.1>
    gravityOverride {z=-980}
    physicsBodyDefinitions [{
      boundBone {boneName=dyn_jacket_front_01_r}
      boxExtents=<30 30 20>
      constraintSetup {angularLimitsMax=<0 0 0>, angularLimitsMin=<0 -30 -10>, angularTarget=<1 0 0>}
      localJointOffset=<3 0 0>
     }]
   }
   n118 = Anim Dynamics - Bone: dyn_jacket_front_02_r!(SimulationSpace:AnimPhysSimSpaceType Enum, ComponentPose=@n117, Alpha=0.8, GravityOverride:Vector="0,0,-1000", ExternalForce="0,0,0") -> (Pose) {
    angularDampingOverride=0.9999
    angularSpringConstant=6000
    bAngularSpring=true
    bOverrideAngularDamping=true
    bUseGravityOverride=true
    gravityScale=0.35
    simulationSpace=RootRelative
    alphaBoolBlend {blendOption=Linear}
    boundBone {boneName=dyn_jacket_front_02_r}
    componentAppliedLinearAccClamp=<3000 3000 3000>
    componentLinearAccScale=<0.1 0.1 0.1>
    componentLinearVelScale=<0.1 0.1 0.1>
    gravityOverride {z=-1000}
    physicsBodyDefinitions [{
      boundBone {boneName=dyn_jacket_front_02_r}
      boxExtents=<30 30 20>
      constraintSetup {angularLimitsMax=<0 0 0>, angularLimitsMin=<0 -25 -10>, angularTarget=<1 0 0>}
      localJointOffset=<3 0 0>
     }]
   }
   n119 = Anim Dynamics - Bone: dyn_jacket_front_03_r!(SimulationSpace:AnimPhysSimSpaceType Enum, ComponentPose=@n118, Alpha=0.8, GravityOverride:Vector="0,0,-1000", ExternalForce="0,0,0") -> (Pose) {
    angularDampingOverride=0.9999
    angularSpringConstant=6000
    bAngularSpring=true
    bOverrideAngularDamping=true
    bUseGravityOverride=true
    gravityScale=0.35
    simulationSpace=RootRelative
    alphaBoolBlend {blendOption=Linear}
    boundBone {boneName=dyn_jacket_front_03_r}
    componentAppliedLinearAccClamp=<3000 3000 3000>
    componentLinearAccScale=<0.1 0.1 0.1>
    componentLinearVelScale=<0.1 0.1 0.1>
    gravityOverride {z=-1000}
    physicsBodyDefinitions [{
      boundBone {boneName=dyn_jacket_front_03_r}
      boxExtents=<30 30 20>
      constraintSetup {angularLimitsMax=<0 0 0>, angularLimitsMin=<0 -20 -10>, angularTarget=<1 0 0>}
      localJointOffset=<3 0 0>
     }]
   }
   n120 = Anim Dynamics - Bone: dyn_jacket_front_04_r!(SimulationSpace:AnimPhysSimSpaceType Enum, ComponentPose=@n119, Alpha=0.5, GravityOverride:Vector="0,0,-1000", ExternalForce="0,0,0") -> (Pose) {
    angularDampingOverride=0.9999
    angularSpringConstant=6000
    bAngularSpring=true
    bOverrideAngularDamping=true
    bUseGravityOverride=true
    gravityScale=0.35
    simulationSpace=RootRelative
    alphaBoolBlend {blendOption=Linear}
    boundBone {boneName=dyn_jacket_front_04_r}
    componentAppliedLinearAccClamp=<3000 3000 3000>
    componentLinearAccScale=<0.1 0.1 0.1>
    componentLinearVelScale=<0.1 0.1 0.1>
    gravityOverride {z=-1000}
    physicsBodyDefinitions [{
      boundBone {boneName=dyn_jacket_front_04_r}
      boxExtents=<30 30 20>
      constraintSetup {angularLimitsMax=<0 0 0>, angularLimitsMin=<0 -15 -10>, angularTarget=<1 0 0>}
      localJointOffset=<3 0 0>
     }]
   }
   n121 = Anim Dynamics - Bone: dyn_jacket_side_01_r!(SimulationSpace:AnimPhysSimSpaceType Enum, ComponentPose=@n120, Alpha=0.8, GravityOverride:Vector="0,0,-980", ExternalForce="0,0,0") -> (Pose) {
    angularDampingOverride=0.9999
    angularSpringConstant=6000
    bAngularSpring=true
    bUseGravityOverride=true
    gravityScale=0.35
    simulationSpace=RootRelative
    alphaBoolBlend {blendOption=Linear}
    boundBone {boneName=dyn_jacket_side_01_r}
    componentAppliedLinearAccClamp=<3000 3000 3000>
    componentLinearAccScale=<0.1 0.1 0.1>
    componentLinearVelScale=<0.1 0.1 0.1>
    gravityOverride {z=-980}
    physicsBodyDefinitions [{
      boundBone {boneName=dyn_jacket_side_01_r}
      boxExtents=<30 30 20>
      constraintSetup {angularLimitsMax=<0 0 0>, angularLimitsMin=<0 -30 -10>, angularTarget=<1 0 0>}
      localJointOffset=<3 0 0>
     }]
   }
   n122 = Anim Dynamics - Bone: dyn_jacket_side_02_r!(SimulationSpace:AnimPhysSimSpaceType Enum, ComponentPose=@n121, Alpha=0.8, GravityOverride:Vector="0,0,-1000", ExternalForce="0,0,0") -> (Pose) {
    angularDampingOverride=0.9999
    angularSpringConstant=6000
    bAngularSpring=true
    bOverrideAngularDamping=true
    bUseGravityOverride=true
    gravityScale=0.35
    simulationSpace=RootRelative
    alphaBoolBlend {blendOption=Linear}
    boundBone {boneName=dyn_jacket_side_02_r}
    componentAppliedLinearAccClamp=<3000 3000 3000>
    componentLinearAccScale=<0.1 0.1 0.1>
    componentLinearVelScale=<0.1 0.1 0.1>
    gravityOverride {z=-1000}
    physicsBodyDefinitions [{
      boundBone {boneName=dyn_jacket_side_02_r}
      boxExtents=<30 30 20>
      constraintSetup {angularLimitsMax=<0 0 0>, angularLimitsMin=<0 -25 -10>, angularTarget=<1 0 0>}
      localJointOffset=<3 0 0>
     }]
   }
   n123 = Anim Dynamics - Bone: dyn_jacket_side_03_r!(SimulationSpace:AnimPhysSimSpaceType Enum, ComponentPose=@n122, Alpha=0.8, GravityOverride:Vector="0,0,-1000", ExternalForce="0,0,0") -> (Pose) {
    angularDampingOverride=0.9999
    angularSpringConstant=6000
    bAngularSpring=true
    bOverrideAngularDamping=true
    bUseGravityOverride=true
    gravityScale=0.35
    simulationSpace=RootRelative
    alphaBoolBlend {blendOption=Linear}
    boundBone {boneName=dyn_jacket_side_03_r}
    componentAppliedLinearAccClamp=<3000 3000 3000>
    componentLinearAccScale=<0.1 0.1 0.1>
    componentLinearVelScale=<0.1 0.1 0.1>
    gravityOverride {z=-1000}
    physicsBodyDefinitions [{
      boundBone {boneName=dyn_jacket_side_03_r}
      boxExtents=<30 30 20>
      constraintSetup {angularLimitsMax=<0 0 0>, angularLimitsMin=<0 -20 -10>, angularTarget=<1 0 0>}
      localJointOffset=<3 0 0>
     }]
   }
   n124 = Anim Dynamics - Bone: dyn_jacket_side_04_r!(SimulationSpace:AnimPhysSimSpaceType Enum, ComponentPose=@n123, Alpha=0.5, GravityOverride:Vector="0,0,-1000", ExternalForce="0,0,0") -> (Pose) {
    angularDampingOverride=0.9999
    angularSpringConstant=6000
    bAngularSpring=true
    bOverrideAngularDamping=true
    bUseGravityOverride=true
    gravityScale=0.35
    simulationSpace=RootRelative
    alphaBoolBlend {blendOption=Linear}
    boundBone {boneName=dyn_jacket_side_04_r}
    componentAppliedLinearAccClamp=<3000 3000 3000>
    componentLinearAccScale=<0.1 0.1 0.1>
    componentLinearVelScale=<0.1 0.1 0.1>
    gravityOverride {z=-1000}
    physicsBodyDefinitions [{
      boundBone {boneName=dyn_jacket_side_04_r}
      boxExtents=<30 30 20>
      constraintSetup {angularLimitsMax=<0 0 0>, angularLimitsMin=<0 -15 -10>, angularTarget=<1 0 0>}
      localJointOffset=<3 0 0>
     }]
   }
   n125 = Anim Dynamics - Bone: dyn_jacket_rear_01_r!(SimulationSpace:AnimPhysSimSpaceType Enum, ComponentPose=@n124, Alpha=0.8, GravityOverride:Vector="0,0,-980", ExternalForce="0,0,0") -> (Pose) {
    angularDampingOverride=0.9999
    angularSpringConstant=6000
    bAngularSpring=true
    bUseGravityOverride=true
    gravityScale=0.35
    simulationSpace=RootRelative
    alphaBoolBlend {blendOption=Linear}
    boundBone {boneName=dyn_jacket_rear_01_r}
    componentAppliedLinearAccClamp=<3000 3000 3000>
    componentLinearAccScale=<0.1 0.1 0.1>
    componentLinearVelScale=<0.1 0.1 0.1>
    gravityOverride {z=-980}
    physicsBodyDefinitions [{
      boundBone {boneName=dyn_jacket_rear_01_r}
      boxExtents=<30 30 20>
      constraintSetup {angularLimitsMax=<0 0 0>, angularLimitsMin=<0 -30 -10>, angularTarget=<1 0 0>}
      localJointOffset=<3 0 0>
     }]
   }
   n126 = Anim Dynamics - Bone: dyn_jacket_rear_02_r!(SimulationSpace:AnimPhysSimSpaceType Enum, ComponentPose=@n125, Alpha=0.8, GravityOverride:Vector="0,0,-1000", ExternalForce="0,0,0") -> (Pose) {
    angularDampingOverride=0.9999
    angularSpringConstant=6000
    bAngularSpring=true
    bOverrideAngularDamping=true
    bUseGravityOverride=true
    gravityScale=0.35
    simulationSpace=RootRelative
    alphaBoolBlend {blendOption=Linear}
    boundBone {boneName=dyn_jacket_rear_02_r}
    componentAppliedLinearAccClamp=<3000 3000 3000>
    componentLinearAccScale=<0.1 0.1 0.1>
    componentLinearVelScale=<0.1 0.1 0.1>
    gravityOverride {z=-1000}
    physicsBodyDefinitions [{
      boundBone {boneName=dyn_jacket_rear_02_r}
      boxExtents=<30 30 20>
      constraintSetup {angularLimitsMax=<0 0 0>, angularLimitsMin=<0 -25 -10>, angularTarget=<1 0 0>}
      localJointOffset=<3 0 0>
     }]
   }
   n127 = Anim Dynamics - Bone: dyn_jacket_rear_03_r!(SimulationSpace:AnimPhysSimSpaceType Enum, ComponentPose=@n126, Alpha=0.8, GravityOverride:Vector="0,0,-1000", ExternalForce="0,0,0") -> (Pose) {
    angularDampingOverride=0.9999
    angularSpringConstant=6000
    bAngularSpring=true
    bOverrideAngularDamping=true
    bUseGravityOverride=true
    gravityScale=0.35
    simulationSpace=RootRelative
    alphaBoolBlend {blendOption=Linear}
    boundBone {boneName=dyn_jacket_rear_03_r}
    componentAppliedLinearAccClamp=<3000 3000 3000>
    componentLinearAccScale=<0.1 0.1 0.1>
    componentLinearVelScale=<0.1 0.1 0.1>
    gravityOverride {z=-1000}
    physicsBodyDefinitions [{
      boundBone {boneName=dyn_jacket_rear_03_r}
      boxExtents=<30 30 20>
      constraintSetup {angularLimitsMax=<0 0 0>, angularLimitsMin=<0 -20 -10>, angularTarget=<1 0 0>}
      localJointOffset=<3 0 0>
     }]
   }
   n128 = Anim Dynamics - Bone: dyn_jacket_rear_04_r!(SimulationSpace:AnimPhysSimSpaceType Enum, ComponentPose=@n127, Alpha=0.5, GravityOverride:Vector="0,0,-1000", ExternalForce="0,0,0") -> (Pose) {
    angularDampingOverride=0.9999
    angularSpringConstant=6000
    bAngularSpring=true
    bOverrideAngularDamping=true
    bUseGravityOverride=true
    gravityScale=0.35
    simulationSpace=RootRelative
    alphaBoolBlend {blendOption=Linear}
    boundBone {boneName=dyn_jacket_rear_04_r}
    componentAppliedLinearAccClamp=<3000 3000 3000>
    componentLinearAccScale=<0.1 0.1 0.1>
    componentLinearVelScale=<0.1 0.1 0.1>
    gravityOverride {z=-1000}
    physicsBodyDefinitions [{
      boundBone {boneName=dyn_jacket_rear_04_r}
      boxExtents=<30 30 20>
      constraintSetup {angularLimitsMax=<0 0 0>, angularLimitsMin=<0 -15 -10>, angularTarget=<1 0 0>}
      localJointOffset=<3 0 0>
     }]
   }
   n129 = Anim Dynamics - Bone: dyn_lapel_l(ComponentPose=@n128, Alpha=1, ExternalForce="0,0,0") {
    angularDampingOverride=1
    angularSpringConstant=12000
    bAngularSpring=true
    bOverrideAngularDamping=true
    bUseGravityOverride=true
    bUseSphericalLimits=true
    gravityScale=0.35
    simulationSpace=RootRelative
    alphaBoolBlend {blendOption=Linear}
    boundBone {boneName=dyn_lapel_l}
    componentLinearAccScale=<1 1 1>
    componentLinearVelScale=<1 1 1>
    gravityOverride {z=-980}
    physicsBodyDefinitions [{
      boundBone {boneName=dyn_lapel_l}
      boxExtents=<12 22 22>
      constraintSetup {angularLimitsMax=<0 0 30>, angularLimitsMin=<0 0 -10>, angularTarget=<1 0 0>}
      localJointOffset=<2 0 0>
     }]
    sphericalLimits [
     {limitRadius=11, drivingBone {boneName=upperarm_l}, sphereLocalOffset=<4 2 -2>}
     {limitRadius=8, drivingBone {boneName=clavicle_l}, sphereLocalOffset=<8 0 0>}
    ]
   }
   n130 = Anim Dynamics - Bone: dyn_lapel_up_l(ComponentPose=@n129, Alpha=1, ExternalForce="0,0,0") {
    angularDampingOverride=1
    angularSpringConstant=10000
    bAngularSpring=true
    bOverrideAngularDamping=true
    bUseGravityOverride=true
    bUseSphericalLimits=true
    gravityScale=0.35
    simulationSpace=RootRelative
    alphaBoolBlend {blendOption=Linear}
    boundBone {boneName=dyn_lapel_up_l}
    componentLinearAccScale=<1 1 1>
    componentLinearVelScale=<1 1 1>
    gravityOverride {z=-980}
    physicsBodyDefinitions [{
      boundBone {boneName=dyn_lapel_up_l}
      boxExtents=<11 21 21>
      constraintSetup {angularLimitsMax=<0 0 30>, angularLimitsMin=<0 0 -10>, angularTarget=<1 0 0>}
      localJointOffset=<2 0 0>
     }]
    sphericalLimits [
     {limitRadius=6, drivingBone {boneName=clavicle_l}, sphereLocalOffset=<10 0 0>}
     {limitRadius=9, drivingBone {boneName=upperarm_l}, sphereLocalOffset=<0 0 0>}
    ]
   }
   n131 = Anim Dynamics - Bone: dyn_lapel_r(ComponentPose=@n130, Alpha=1, ExternalForce="0,0,0") {
    angularDampingOverride=1
    angularSpringConstant=12000
    bAngularSpring=true
    bOverrideAngularDamping=true
    bUseGravityOverride=true
    bUseSphericalLimits=true
    gravityScale=0.35
    simulationSpace=RootRelative
    alphaBoolBlend {blendOption=Linear}
    boundBone {boneName=dyn_lapel_r}
    componentLinearAccScale=<1 1 1>
    componentLinearVelScale=<1 1 1>
    gravityOverride {z=-980}
    physicsBodyDefinitions [{
      boundBone {boneName=dyn_lapel_r}
      boxExtents=<12 22 22>
      constraintSetup {angularLimitsMax=<0 0 30>, angularLimitsMin=<0 0 -10>, angularTarget=<1 0 0>}
      localJointOffset=<2 0 0>
     }]
    sphericalLimits [
     {limitRadius=12, drivingBone {boneName=upperarm_r}, sphereLocalOffset=<-8 2 2>}
     {limitRadius=8, drivingBone {boneName=clavicle_r}, sphereLocalOffset=<-8 0 0>}
    ]
   }
   n132 = Anim Dynamics - Bone: dyn_lapel_up_r(ComponentPose=@n131, Alpha=1, ExternalForce="0,0,0") {
    angularDampingOverride=1
    angularSpringConstant=10000
    bAngularSpring=true
    bOverrideAngularDamping=true
    bUseGravityOverride=true
    bUseSphericalLimits=true
    gravityScale=0.35
    simulationSpace=RootRelative
    alphaBoolBlend {blendOption=Linear}
    boundBone {boneName=dyn_lapel_up_r}
    componentLinearAccScale=<1 1 1>
    componentLinearVelScale=<1 1 1>
    gravityOverride {z=-980}
    physicsBodyDefinitions [{
      boundBone {boneName=dyn_lapel_up_r}
      boxExtents=<11 21 21>
      constraintSetup {angularLimitsMax=<0 0 30>, angularLimitsMin=<0 0 -10>, angularTarget=<1 0 0>}
      localJointOffset=<2 0 0>
     }]
    sphericalLimits [{limitRadius=10, drivingBone {boneName=head}, sphereLocalOffset=<0 0 0>}]
   }
   n133 = Anim Dynamics - Bone: dyn_collar(ComponentPose=@n132, Alpha=1, ExternalForce="0,0,0") {
    angularDampingOverride=1
    angularSpringConstant=12000
    bAngularSpring=true
    bOverrideAngularDamping=true
    bUseGravityOverride=true
    bUseSphericalLimits=true
    gravityScale=0.35
    simulationSpace=RootRelative
    alphaBoolBlend {blendOption=Linear}
    boundBone {boneName=dyn_collar}
    componentLinearAccScale=<1 1 1>
    componentLinearVelScale=<1 1 1>
    gravityOverride {z=-980}
    physicsBodyDefinitions [{
      boundBone {boneName=dyn_collar}
      boxExtents=<12 22 22>
      constraintSetup {angularLimitsMax=<0 0 50>, angularLimitsMin=<0 0 -2>, angularTarget=<1 0 0>}
      localJointOffset=<2 0 0>
     }]
    sphericalLimits [{limitRadius=18, drivingBone {boneName=head}, sphereLocalOffset=<0 0 0>}]
   }
   n134 = Anim Dynamics - Bone: dyn_Chain_01(ComponentPose=@n133, Alpha=1, ExternalForce="0,0,0") {
    angularDampingOverride=1
    angularSpringConstant=5000
    bAngularSpring=true
    bOverrideAngularDamping=true
    bUseGravityOverride=true
    bUseSphericalLimits=true
    gravityScale=0.35
    simulationSpace=RootRelative
    alphaBoolBlend {blendOption=Linear}
    boundBone {boneName=dyn_Chain_01}
    gravityOverride {z=-980}
    physicsBodyDefinitions [{
      boundBone {boneName=dyn_Chain_01}
      boxExtents=<20 20 20>
      constraintSetup {angularLimitsMax=<0 0 10>, angularLimitsMin=<0 -10 -10>, angularTarget=<1 0 0>}
      localJointOffset=<0.2 0 0>
     }]
   }
   n135 = Get Sprinting() -> (Sprinting:double)
   n136 = "Transform (Modify) Bone - Bone: mstr_jacket_front_01_l"(ComponentPose=@n134, Alpha=@n135, Translation="0,0,0", Rotation="-15,0,0", Scale="1,1,1") {
    rotationMode=BMM_Additive
    rotationSpace=BCS_BoneSpace
    translationMode=BMM_Additive
    translationSpace=BCS_BoneSpace
    alphaBoolBlend {blendOption=Linear}
    alphaScaleBiasClamp {bInterpResult=true}
    boneToModify {boneName=mstr_jacket_front_01_l}
    rotation {pitch=-15}
   }
   n137 = "Transform (Modify) Bone - Bone: mstr_jacket_front_01_r"(ComponentPose=@n136, Alpha=@n135, Translation="0,0,0", Rotation="-10,0,0", Scale="1,1,1") {
    rotationMode=BMM_Additive
    rotationSpace=BCS_BoneSpace
    translationMode=BMM_Additive
    translationSpace=BCS_BoneSpace
    alphaBoolBlend {blendOption=Linear}
    alphaScaleBiasClamp {bInterpResult=true}
    boneToModify {boneName=mstr_jacket_front_01_r}
    rotation {pitch=-10}
   }
   n138 = Get ADSWalk() -> (ADSWalk:float)
   n139 = "Transform (Modify) Bone - Bone: mstr_jacket_rear_01_l"(ComponentPose=@n137, Alpha=@n138, Translation="0,0,0", Rotation="-10,0,0", Scale="1,1,1") {
    rotationMode=BMM_Additive
    rotationSpace=BCS_BoneSpace
    translationMode=BMM_Additive
    translationSpace=BCS_BoneSpace
    alphaBoolBlend {blendOption=Linear}
    alphaScaleBiasClamp {bInterpResult=true, interpSpeedDecreasing=5, interpSpeedIncreasing=2}
    boneToModify {boneName=mstr_jacket_rear_01_l}
    rotation {pitch=-10}
   }
   n140 = "Transform (Modify) Bone - Bone: dyn_jacket_side_01_l"(ComponentPose=@n139, Alpha=@n138, Translation="0,0,0", Rotation="-10,0,0", Scale="1,1,1") {
    rotationMode=BMM_Additive
    rotationSpace=BCS_BoneSpace
    translationMode=BMM_Additive
    translationSpace=BCS_BoneSpace
    alphaBoolBlend {blendOption=Linear}
    alphaScaleBiasClamp {bInterpResult=true, interpSpeedDecreasing=5, interpSpeedIncreasing=2}
    boneToModify {boneName=dyn_jacket_side_01_l}
    rotation {pitch=-10}
   }
   n141 = "Transform (Modify) Bone - Bone: mstr_jacket_rear_01_l"(ComponentPose=@n140, Alpha=@n138, Translation="0,0,0", Rotation="-10,0,0", Scale="1,1,1") {
    rotationMode=BMM_Additive
    rotationSpace=BCS_BoneSpace
    translationMode=BMM_Additive
    translationSpace=BCS_BoneSpace
    alphaBoolBlend {blendOption=Linear}
    alphaScaleBiasClamp {bInterpResult=true, interpSpeedDecreasing=5, interpSpeedIncreasing=2}
    boneToModify {boneName=mstr_jacket_rear_01_l}
    rotation {pitch=-10}
   }
   n142 = "Transform (Modify) Bone - Bone: mstr_jacket_rear_01_r"(ComponentPose=@n141, Alpha=@n138, Translation="0,0,0", Rotation="-10,0,0", Scale="1,1,1") {
    rotationMode=BMM_Additive
    rotationSpace=BCS_BoneSpace
    translationMode=BMM_Additive
    translationSpace=BCS_BoneSpace
    alphaBoolBlend {blendOption=Linear}
    alphaScaleBiasClamp {bInterpResult=true, interpSpeedDecreasing=5, interpSpeedIncreasing=2}
    boneToModify {boneName=mstr_jacket_rear_01_r}
    rotation {pitch=-10}
   }
   n143 = "Transform (Modify) Bone - Bone: dyn_jacket_rear_01_l"!(ComponentPose=@n142, bAlphaBoolEnabled:bool=True, Translation="0,0,0", Rotation="-24.130,1.477,15.181", Scale="1,1,1") -> (Pose) {
    alphaInputType=Bool
    rotationMode=BMM_Additive
    rotationSpace=BCS_BoneSpace
    translationMode=BMM_Additive
    translationSpace=BCS_BoneSpace
    alphaBoolBlend {blendInTime=0.5, blendOption=Linear}
    alphaScaleBiasClamp {bInterpResult=true}
    boneToModify {boneName=dyn_jacket_rear_01_l}
    rotation=<rot -24.13 1.477 15.181>
   }
   n144 = "Transform (Modify) Bone - Bone: dyn_jacket_side_01_l"!(ComponentPose=@n143, bAlphaBoolEnabled:bool=True, Translation="0,0,0", Rotation="-20,0,30", Scale="1,1,1") -> (Pose) {
    alphaInputType=Bool
    rotationMode=BMM_Additive
    rotationSpace=BCS_BoneSpace
    translationMode=BMM_Additive
    translationSpace=BCS_BoneSpace
    alphaBoolBlend {blendInTime=0.5, blendOption=Linear}
    alphaScaleBiasClamp {bInterpResult=true}
    boneToModify {boneName=dyn_jacket_side_01_l}
    rotation {pitch=-20, roll=30}
   }
   n145 = "Transform (Modify) Bone - Bone: dyn_jacket_side_03_l"(ComponentPose=@n144, bAlphaBoolEnabled=True, Translation="0,0,0", Rotation="0,0,20", Scale="1,1,1") {
    alphaInputType=Bool
    rotationMode=BMM_Additive
    rotationSpace=BCS_BoneSpace
    translationMode=BMM_Additive
    translationSpace=BCS_BoneSpace
    alphaBoolBlend {blendInTime=0.5, blendOption=Linear}
    alphaScaleBiasClamp {bInterpResult=true}
    boneToModify {boneName=dyn_jacket_side_03_l}
    rotation {roll=20}
   }
   n146 = "Transform (Modify) Bone - Bone: dyn_jacket_rear_01_r"!(ComponentPose=@n145, bAlphaBoolEnabled:bool=True, Translation="0,0,0", Rotation="-10,0,0", Scale="1,1,1") -> (Pose) {
    alphaInputType=Bool
    rotationMode=BMM_Additive
    rotationSpace=BCS_BoneSpace
    translationMode=BMM_Additive
    translationSpace=BCS_BoneSpace
    alphaBoolBlend {blendInTime=0.5, blendOption=Linear}
    alphaScaleBiasClamp {bInterpResult=true}
    boneToModify {boneName=dyn_jacket_rear_01_r}
    rotation {pitch=-10}
   }
   n147 = "Transform (Modify) Bone - Bone: dyn_jacket_side_01_r"!(ComponentPose=@n146, bAlphaBoolEnabled:bool=True, Translation="0,0,0", Rotation="-5,0,0", Scale="1,1,1") -> (Pose) {
    alphaInputType=Bool
    rotationMode=BMM_Additive
    rotationSpace=BCS_BoneSpace
    translationMode=BMM_Additive
    translationSpace=BCS_BoneSpace
    alphaBoolBlend {blendInTime=0.5, blendOption=Linear}
    alphaScaleBiasClamp {bInterpResult=true}
    boneToModify {boneName=dyn_jacket_side_01_r}
    rotation {pitch=-5}
   }
   n148 = "Transform (Modify) Bone - Bone: dyn_jacket_front_02_l"!(ComponentPose=@n147, bAlphaBoolEnabled:bool=True, Translation="0,0,0", Rotation="0,15,0", Scale="1,1,1") -> (Pose) {
    alphaInputType=Bool
    rotationMode=BMM_Additive
    rotationSpace=BCS_BoneSpace
    translationMode=BMM_Additive
    translationSpace=BCS_BoneSpace
    alphaBoolBlend {blendInTime=0.5, blendOption=Linear}
    alphaScaleBiasClamp {bInterpResult=true}
    boneToModify {boneName=dyn_jacket_front_02_l}
    rotation {yaw=15}
   }
   n149 = "Transform (Modify) Bone - Bone: mstr_jacket_front_01_l"!(ComponentPose=@n148, bAlphaBoolEnabled:bool=True, Translation="0,0,0", Rotation="-5,10,20", Scale="1,1,1") -> (Pose) {
    alphaInputType=Bool
    rotationMode=BMM_Additive
    rotationSpace=BCS_BoneSpace
    translationMode=BMM_Additive
    translationSpace=BCS_BoneSpace
    alphaBoolBlend {blendInTime=0.5, blendOption=Linear}
    alphaScaleBiasClamp {bInterpResult=true}
    boneToModify {boneName=mstr_jacket_front_01_l}
    rotation=<rot -5 10 20>
   }
   n150 = NOT Boolean(A=Get bIsCrouching())
   n151 = "Transform (Modify) Bone - Bone: mstr_jacket_rear_01_l"!(ComponentPose=@n149, bAlphaBoolEnabled:bool=@n150, Translation="0,-2,0", Rotation="0,0,0", Scale="1,1,1") -> (Pose) {
    alphaInputType=Bool
    rotationMode=BMM_Additive
    scaleMode=BMM_Additive
    translationMode=BMM_Additive
    alphaBoolBlend {blendOption=Linear}
    boneToModify {boneName=mstr_jacket_rear_01_l}
    translation {y=-2}
   }
   n152 = "Transform (Modify) Bone - Bone: mstr_jacket_rear_01_r"!(ComponentPose=@n151, bAlphaBoolEnabled:bool=@n150, Translation="0,-2,0", Rotation="0,0,0", Scale="1,1,1") -> (Pose) {
    alphaInputType=Bool
    rotationMode=BMM_Additive
    scaleMode=BMM_Additive
    translationMode=BMM_Additive
    alphaBoolBlend {blendOption=Linear}
    boneToModify {boneName=mstr_jacket_rear_01_r}
    translation {y=-2}
   }
   n153 = Get bIsInFrontEnd()
   n154 = "Transform (Modify) Bone - Bone: mstr_jacket_front_01_l"(ComponentPose=@n152, bAlphaBoolEnabled=@n153, Translation="0,0,0", Rotation="5,5,0", Scale="1,1,1") {
    alphaInputType=Bool
    rotationMode=BMM_Additive
    rotationSpace=BCS_BoneSpace
    translationMode=BMM_Additive
    translationSpace=BCS_BoneSpace
    alphaBoolBlend {blendOption=Linear}
    alphaScaleBiasClamp {bInterpResult=true}
    boneToModify {boneName=mstr_jacket_front_01_l}
    rotation {pitch=5, yaw=5}
   }
   n155 = "Transform (Modify) Bone - Bone: dyn_jacket_rear_02_l"(ComponentPose=@n154, bAlphaBoolEnabled=@n153, Translation="0,0,0", Rotation="0,0,0", Scale="1,1,1") {
    alphaInputType=Bool
    rotationMode=BMM_Additive
    rotationSpace=BCS_BoneSpace
    translationMode=BMM_Additive
    translationSpace=BCS_BoneSpace
    alphaBoolBlend {blendOption=Linear}
    alphaScaleBiasClamp {bInterpResult=true}
    boneToModify {boneName=dyn_jacket_rear_02_l}
   }
   n156 = "Transform (Modify) Bone - Bone: dyn_jacket_front_03_l"!(ComponentPose=@n155, bAlphaBoolEnabled:bool=@n153, Translation="-2,0,0", Rotation="0,0,0", Scale="1,1,1") -> (Pose) {
    alphaInputType=Bool
    rotationMode=BMM_Additive
    rotationSpace=BCS_BoneSpace
    translationMode=BMM_Additive
    translationSpace=BCS_BoneSpace
    alphaBoolBlend {blendOption=Linear}
    alphaScaleBiasClamp {bInterpResult=true}
    boneToModify {boneName=dyn_jacket_front_03_l}
    translation {x=-2}
   }
   n157 = "Transform (Modify) Bone - Bone: dyn_jacket_rear_04_l"(ComponentPose=@n156, bAlphaBoolEnabled=True, Translation="0,0,0", Rotation="0,0,0", Scale="1,1,1") {
    alphaInputType=Bool
    rotationMode=BMM_Additive
    rotationSpace=BCS_BoneSpace
    translationMode=BMM_Additive
    translationSpace=BCS_BoneSpace
    alphaBoolBlend {blendOption=Linear}
    alphaScaleBiasClamp {bInterpResult=true}
    boneToModify {boneName=dyn_jacket_rear_04_l}
   }
   n158 = "Transform (Modify) Bone - Bone: dyn_jacket_side_01_l"(ComponentPose=@n157, bAlphaBoolEnabled=@n153, Translation="-2,0,3", Rotation="0,0,0.285", Scale="1,1,1") {
    alphaInputType=Bool
    rotationMode=BMM_Additive
    rotationSpace=BCS_BoneSpace
    translationMode=BMM_Additive
    translationSpace=BCS_BoneSpace
    alphaBoolBlend {blendOption=Linear}
    alphaScaleBiasClamp {bInterpResult=true}
    boneToModify {boneName=dyn_jacket_side_01_l}
    rotation {roll=0.285}
    translation {x=-2, z=3}
   }
   n159 = "Transform (Modify) Bone - Bone: dyn_jacket_side_02_l"!(ComponentPose=@n158, bAlphaBoolEnabled:bool=@n153, Translation="0,0,0", Rotation="0,0,0", Scale="1,1,1") -> (Pose) {
    alphaInputType=Bool
    rotationMode=BMM_Additive
    rotationSpace=BCS_BoneSpace
    translationMode=BMM_Additive
    translationSpace=BCS_BoneSpace
    alphaBoolBlend {blendOption=Linear}
    alphaScaleBiasClamp {bInterpResult=true}
    boneToModify {boneName=dyn_jacket_side_02_l}
   }
   n160 = "Transform (Modify) Bone - Bone: dyn_jacket_side_03_l"(ComponentPose=@n159, bAlphaBoolEnabled=@n153, Translation="0,0,0", Rotation="0,0,0", Scale="1,1,1") {
    alphaInputType=Bool
    rotationMode=BMM_Additive
    rotationSpace=BCS_BoneSpace
    translationMode=BMM_Additive
    translationSpace=BCS_BoneSpace
    alphaBoolBlend {blendOption=Linear}
    alphaScaleBiasClamp {bInterpResult=true}
    boneToModify {boneName=dyn_jacket_side_03_l}
   }
   n161 = "Transform (Modify) Bone - Bone: dyn_jacket_side_04_l"(ComponentPose=@n160, bAlphaBoolEnabled=True, Translation="0,0,0", Rotation="0,0,0", Scale="1,1,1") {
    alphaInputType=Bool
    rotationMode=BMM_Additive
    rotationSpace=BCS_BoneSpace
    translationMode=BMM_Additive
    translationSpace=BCS_BoneSpace
    alphaBoolBlend {blendOption=Linear}
    alphaScaleBiasClamp {bInterpResult=true}
    boneToModify {boneName=dyn_jacket_side_04_l}
   }
   n162 = "Transform (Modify) Bone - Bone: dyn_jacket_rear_01_r"(ComponentPose=@n161, bAlphaBoolEnabled=@n153, Translation="0,0,0", Rotation="0,0,0", Scale="1,1,1") {
    alphaInputType=Bool
    rotationMode=BMM_Additive
    rotationSpace=BCS_BoneSpace
    translationMode=BMM_Additive
    translationSpace=BCS_BoneSpace
    alphaBoolBlend {blendOption=Linear}
    alphaScaleBiasClamp {bInterpResult=true}
    boneToModify {boneName=dyn_jacket_rear_01_r}
   }
   n163 = "Transform (Modify) Bone - Bone: dyn_jacket_rear_02_r"(ComponentPose=@n162, bAlphaBoolEnabled=@n153, Translation="0,0,0", Rotation="0,0,0", Scale="1,1,1") {
    alphaInputType=Bool
    rotationMode=BMM_Additive
    rotationSpace=BCS_BoneSpace
    translationMode=BMM_Additive
    translationSpace=BCS_BoneSpace
    alphaBoolBlend {blendOption=Linear}
    alphaScaleBiasClamp {bInterpResult=true}
    boneToModify {boneName=dyn_jacket_rear_02_r}
   }
   n164 = "Transform (Modify) Bone - Bone: dyn_jacket_rear_03_r"(ComponentPose=@n163, bAlphaBoolEnabled=@n153, Translation="0,0,2", Rotation="0,0,0", Scale="1,1,1") {
    alphaInputType=Bool
    rotationMode=BMM_Additive
    rotationSpace=BCS_BoneSpace
    translationMode=BMM_Additive
    translationSpace=BCS_BoneSpace
    alphaBoolBlend {blendOption=Linear}
    alphaScaleBiasClamp {bInterpResult=true}
    boneToModify {boneName=dyn_jacket_rear_03_r}
    translation {z=2}
   }
   n165 = "Transform (Modify) Bone - Bone: dyn_jacket_side_01_r"(ComponentPose=@n164, bAlphaBoolEnabled=@n153, Translation="0,0,0", Rotation="0,0,0", Scale="1,1,1") {
    alphaInputType=Bool
    rotationMode=BMM_Additive
    rotationSpace=BCS_BoneSpace
    translationMode=BMM_Additive
    translationSpace=BCS_BoneSpace
    alphaBoolBlend {blendOption=Linear}
    alphaScaleBiasClamp {bInterpResult=true}
    boneToModify {boneName=dyn_jacket_side_01_r}
   }
   n166 = "Transform (Modify) Bone - Bone: dyn_jacket_side_02_r"(ComponentPose=@n165, bAlphaBoolEnabled=@n153, Translation="0,0,0", Rotation="0,0,0", Scale="1,1,1") {
    alphaInputType=Bool
    rotationMode=BMM_Additive
    rotationSpace=BCS_BoneSpace
    translationMode=BMM_Additive
    translationSpace=BCS_BoneSpace
    alphaBoolBlend {blendOption=Linear}
    alphaScaleBiasClamp {bInterpResult=true}
    boneToModify {boneName=dyn_jacket_side_02_r}
   }
   n167 = "Transform (Modify) Bone - Bone: dyn_jacket_side_03_r"(ComponentPose=@n166, bAlphaBoolEnabled=@n153, Translation="0,0,0", Rotation="0,0,0", Scale="1,1,1") {
    alphaInputType=Bool
    rotationMode=BMM_Additive
    rotationSpace=BCS_BoneSpace
    translationMode=BMM_Additive
    translationSpace=BCS_BoneSpace
    alphaBoolBlend {blendOption=Linear}
    alphaScaleBiasClamp {bInterpResult=true}
    boneToModify {boneName=dyn_jacket_side_03_r}
   }
   n168 = "Transform (Modify) Bone - Bone: mstr_jacket_front_01_l"(ComponentPose=@n167, bAlphaBoolEnabled=Get bIsPlayingEmote(), Translation="0,0,0", Rotation="0,0,0", Scale="1,1,1") {
    alphaInputType=Bool
    rotationMode=BMM_Additive
    rotationSpace=BCS_BoneSpace
    translationMode=BMM_Additive
    translationSpace=BCS_BoneSpace
    alphaBoolBlend {blendOption=Linear}
    alphaScaleBiasClamp {bInterpResult=true}
    boneToModify {boneName=mstr_jacket_front_01_l}
   }
   Save cached pose 'DynPose'(Pose=Component To Local(ComponentPose=@n168)) {cacheName=DynPose}
   n169 = Get bIsInAir() -> (bIsInAir:bool)
   n170 = "Transform (Modify) Bone - Bone: dyn_jacket_rear_01_l"(
    ComponentPose=Local To Component(LocalPose=Use cached pose 'DynPose'())
    bAlphaBoolEnabled=@n169
    Translation="0,0,0"
    Rotation="5,0,0"
    Scale="1,1,1"
   ) {
    alphaInputType=Bool
    rotationMode=BMM_Additive
    rotationSpace=BCS_BoneSpace
    translationMode=BMM_Additive
    translationSpace=BCS_BoneSpace
    alphaBoolBlend {blendOption=Linear}
    alphaScaleBiasClamp {bInterpResult=true}
    boneToModify {boneName=dyn_jacket_rear_01_l}
    rotation {pitch=5}
   }
   n171 = "Transform (Modify) Bone - Bone: dyn_jacket_rear_02_l"(ComponentPose=@n170, bAlphaBoolEnabled=@n169, Translation="0,0,0", Rotation="5,0,0", Scale="1,1,1") {
    alphaInputType=Bool
    rotationMode=BMM_Additive
    rotationSpace=BCS_BoneSpace
    translationMode=BMM_Additive
    translationSpace=BCS_BoneSpace
    alphaBoolBlend {blendOption=Linear}
    alphaScaleBiasClamp {bInterpResult=true}
    boneToModify {boneName=dyn_jacket_rear_02_l}
    rotation {pitch=5}
   }
   n172 = "Transform (Modify) Bone - Bone: dyn_jacket_rear_03_l"(ComponentPose=@n171, bAlphaBoolEnabled=@n169, Translation="0,0,0", Rotation="5,0,0", Scale="1,1,1") {
    alphaInputType=Bool
    rotationMode=BMM_Additive
    rotationSpace=BCS_BoneSpace
    translationMode=BMM_Additive
    translationSpace=BCS_BoneSpace
    alphaBoolBlend {blendOption=Linear}
    alphaScaleBiasClamp {bInterpResult=true}
    boneToModify {boneName=dyn_jacket_rear_03_l}
    rotation {pitch=5}
   }
   n173 = "Transform (Modify) Bone - Bone: dyn_jacket_side_01_r"(ComponentPose=@n172, bAlphaBoolEnabled=@n169, Translation="0,0,0", Rotation="0,0,0", Scale="1,1,1") {
    alphaInputType=Bool
    rotationMode=BMM_Additive
    rotationSpace=BCS_BoneSpace
    translationMode=BMM_Additive
    translationSpace=BCS_BoneSpace
    alphaBoolBlend {blendOption=Linear}
    alphaScaleBiasClamp {bInterpResult=true}
    boneToModify {boneName=dyn_jacket_side_01_r}
   }
   n174 = "Transform (Modify) Bone - Bone: dyn_jacket_rear_02_r"(ComponentPose=@n173, bAlphaBoolEnabled=@n169, Translation="0,0,0", Rotation="0,0,0", Scale="1,1,1") {
    alphaInputType=Bool
    rotationMode=BMM_Additive
    rotationSpace=BCS_BoneSpace
    translationMode=BMM_Additive
    translationSpace=BCS_BoneSpace
    alphaBoolBlend {blendOption=Linear}
    alphaScaleBiasClamp {bInterpResult=true}
    boneToModify {boneName=dyn_jacket_rear_02_r}
   }
   n175 = "Transform (Modify) Bone - Bone: dyn_jacket_rear_03_r"(ComponentPose=@n174, bAlphaBoolEnabled=@n169, Translation="0,0,0", Rotation="0,0,0", Scale="1,1,1") {
    alphaInputType=Bool
    rotationMode=BMM_Additive
    rotationSpace=BCS_BoneSpace
    translationMode=BMM_Additive
    translationSpace=BCS_BoneSpace
    alphaBoolBlend {blendOption=Linear}
    alphaScaleBiasClamp {bInterpResult=true}
    boneToModify {boneName=dyn_jacket_rear_03_r}
   }
   Save cached pose 'InAirPose'(Pose=Component To Local(ComponentPose=@n175)) {cacheName=InAirPose}
   Reroute Node(InputPin:Wildcard) -> (OutputPin:Wildcard)
   Get IsCrouchingAlpha()
   Get is Crouch Walking()
   Get SprintingAlpha() -> (SprintingAlpha:double)
  ]}
 EventGraph {nodes [
   Event BlueprintInitializeAnimation()
   > Set Skel Mesh(Skel Mesh=GetOwnerSkeletalMeshComponent())
   n1 = TryGetPawnOwner()
   > Is Valid(InputObject=@n1)
   > n2 = Cast To FortPlayerPawnAthena(Object=@n1)
   > Set PlayerPawn(PlayerPawn:Fort Player Pawn Athena Object Reference=@n2) -> (Output_Get:Fort Player Pawn Athena Object Reference)
   Event OnCharacterCustomizationLoaded()
   > n3 = Set Skel Mesh(Skel Mesh=GetPartSkeletalMeshComponent())
   > n4 = Get Dynamic Material Instance(SkeletalMeshComponent=@n3)
   > Set Mat Instance Dynamic(Mat Instance Dynamic=@n4)
   > n5 = GetOrCreateDynamicOverlayMaterial(MeshComponent=@n3, MaterialName=None)
   > Set Overlay Mat Instance Dynamic(Overlay Mat Instance Dynamic=@n5)
   n6 = Event BlueprintUpdateAnimation()
   > Set Delta Time X(Delta Time X:double=@n6.DeltaTimeX) -> (Output_Get:double)
   > n7 = Sequence!(execute) -> (then_0, then_1, then_2:Exec, then_3:Exec, then_4:Exec, then_5:Exec, then_6:Exec, then_7:Exec)
   > Set pawnSpeedRange(pawnSpeedRange:double=MapRangeClamped(Value=Get PawnSpeed2D(), InRangeA=300, InRangeB=550)) -> (Output_Get:double)
   Set thigh_rot_up_l(
    execute=@n7.then_1
    thigh_rot_up_l:double=GetAlphaFromReferenceTransformDelta(
     SkeletalMeshComponent=Get Skel Mesh()
     BoneName=thigh_l
     BaseName=pelvis
     DeltaComponentType=QuaternionTwist
     InputRange=(X=0,Y=100)
     TwistAxis="0,0,1"
    )
   ) -> (Output_Get:double)
   > Set thigh_rot_up_01_l(thigh_rot_up_01_l:double=GetAlphaFromReferenceTransformDelta(
     SkeletalMeshComponent=Get Skel Mesh()
     BoneName=thigh_l
     BaseName=pelvis
     DeltaComponentType=QuaternionTwist
     InputRange=(X=0,Y=50)
     TwistAxis="0,0,1"
    )) -> (Output_Get:double)
   > Set thigh_rot_up_02_l(thigh_rot_up_02_l:double=GetAlphaFromReferenceTransformDelta(
     SkeletalMeshComponent=Get Skel Mesh()
     BoneName=thigh_l
     BaseName=pelvis
     DeltaComponentType=QuaternionTwist
     InputRange=(X=40,Y=100)
     TwistAxis="0,0,1"
    )) -> (Output_Get:double)
   > Set thigh_rot_up_r(thigh_rot_up_r:double=GetAlphaFromReferenceTransformDelta(
     SkeletalMeshComponent=Get Skel Mesh()
     BoneName=thigh_r
     BaseName=pelvis
     DeltaComponentType=QuaternionTwist
     InputRange=(X=0,Y=100)
     TwistAxis="0,0,1"
    )) -> (Output_Get:double)
   > Set thigh_rot_up_r(thigh_rot_up_r=GetAlphaFromReferenceTransformDelta(
     SkeletalMeshComponent=Get Skel Mesh()
     BoneName=thigh_r
     BaseName=pelvis
     DeltaComponentType=QuaternionTwist
     InputRange=(X=0,Y=100)
     TwistAxis="0,0,1"
    ))
   Set pushOutLeftJacketAlpha(execute=@n7.then_2, pushOutLeftJacketAlpha:double=float * float() {is=title K2Node_PromotableOperator}) -> (Output_Get:double)
   > Branch(Condition=AND Boolean(A=Get bIsAccelerating2D(), B=Get bIsOnGround() -> (bIsOnGround:bool)))
   n8 = Sequence!(execute=@n7.then_3) -> (then_0, then_1, then_2, then_3, then_4, then_5, then_6, then_7, then_8:Exec, then_9:Exec)
   n9 = AND Boolean(A=Get bIsCrouching(), B=NOT Boolean(A=Get bIsMoving2D()))
   Set IsCrouchingAlpha(execute=@n8.then_2, IsCrouchingAlpha:double="To Float (Boolean)"(InBool=@n9)) -> (Output_Get:double)
   n10 = Get bIsSkyDiving()
   n11 = Set is In Air(
    execute=@n8.then_6
    is In Air:bool=OR Boolean!(A=Get bIsJumping(), B=Get bIsFalling(), C:bool=Get bIsSprinting(), D:bool=NOT Boolean(A=@n10)) -> (ReturnValue)
   ) -> (Output_Get:bool)
   > Set is in Air alpha(is in Air alpha:double="To Float (Boolean)"(InBool=@n11)) -> (Output_Get:double)
   n12 = Get bIsCrouching()
   n13 = Get bIsMoving2D()
   > Set is Crouch Walking(is Crouch Walking:double="To Float (Boolean)"(InBool=AND Boolean(A=@n13, B=@n12))) -> (Output_Get:double)
   n14 = OR Boolean!(A=Get bIsJumping(), B=Get bIsFalling()) -> (ReturnValue)
   n15 = Get bIsTacticalSprinting()
   > n16 = Branch(Condition=OR Boolean(A=Get bIsCrouching(), B=AND Boolean(A=@n15, B=@n14)))
   > Set Standing Fixes Coat(Standing Fixes Coat:double=0) -> (Output_Get:double)
   Set Standing Fixes Coat(execute=@n16.else, Standing Fixes Coat=1)
   n17 = Get bIsTacticalSprinting()
   n18 = Get bIsSprinting()
   Set Sprinting(execute=@n8.then_8, Sprinting:double="To Float (Boolean)"(InBool=OR Boolean(A=@n18, B=@n17))) -> (Output_Get:double)
   n19 = "To Float (Boolean)"(InBool=NOT Boolean(A=Get bIsCrouching()))
   n20 = Get Skel Mesh()
   > Set Leg_L_Up(Leg_L_Up:float=float * float!(
     A=GetAlphaFromReferenceTransformDelta(SkeletalMeshComponent=@n20, BoneName=calf_l, BaseName=thigh_l, InputRange=(X=0,Y=10), TwistAxis="0,0,1")
     B=Get Sprinting()
     C:double=@n19
    ) -> (ReturnValue)) -> (Output_Get:float)
   > Set Leg_R_Up(Leg_R_Up:float=float * float(
     A=GetAlphaFromReferenceTransformDelta(SkeletalMeshComponent=@n20, BoneName=calf_r, BaseName=thigh_r, InputRange=(X=0,Y=-30), TwistAxis="0,0,1")
     B=Get Sprinting()
     C=@n19
    )) -> (Output_Get:float)
   Set CrouchingStandingLookAtAlpha(execute=@n8.then_3, CrouchingStandingLookAtAlpha:double=SelectFloat(A=0.85, B=0, bPickA=@n9)) -> (Output_Get:double)
   > n21 = Branch(Condition=@n9)
   > n22 = Set CrouchingStanding(CrouchingStanding:double=1) -> (Output_Get:double)
   Set CrouchingStanding(
    execute=@n21.else
    CrouchingStanding=FInterpTo(Current=@n22, Target=0, DeltaTime=Get Delta Time X() -> (Delta Time X:double), InterpSpeed=20)
   )
   n23 = Branch(execute=@n8.then_4, Condition=AND Boolean(A=Get bIsCrouching(), B=Get bIsMoving2D()))
   > n24 = Set CrouchingMovingAlpha(CrouchingMovingAlpha:double=1) -> (Output_Get:double)
   Set CrouchingMovingAlpha(execute=@n23.else, CrouchingMovingAlpha=FInterpTo(Current=@n24, Target=0, DeltaTime=Get Delta Time X(), InterpSpeed=20))
   n25 = Get bIsCrouching()
   Set StandingAndNotCrouching(execute=@n8.then_5, StandingAndNotCrouching:bool=AND Boolean(A=Get IsStanding() -> (IsStanding:bool), B=NOT Boolean(A=@n25))) -> (Output_Get:bool)
   > Set SkyDivingBackwardsMovementAlpha(SkyDivingBackwardsMovementAlpha:double=float * float!(A:bool=Get bIsSkydiveFloating(), B=Get BackwardMovementAlpha() -> (BackwardMovementAlpha:double)) -> (ReturnValue)) -> (Output_Get:double)
   > Set Riding Grindrail(Riding Grindrail:double="To Float (Boolean)"(InBool=Get bIsOnGrindRail())) -> (Output_Get:double)
   > Set FrontEndandPayingEmote(FrontEndandPayingEmote:bool=AND Boolean(A=Get bIsInFrontEnd(), B=Get bIsPlayingEmote())) -> (Output_Get:bool)
   Is Valid(exec=@n8.then_0, InputObject=Get PlayerPawn() -> (PlayerPawn:Fort Player Pawn Athena Object Reference))
   > Set WPO Globals Function(
    Overlay Material=Get Overlay Mat Instance Dynamic()
    Skeletal Mesh=Get Skel Mesh()
    "Local Noise Transform  Location"="Get Local Noise Transform  Location"() -> ("Local Noise Transform  Location":Vector)
    Local Noise Offset=Get Local Noise Offset()
   )
   n26 = Get Delta Time X()
   > Set WPO Regional Function!(
    execute
    Material:Material Instance Dynamic Object Reference=Get Mat Instance Dynamic()
    Overlay Material:Material Instance Dynamic Object Reference=Get Overlay Mat Instance Dynamic()
    Skeletal Mesh:Skeletal Mesh Component Object Reference=Get Skel Mesh()
    Velocity Transform:Name=foot_r
    Velocity Interp Speed:double=Get Velocity Interp Speed()
    Collision Transform:Name=calf_r
    Collision Transform Axis:EAxis Enum
    Invert Collision Transform Axis:bool=false
    Delta Time X:double=@n26
    Previous Velocity Transform Location:"Vector (by ref)"=Get R1 Velocity Transform Location()
    Previous Velocity Transform Velocity:"Vector (by ref)"=Get R1 Velocity Transform Velocity()
    Region:int=1
   ) -> (then)
   > Set WPO Regional Function(
    Material=Get Mat Instance Dynamic()
    Overlay Material=Get Overlay Mat Instance Dynamic()
    Skeletal Mesh=Get Skel Mesh()
    Velocity Transform=foot_l
    Velocity Interp Speed=Get Velocity Interp Speed()
    Collision Transform=calf_l
    Invert Collision Transform Axis=true
    Delta Time X=@n26
    Previous Velocity Transform Location=Get R2 Velocity Transform Location()
    Previous Velocity Transform Velocity=Get R2 Velocity Transform Velocity()
    Region=2
   )
   > n27 = Branch(Condition=Get bIsSkydiveFloating())
   > n28 = Set WPO Pho Wind Intensity(WPO Pho Wind Intensity=1)
   SetScalarParameterValue(execute=[@n28.then, @n29.then], self=Get Mat Instance Dynamic(), ParameterName=ClothSim_WindStrength, Value=Get WPO Pho Wind Intensity())
   n29 = Set WPO Pho Wind Intensity(execute=@n27.else, WPO Pho Wind Intensity=0)
   n30 = Self-Reference()
   n31 = PlayerMovement(execute=@n8.then_7, AnimInstance=@n30, XDotProductCoefficient=100, YDotProductCoefficient=100, ZDotProductCoefficient=100)
   n32 = Get bIsSkyDiving()
   n33 = Get bIsCrouching()
   > n34 = Branch(Condition=OR Boolean(A=@n33, B=@n32))
   n35 = BreakVector(InVec=MakeVector(X=@n31, Y=@n31.YMovement))
   Set Left(execute=@n34.else, Left:double="To Float (Boolean)"(InBool="float < float"(A=@n35, B=-50))) -> (Output_Get:double)
   > Set Right(Right:double="To Float (Boolean)"(InBool="float > float"(A=@n35, B=50))) -> (Output_Get:double)
   > Set Forward(Forward:double="To Float (Boolean)"(InBool="float > float"(A=@n35.Y, B=40))) -> (Output_Get:double)
   > Set Backward(Backward:double="To Float (Boolean)"(InBool="float < float"(A=@n35.Y, B=-40))) -> (Output_Get:double)
   n36 = Get bIsCrouching()
   n37 = Get PawnSpeed2D()
   n38 = Get Skel Mesh()
   Set Is Twisting in Place(
    execute=@n8.then_9
    Is Twisting in Place:double=float * float!(
     A:double=GetAlphaFromReferenceTransformDelta(SkeletalMeshComponent=@n38, BoneName=thigh_r, BaseName=pelvis, InputRange=(X=0,Y=-30), TwistAxis="1,0,1")
     B=Remap Float(Input=@n37, OldRange=(X=0,Y=1), NewRange=(X=1,Y=0), Clamp=true) -> (Output=~)
     C:bool=Get bIsOnGround()
     D:bool=NOT Boolean(A=@n36)
    ) -> (ReturnValue)
   ) -> (Output_Get:double)
   n39 = Get bIsSprinting()
   n40 = Get bIsTacticalSprinting()
   > Set SprintingAlpha(SprintingAlpha:double="To Float (Boolean)"(InBool=AND Boolean!(A=@n39, B=@n40, C:bool) -> (ReturnValue))) -> (Output_Get:double)
   n41 = Get bIsTargeting()
   > Set ADSWalk(ADSWalk:float=float * float!(A=Get Forward(), B:bool=@n41) -> (ReturnValue)) -> (Output_Get:float)
   > Set Is Sliding Alpha(Is Sliding Alpha:double="To Float (Boolean)"(InBool=Get bIsSliding())) -> (Output_Get:double)
   > Set Is Knee Sliding Alpha(Is Knee Sliding Alpha:double="To Float (Boolean)"(InBool=Get bIsKneeSliding())) -> (Output_Get:double)
   > Set Is baseball Sliding Alpha(Is baseball Sliding Alpha:double="To Float (Boolean)"(InBool=Get bIsBaseballSliding())) -> (Output_Get:double)
   > n42 = Set Sliding(Sliding:bool=OR Boolean!(A=Get bIsSliding(), B=Get bIsKneeSliding(), C:bool=Get bIsBaseballSliding()) -> (ReturnValue)) -> (Output_Get:bool)
   > n43 = Branch(Condition=OR Boolean!(A=Get bIsOnGrindRail(), B=@n42) -> (ReturnValue))
   > Set NormalRbanAlpha(NormalRbanAlpha:double=0) -> (Output_Get:double)
   Set NormalRbanAlpha(execute=@n43.else, NormalRbanAlpha=1)
   Set simSpace(execute=@n7.then_4, simSpace:AnimPhysSimSpaceType Enum) -> (Output_Get:AnimPhysSimSpaceType Enum)
   > Branch(Condition=OR Boolean(A=Get bIsDBNO(), B=Get bIsInFrontEnd()))
   > n44 = Branch(Condition=Get bIsDBNOCarried())
   > Set simSpace(simSpace=World)
   Set simSpace(execute=@n44.else)
   Gravity Override State Function DEPRECATED(
    execute=@n7.then_5
    CharacterPartAnimInstance=Self-Reference()
    GameTimeInSeconds=GetGameTimeInSeconds()
    GravityOverrideParams=Get Gravity Override Params()
   )
   Set CrouchAndEmote(execute=@n7.then_6, CrouchAndEmote:bool=OR Boolean(A=Get bIsCrouching(), B=Get bIsPlayingEmote())) -> (Output_Get:bool)
   "Set Skydiving Battle Bus "(execute=@n7.then_7, "Skydiving Battle Bus ":double="To Float (Boolean)"(InBool=Get bIsSkyDivingFromBus() -> (bIsSkyDivingFromBus:bool))) -> (Output_Get:double)
   n45 = Get bIsParachuteOpened()
   n46 = Get bIsSkyDivingFromBus()
   > Set Skydiving Battle Bus _Parachute(Skydiving Battle Bus _Parachute:double="To Float (Boolean)"(InBool=OR Boolean(A=@n46, B=@n45))) -> (Output_Get:double)
   PlayerMovement(AnimInstance=@n30, XDotProductCoefficient=100, YDotProductCoefficient=100, ZDotProductCoefficient=100)
   "To Float (Boolean)"(InBool=AND Boolean!(A=Get bIsMoving2D(), B=@n41) -> (ReturnValue))
   NOT Boolean(A=AND Boolean(A=@n33, B=@n32))
   AND Boolean(A=@n46, B=@n45)
   Get Overlay Mat Instance Dynamic()
   Get bIsSkyDiving()
   Get bIsMoving2D()
   Get bIsMoving2D()
  ]}
]
variables [
 ADSWalk {type=number}
 Backward {type=number}
 BackwardMovementAlpha {type=number}
 Body Mesh {type=SkeletalMeshComponent}
 Component Linear Acc Scale {type=Vector}
 Component Linear Vel Scale {type=Vector}
 CrouchAndEmote {type=boolean}
 CrouchingMovingAlpha {type=number}
 CrouchingStanding {type=number}
 CrouchingStandingLookAtAlpha {type=number}
 Delta Time X {type=number, default=0.023}
 Forward {type=number}
 frontendAlpha {type=number}
 FrontEndandPayingEmote {type=boolean}
 Gravity Override {type=Vector}
 Gravity Override Params {type=GravityOverrideParamsStruct}
 Head_rot_f {type=number}
 Is baseball Sliding Alpha {type=number}
 is Crouch Walking {type=number}
 is In Air {type=boolean}
 is in Air alpha {type=number}
 Is Knee Sliding Alpha {type=number}
 Is Sliding Alpha {type=number}
 Is Twisting in Place {type=number}
 IsCrouchingAlpha {type=number}
 isCrouchNotMoving {type=number}
 IsStanding {type=boolean}
 L leg pitch {type=number}
 L leg up {type=number}
 L leg up crouch {type=number}
 Left {type=number}
 Leg_L_Up {type=number}
 Leg_R_Up {type=number}
 Local Noise Offset {type=Vector}
 "Local Noise Transform  Location" {type=Vector}
 Mat Instance Dynamic {type=MaterialInstanceDynamic}
 Moving 2d {type=number}
 NormalRbanAlpha {type=number}
 Overlay Mat Instance Dynamic {type=MaterialInstanceDynamic}
 pawnSpeedRange {type=number}
 PlayerPawn {type=FortPlayerPawnAthena}
 pushOutLeftJacketAlpha {type=number}
 R leg negative pitch {type=number}
 R leg pitch {type=number}
 R leg up {type=number}
 R Leg Up Crouch {type=number}
 R1 Velocity Transform Location {type=Vector}
 R1 Velocity Transform Velocity {type=Vector}
 R2 Velocity Transform Location {type=Vector}
 R2 Velocity Transform Velocity {type=Vector}
 RBAN Settings Out {
  type=FortRigidBodyAnimNodeParameters
  default {simSpaceSettings {maxAngularAcceleration=500, maxAngularVelocity=4, maxLinearAcceleration=700, maxLinearVelocity=1000, velocityScaleZ=0, worldAlpha=1}}
 }
 Riding Grindrail {type=number}
 Right {type=number}
 Shoulder_rot_f_l {type=number}
 Shoulder_rot_f_r {type=number}
 Shoulder_rot_up_l {type=number}
 Shoulder_rot_up_r {type=number}
 simSpace {type=AnimPhysSimSpaceType, default=RootRelative}
 Skel Mesh {type=SkeletalMeshComponent}
 "Skydiving Battle Bus " {type=number}
 Skydiving Battle Bus _Parachute {type=number}
 SkyDivingBackwardsMovementAlpha {type=number}
 Sliding {type=boolean}
 Sprinting {type=number}
 SprintingAlpha {type=number}
 Standing Fixes Coat {type=number}
 StandingAndNotCrouching {type=boolean}
 thigh_rot_up_01_l {type=number}
 thigh_rot_up_02_l {type=number}
 thigh_rot_up_l {type=number}
 thigh_rot_up_r {type=number}
 upperarm_rot_dn_l {type=number}
 upperarm_rot_dn_r {type=number}
 Velocity Interp Speed {type=number}
 WPO Pho Wind Intensity {type=number}
]
