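# Physics asset of cape colliders for the M_MED_CloakedIron medium male player cosmetic body.
# Capsules cover the pelvis, spine, clavicles and neck, with a sphere on the head, while arm, leg and foot bodies have no shapes.
# Capsules also sit on four-link dyn_jacket bone chains at center, front, rear and sides.
# It has no constraints, so it defines collision volumes only, with no joints between bodies.
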PhysicsAsset /BRCosmetics/Characters/Player/Male/Medium/Bodies/M_MED_CloakedIron/Meshes/M_MED_CloakedIron_CapeColliders
bodies {
 calf_l []
 calf_r []
 clavicle_l [{shapeType=Capsule, length=9.0192, radius=9.8871, center=<9.4266 -1.1806 -2.4977>, rotation=<rot 69.3297 1.4476 176.9462>}]
 clavicle_r [{shapeType=Capsule, length=9.0192, radius=9.8871, center=<-9.4266 1.1806 2.4977>, rotation=<rot 69.3297 1.4476 -3.0538>}]
 dyn_jacket_center_01 [{shapeType=Capsule, length=0.505, radius=9, shapeName=dyn_jacket_center_01_capsule, rotation=<rot 0 -90 0>}]
 dyn_jacket_center_02 [{shapeType=Capsule, length=15, radius=10, shapeName=dyn_jacket_center_02_capsule, center=<5 0 0>, rotation=<rot 0 0 -90>}]
 dyn_jacket_center_03 [{shapeType=Capsule, length=15, radius=10, shapeName=dyn_jacket_center_03_capsule, center=<10 0 0>, rotation=<rot 0 0 -90>}]
 dyn_jacket_center_04 [{shapeType=Capsule, length=10, radius=15, shapeName=dyn_jacket_center_04_capsule, center=<18 0 0>, rotation=<rot 0 0 -90>}]
 dyn_jacket_front_01_l [{
   shapeType=Capsule
   length=7.9123
   radius=9
   shapeName=dyn_jacket_front_01_l_capsule
   center=<12.3176 1.3002 0.5425>
   rotation=<rot 0 -97.4476 -87.3092>
  }]
 dyn_jacket_front_01_r [{
   shapeType=Capsule
   length=9.8142
   radius=9
   shapeName=dyn_jacket_front_01_r_capsule
   center=<13.8156 -0.4671 0.4027>
   rotation=<rot 0 -77.5603 -85.7406>
  }]
 dyn_jacket_front_02_l [{shapeType=Capsule, length=15, radius=10, shapeName=dyn_jacket_front_02_l_capsule, center=<5 0 0>, rotation=<rot 0 0 -90>}]
 dyn_jacket_front_02_r [{shapeType=Capsule, length=15, radius=10, shapeName=dyn_jacket_front_02_r_capsule, center=<5 0 0>, rotation=<rot 0 0 -90>}]
 dyn_jacket_front_03_l [{shapeType=Capsule, length=15, radius=10, shapeName=dyn_jacket_front_03_l_capsule, center=<10 0 0>, rotation=<rot 0 0 -90>}]
 dyn_jacket_front_03_r [{shapeType=Capsule, length=15, radius=10, shapeName=dyn_jacket_front_03_r_capsule, center=<10 0 0>, rotation=<rot 0 0 -90>}]
 dyn_jacket_front_04_l [{shapeType=Capsule, length=10, radius=15, shapeName=dyn_jacket_front_04_l_capsule, center=<18 0 0>, rotation=<rot 0 0 -90>}]
 dyn_jacket_front_04_r [{shapeType=Capsule, length=10, radius=15, shapeName=dyn_jacket_front_04_r_capsule, center=<18 0 0>, rotation=<rot 0 0 -90>}]
 dyn_jacket_rear_01_l [{
   shapeType=Capsule
   length=9.4015
   radius=9
   shapeName=dyn_jacket_rear_01_l_capsule
   center=<13.6841 -1.3714 -0.4899>
   rotation=<rot 0 -79.9391 -86.746>
  }]
 dyn_jacket_rear_01_r [{
   shapeType=Capsule
   length=7.9913
   radius=9
   shapeName=dyn_jacket_rear_01_r_capsule
   center=<12.1763 1.4204 -0.7799>
   rotation=<rot 0 -84.0404 -83.4042>
  }]
 dyn_jacket_rear_02_l [{shapeType=Capsule, length=15, radius=10, shapeName=dyn_jacket_rear_02_l_capsule, center=<5 0 0>, rotation=<rot 0 0 -90>}]
 dyn_jacket_rear_02_r [{shapeType=Capsule, length=15, radius=10, shapeName=dyn_jacket_rear_02_r_capsule, center=<5 0 0>, rotation=<rot 0 0 -90>}]
 dyn_jacket_rear_03_l [{shapeType=Capsule, length=15, radius=10, shapeName=dyn_jacket_rear_03_l_capsule, center=<10 0 0>, rotation=<rot 0 0 -90>}]
 dyn_jacket_rear_03_r [{shapeType=Capsule, length=15, radius=10, shapeName=dyn_jacket_rear_03_r_capsule, center=<10 0 0>, rotation=<rot 0 0 -90>}]
 dyn_jacket_rear_04_l [{shapeType=Capsule, length=10, radius=15, shapeName=dyn_jacket_rear_04_l_capsule, center=<18 0 0>, rotation=<rot 0 0 -90>}]
 dyn_jacket_rear_04_r [{shapeType=Capsule, length=10, radius=15, shapeName=dyn_jacket_rear_04_r_capsule, center=<18 0 0>, rotation=<rot 0 0 -90>}]
 dyn_jacket_side_01_l [{
   shapeType=Capsule
   length=8.0465
   radius=9
   shapeName=dyn_jacket_side_01_l_capsule
   center=<14.1751 -0.2165 0.0684>
   rotation=<rot 0 -93.1715 -87.3945>
  }]
 dyn_jacket_side_01_r [{
   shapeType=Capsule
   length=8.1302
   radius=9
   shapeName=dyn_jacket_side_01_r_capsule
   center=<14.5094 0.4794 -0.0303>
   rotation=<rot 0 -91.7875 -87.7746>
  }]
 dyn_jacket_side_02_l [{shapeType=Capsule, length=15, radius=10, shapeName=dyn_jacket_side_02_l_capsule, center=<5 0 0>, rotation=<rot 0 0 -90>}]
 dyn_jacket_side_02_r [{shapeType=Capsule, length=15, radius=10, shapeName=dyn_jacket_side_02_r_capsule, center=<5 0 0>, rotation=<rot 0 0 -90>}]
 dyn_jacket_side_03_l [{shapeType=Capsule, length=15, radius=10, shapeName=dyn_jacket_side_03_l_capsule, center=<10 0 0>, rotation=<rot 0 0 -90>}]
 dyn_jacket_side_03_r [{shapeType=Capsule, length=15, radius=10, shapeName=dyn_jacket_side_03_r_capsule, center=<10 0 0>, rotation=<rot 0 0 -90>}]
 dyn_jacket_side_04_l [{shapeType=Capsule, length=10, radius=15, shapeName=dyn_jacket_side_04_l_capsule, center=<18 0 0>, rotation=<rot 0 0 -90>}]
 dyn_jacket_side_04_r [{shapeType=Capsule, length=10, radius=15, shapeName=dyn_jacket_side_04_r_capsule, center=<18 0 0>, rotation=<rot 0 0 -90>}]
 foot_l []
 foot_r []
 head [{shapeType=Sphere, radius=17.3125, center=<11.028 -0.174 0>}]
 lowerarm_l []
 lowerarm_r []
 neck_02 [{shapeType=Capsule, length=13.4375, radius=7.3175, center=<-0.2642 -0.3408 0>, rotation=<rot 90 -171.1193 -175.3978>}]
 pelvis [{shapeType=Capsule, length=8.6456, radius=14.5388, center=<-2.4887 0.1881 0>}]
 spine_01 [{shapeType=Capsule, length=7.9438, radius=13.5586, center=<-1.7965 0.1884 0>}]
 spine_02 [
  {shapeType=Capsule, length=7.3264, radius=13.6322, center=<-1.711 -0.1851 0.0001>}
  {shapeType=Capsule, length=29.3956, radius=11.5807, center=<2.5024 0.4992 3.9825>, rotation=<rot 90 0.0049 0.0049>}
  {shapeType=Capsule, length=29.875, radius=11.1013, center=<2.4704 0.4991 -3.9825>, rotation=<rot 90 -0.0396 -0.0396>}
 ]
 spine_03 [{shapeType=Capsule, length=8.2107, radius=13.7031, center=<0.1445 0.2396 0.0001>}]
 spine_04 [{shapeType=Capsule, length=9.4404, radius=14.1325, center=<2.766 0.7765 0.0002>}]
 spine_05 [
  {shapeType=Capsule, length=8.0226, radius=15.695, center=<-2.2795 1.8109 0>}
  {shapeType=Capsule, length=13.9518, radius=13.9425, center=<-0.2916 0.4452 0>}
 ]
 thigh_l []
 thigh_r []
 upperarm_l []
 upperarm_r []
}
constraints []
