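# Physics asset for the face accessory part of the CloakedIron medium male player body cosmetic.
# It holds a single body on the head bone, built from four overlapping capsules of differing size and orientation that together approximate the head volume.
# It has no constraints, so it defines collision shapes only, with no jointed ragdoll chain.
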PhysicsAsset /BRCosmetics/Characters/Player/Male/Medium/Bodies/M_MED_CloakedIron/Meshes/Parts/M_MED_CloakedIron_FaceAcc_Physics
bodies {head [
  {shapeType=Capsule, length=8.3562, radius=5.1438, center=<3.9861 2.7911 0>, rotation=<rot 90 180 145>}
  {shapeType=Capsule, length=12.3092, radius=2.0036, shapeName=head_capsule, center=<-2.281 -1.8731 -0.0061>, rotation=<rot 0 -0.0005 -0.7709>}
  {shapeType=Capsule, length=10, radius=4, shapeName=head_capsule, center=<8.095 5.324 0>}
  {
   shapeType=Capsule
   length=3.638
   radius=4.522
   shapeName=head_capsule
   center=<3.5108 5.5479 1.0651>
   rotation=<rot 63.433 -83.9999 -110.4179>
  }
 ]}
constraints []
